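# Niagara system for the GnatGala Pancake traversal emote, adding water and fog effects around the moving character.
# Skeletal-mesh-attached emitters draw a ribbon liquid trail, rising-and-fading ripple meshes, and "Ribbons001" meshes faded by the wake-opacity user float.
# Fog_system_Condensed merges the disabled FOG and Ground_fog001 sprite emitters through spawn-group switches, varying with owner velocity and front-end context.
# Time-of-day functions tune color and opacity.
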NiagaraSystem /BRCosmetics/Animation/Game/MainPlayer/Emotes/GnatGala_Pancake/FX/NS_Emote_GnatGalaPancake_Water_Traversal
dependencies {
 usedDataInterfaces [
  &NiagaraDataInterfaceCurve
  &NiagaraDataInterfacePlatformSet
  &NiagaraDataInterfaceVectorCurve
  &NiagaraDataInterfaceArrayFloat
  &NiagaraDataInterfaceMeshRendererInfo
  &NiagaraDataInterfaceArrayFloat3
  &NiagaraDataInterfaceSkeletalMesh
 ]
 usedDynamicInputs [&RemapRange, &Add_Float, &Abs_Float]
 usedModules [
  &self:SystemSpawnScript.SystemScriptSource.SystemScriptGraph.NiagaraNodeAssignment_1.SetVariables_#1
  &self:SystemSpawnScript.SystemScriptSource.SystemScriptGraph.NiagaraNodeAssignment_3.SetVariables_#2
  &SystemState
  &BlendSpaceHelper
  &self:SystemSpawnScript.SystemScriptSource.SystemScriptGraph.NiagaraNodeAssignment_2.SetVariables_#3
  &self:SystemSpawnScript.SystemScriptSource.SystemScriptGraph.NiagaraNodeAssignment_4.SetVariables_#4
 ]
}
emitters {
 FOG {
  bEnabled=false
  emitterUpdateScript [
   &EmitterState(Inactive Response=::NewEnumerator1, Life Cycle Mode=::NewEnumerator1, Loop Duration=1, UseLoopDelay=false)
   &SpawnRate(
    Spawn Probability=0.5
    SpawnRate=&FN_FrontEndFloatModifier(
     Gameplay=&MakeCustomFloatFromBool(
      Bool=&SetBoolByFloatComparison(A=&VectorLength(Vector=$User.OwnerLocalVelocity), B=90)
      True Float=10
      False Float=7
     )
     Front End=7
    )
    Use Spawn Probability=true
   )
  ]
  particleSpawnScript [
   &InitializeParticle(
    Color=$System.Global color
    Lifetime=&RandomRangeFloat(Minimum=2, Maximum=2.5)
    Position Offset=<0 0 -15>
    Sprite Rotation Mode=::NewEnumerator2
    Sprite Size Mode=::NewEnumerator1
    Uniform Sprite Size Max=35
    Uniform Sprite Size Min=25
    UsePositionOffset=false
    Write Material Random=true
   )
   &AddVelocity(Velocity=&MakeVector(
     X=0
     Y=&FN_FrontEndFloatModifier(
      Gameplay=&MakeCustomFloatFromBool(
       Bool=&SetBoolByFloatComparison(A=&VectorLength(Vector=$User.OwnerLocalVelocity), B=90)
       True Float=&RandomRangeFloat(Minimum=-20, Maximum=-50)
       False Float=&RandomRangeFloat(Minimum=-10, Maximum=-15)
      )
      Front End=7
     )
     Z=0
    ))
   &ShapeLocation(MaskSpawnGroup=false, Non Uniform Scale=<1 1 0.5>, Override Seed=false, Sphere Radius=85, UseOwnerScale=true)
  ]
  particleUpdateScript [
   &ParticleState()
   &SubUVAnimation(
    End Frame=15
    SubUV Animation Mode:ENiagara_SubUVLookupMode=::NewEnumerator1
    SubUV Curve=DataInterface() {propertyValues {Curve {keys [{tangentMode=RCTM_User, time=0, value=0}, {tangentMode=RCTM_User, time=1.0138, value=0.388}]}}}
   )
   &Drag(
    Drag=&Multiply_Float(
     A=&RandomRangeFloat(Minimum=0.5, Maximum=1)
     B=&FloatFromCurve(FloatCurve=DataInterface() {propertyValues {Curve {keys [
          {interpMode=RCIM_Cubic, leaveTangent=0.248, tangentMode=RCTM_Break, time=0, value=0}
          {interpMode=RCIM_Cubic, tangentMode=RCTM_Break, time=1, value=1}
         ]}}})
    )
    Use Linear Drag=true
    Use Rotational Drag=false
   )
   &SolveForcesAndVelocity(Clamp Velocity=false, Limit Acceleration=false)
   &ScaleColor(
    Scale Alpha=&FloatFromCurve(
     FloatCurve=DataInterface() {propertyValues {Curve {keys [
         {arriveTangent=3.5962, interpMode=RCIM_Cubic, leaveTangent=3.5962, tangentMode=RCTM_User, time=0, value=0}
         {
          arriveTangent=0.1817
          arriveTangentWeight=0.1842
          interpMode=RCIM_Cubic
          leaveTangent=0.1817
          leaveTangentWeight=0.7369
          tangentMode=RCTM_User
          time=0.5037
          value=0.7647
         }
         {arriveTangent=-4.7233, tangentMode=RCTM_User, time=1, value=0}
        ]}}}
     Scale Curve=&RandomRangeFloat(
      Minimum=0.02
      Maximum=&FNTimeOfDayScalar(Night Scale=0.03, Morning Scale=0.03, Day Scale=0.05, Evening Scale=0.03)
     )
    )
    ScaleA=true
    ScaleRGB=true
   )
   &ScaleSpriteSize(
    Uniform Curve Scale=10
    Uniform Curve Sprite Scale=DataInterface() {propertyValues {Curve {defaultValue=1, keys [{tangentMode=RCTM_User, time=0, value=0.7}, {tangentMode=RCTM_User, time=1, value=1}]}}}
   )
   &DynamicMaterialParameters(
    Index 0 Param 2=1
    Index 0 Param 3=1
    Index 0 Param 4=1
    Module.Index0Param0Write=true
    Module.Index0Param1Write=true
    Module.Index0Param2Write=true
    Module.Index0Param3Write=true
   )
   &BlendSpace("Position Blend (World <-> Local)"=0.67, Rotation Contribution=0, Translation Contribution=0.85) {enabled=false}
   &BlendSpace(Rotation Contribution=0.5, Translation Contribution=0.95) {enabled=false, module=BlendSpace001}
   &BlendSpace("Position Blend (World <-> Local)":NiagaraFloat=&MakeCustomFloatFromBool(
     Bool=&SetBoolByFloatComparison(A=&VectorLength(Vector=$Engine.Owner.Velocity), B=400)
     True Float=1
     False Float=&FloatFromCurve(FloatCurve=DataInterface() {propertyValues {Curve {keys [<at 0 1 RCIM_Cubic>, {arriveTangent=-3.5186, interpMode=RCIM_Cubic, leaveTangent=-3.5186, tangentMode=RCTM_User, time=0.8, value=0}]}}})
     Translation Contribution:NiagaraFloat=1
     Rotation Contribution:NiagaraFloat=1
    )) {module=BlendSpace002}
  ]
  rendererClasses [&NiagaraSpriteRendererProperties]
  renderers [{rendererClass=&NiagaraSpriteRendererProperties}]
 }
 Fog_system_Condensed {
  emitterUpdateScript [
   &EmitterState(Life Cycle Mode=::NewEnumerator1, Loop Duration=1, UseLoopDelay=false)
   &SpawnRate(
    Spawn Probability=0.5
    SpawnGroup=1
    SpawnRate=&FN_FrontEndFloatModifier(
     Gameplay=&MakeCustomFloatFromBool(
      Bool=&SetBoolByFloatComparison(A=&VectorLength(Vector=$User.OwnerLocalVelocity), B=90)
      True Float=45
      False Float=37
     )
     Front End=45
    )
    Use Spawn Probability=true
   ) {module=SpawnRate001}
   &SpawnRate(
    Spawn Probability=0.5
    SpawnRate=&FN_FrontEndFloatModifier(
     Gameplay=&MakeCustomFloatFromBool(
      Bool=&SetBoolByFloatComparison(A=&VectorLength(Vector=$User.OwnerLocalVelocity), B=90)
      True Float=10
      False Float=7
     )
     Front End=12
    )
    Use Spawn Probability=true
   )
  ]
  particleSpawnScript [
   /SetVariables(Particles.VisibilityTag=&SpawnGroup()) {isSetParameters=true}
   &InitializeParticle(
    Color=$System.Global color
    Lifetime=&MergeHelper_FloatSwitch(
     0=&RandomRangeFloat(Minimum=1.5, Maximum=2.5)
     1=&RandomRangeFloat(Minimum=2, Maximum=2.4)
     2=?
     3=?
     Index=&SpawnGroup()
     Num=::NewEnumerator0
    )
    Sprite Size=&MergeHelper_Vector2Switch(
     0=&Vector2DFromFloat(Value=&RandomRangeFloat(Minimum=25, Maximum=35))
     1=&RandomRangeVector2D(Minimum=<20 10>, Maximum=<50 35>)
     2=?
     3=?
     4=?
     Index=&SpawnGroup()
     Num=::NewEnumerator0
    )
    Sprite Size Mode=::NewEnumerator4
    UsePositionOffset=false
    Write Material Random=true
   )
   /SetVariables(Particles.0 - Env Fog Velocity:Vector3f=&MakeVector(
     X=0
     Y=&FN_FrontEndFloatModifier(
      Gameplay=&MakeCustomFloatFromBool(
       Bool=&SetBoolByFloatComparison(A=&VectorLength(Vector=$User.OwnerLocalVelocity), B=90)
       True Float=&RandomRangeFloat(Minimum=-20, Maximum=-50)
       False Float=&RandomRangeFloat(Minimum=-10, Maximum=-15)
      )
      Front End=7
     )
     Z=0
    )) {isSetParameters=true}
   /SetVariables(Particles.1 - Ground Fog Velocity:Vector3f=&MakeVector(
     X=0
     Y=&FN_FrontEndFloatModifier(
      Gameplay=&MakeCustomFloatFromBool(
       Bool=&SetBoolByFloatComparison(A=&VectorLength(Vector=$User.OwnerLocalVelocity), B=90)
       True Float=-80
       False Float=-80
      )
      Front End=-80
     )
     Z=0
    )) {isSetParameters=true}
   &ShapeLocation(
    MaskSpawnGroup=false
    Non Uniform Scale=&MergeHelper_VectorSwitch(0=<1 1 0.5>, 1=<1 1 0.1>, 2=?, 3=?, Index=&SpawnGroup(), Num=::NewEnumerator0)
    Offset=&MergeHelper_VectorSwitch(
     0=&FN_FrontEndVector3Modifier(Gameplay=<0 0 0>, Front End=<0 0 0>)
     1=<0 50 -50>
     2=?
     3=?
     Index=&SpawnGroup()
     Num=::NewEnumerator0
    )
    Offset Mode=::NewEnumerator0
    Override Seed=false
    Sphere Radius=&MergeHelper_FloatSwitch(0=85, 1=90, 2=?, 3=?, Index=&SpawnGroup(), Num=::NewEnumerator0)
    UseOwnerScale=true
   )
   &AddVelocity(Velocity=&MergeHelper_VectorSwitch(
     0=&MakeVector(
      X=0
      Y=&FN_FrontEndFloatModifier(
       Gameplay=&MakeCustomFloatFromBool(
        Bool=&SetBoolByFloatComparison(A=&VectorLength(Vector=$User.OwnerLocalVelocity), B=90)
        True Float=&RandomRangeFloat(Minimum=-20, Maximum=-50)
        False Float=&RandomRangeFloat(Minimum=-10, Maximum=-15)
       )
       Front End=&RandomRangeFloat(Minimum=-20, Maximum=-50)
      )
      Z=0
     )
     1=&MakeVector(
      X=0
      Y=&FN_FrontEndFloatModifier(
       Gameplay=&MakeCustomFloatFromBool(
        Bool=&SetBoolByFloatComparison(A=&VectorLength(Vector=$User.OwnerLocalVelocity), B=90)
        True Float=-80
        False Float=-85
       )
       Front End=-80
      )
      Z=0
     )
     2=?
     3=?
     Index=&SpawnGroup()
     Num=::NewEnumerator0
    ))
  ]
  particleUpdateScript [
   &ParticleState()
   &ScaleSpriteSize(
    Uniform Curve Scale=10
    Uniform Curve Sprite Scale=DataInterface() {propertyValues {Curve {defaultValue=1, keys [{tangentMode=RCTM_User, time=0, value=0.7}, {tangentMode=RCTM_User, time=1, value=1}]}}}
   )
   &SubUVAnimation(
    End Frame=15
    SubUV Animation Mode:ENiagara_SubUVLookupMode=::NewEnumerator1
    SubUV Curve=DataInterface() {propertyValues {Curve {keys [{tangentMode=RCTM_User, time=0, value=0}, {tangentMode=RCTM_User, time=1.0138, value=0.388}]}}}
   )
   &Drag(
    Drag=&Multiply_Float(
     A=&RandomRangeFloat(Minimum=0.5, Maximum=1)
     B=&FloatFromCurve(FloatCurve=DataInterface() {propertyValues {Curve {keys [
          {interpMode=RCIM_Cubic, leaveTangent=0.248, tangentMode=RCTM_Break, time=0, value=0}
          {interpMode=RCIM_Cubic, tangentMode=RCTM_Break, time=1, value=1}
         ]}}})
    )
    Use Linear Drag=true
    Use Rotational Drag=false
   )
   &SolveForcesAndVelocity(Clamp Velocity=false, Limit Acceleration=false)
   /SetVariables(
    Particles.0 - Env Fog Scale Color:LinearColor=&MakeLinearColorFromVectorAndFloat(
     "Vector (RGB)"=<1 1 1>
     "Float (Alpha)"=&FloatFromCurve(
      FloatCurve=DataInterface() {propertyValues {Curve {keys [
          {arriveTangent=3.5962, interpMode=RCIM_Cubic, leaveTangent=3.5962, tangentMode=RCTM_User, time=0, value=0}
          {
           arriveTangent=0.1817
           arriveTangentWeight=0.1842
           interpMode=RCIM_Cubic
           leaveTangent=0.1817
           leaveTangentWeight=0.7369
           tangentMode=RCTM_User
           time=0.5037
           value=0.7647
          }
          {arriveTangent=-4.7233, tangentMode=RCTM_User, time=1, value=0}
         ]}}}
      Scale Curve=&RandomRangeFloat(
       Minimum=0.02
       Maximum=&FNTimeOfDayScalar(Night Scale=0.05, Morning Scale=0.05, Day Scale=0.07, Evening Scale=0.05)
      )
     )
    )
    Particles.1 - Ground Fog Scale Color:LinearColor=&MakeLinearColorFromVectorAndFloat(
     "Vector (RGB)"=<1 1 1>
     "Float (Alpha)"=&FloatFromCurve(
      FloatCurve=DataInterface() {propertyValues {Curve {keys [
          {arriveTangent=5, interpMode=RCIM_Cubic, time=0, value=0}
          {arriveTangent=1.5181, interpMode=RCIM_Cubic, leaveTangent=1.5181, time=0.2, value=0.7}
          {
           arriveTangent=0.1817
           arriveTangentWeight=0.1842
           interpMode=RCIM_Cubic
           leaveTangent=0.1817
           leaveTangentWeight=0.7369
           tangentMode=RCTM_User
           time=0.5037
           value=0.7647
          }
          <at 1 0>
         ]}}}
      Scale Curve=&RandomRangeFloat(
       Minimum=&FNTimeOfDayScalar(Night Scale=0.035, Morning Scale=0.07, Day Scale=0.05, Evening Scale=0.07)
       Maximum=&FNTimeOfDayScalar(Night Scale=0.05, Morning Scale=0.07, Day Scale=0.07, Evening Scale=0.1)
      )
     )
    )
   ) {isSetParameters=true}
   /SetVariables(
    Particles.0 - Env Fog Dyn Param:Vector4f=&MakeVector4(X=0, Y=1, Z=1, W=1)
    Particles.1 - Ground Fog Dyn Param:Vector4f=&MakeVector4(X=0, Y=1, Z=0.1, W=1)
   ) {isSetParameters=true}
   &BlendSpace(Rotation Contribution=0.5, Translation Contribution=0.95) {enabled=false}
   &ScaleColor(
    Scale Alpha=&MergeHelper_FloatSwitch(
     0=&FloatFromCurve(
      FloatCurve=DataInterface() {propertyValues {Curve {keys [
          {arriveTangent=3.5962, interpMode=RCIM_Cubic, leaveTangent=3.5962, tangentMode=RCTM_User, time=0, value=0}
          {
           arriveTangent=0.1817
           arriveTangentWeight=0.1842
           interpMode=RCIM_Cubic
           leaveTangent=0.1817
           leaveTangentWeight=0.7369
           tangentMode=RCTM_User
           time=0.5037
           value=0.7647
          }
          {arriveTangent=-4.7233, tangentMode=RCTM_User, time=1, value=0}
         ]}}}
      Scale Curve=&RandomRangeFloat(
       Minimum=0.02
       Maximum=&FNTimeOfDayScalar(Night Scale=0.03, Morning Scale=0.03, Day Scale=0.05, Evening Scale=0.03)
      )
     )
     1=&FloatFromCurve(
      FloatCurve=DataInterface() {propertyValues {Curve {keys [
          {arriveTangent=6.4271, interpMode=RCIM_Cubic, leaveTangent=6.4271, tangentMode=RCTM_User, time=0, value=0}
          {arriveTangent=0.5834, interpMode=RCIM_Cubic, leaveTangent=0.5834, tangentMode=RCTM_User, time=0.1413, value=0.7}
          {
           arriveTangent=0.1817
           arriveTangentWeight=0.1842
           interpMode=RCIM_Cubic
           leaveTangent=0.1817
           leaveTangentWeight=0.7369
           tangentMode=RCTM_User
           time=0.5037
           value=0.8
          }
          <at 1 0>
         ]}}}
      Scale Curve=&RandomRangeFloat(
       Minimum=&FNTimeOfDayScalar(Night Scale=0.035, Morning Scale=0.03, Day Scale=0.05, Evening Scale=0.04)
       Maximum=&FNTimeOfDayScalar(Night Scale=0.05, Morning Scale=0.05, Day Scale=0.07, Evening Scale=0.07)
      )
     )
     2=?
     3=?
     Index=&SpawnGroup()
     Num=::NewEnumerator0
    )
    ScaleA=true
    ScaleRGB=true
   )
   &BlendSpace("Position Blend (World <-> Local)":NiagaraFloat=&MakeCustomFloatFromBool(
     Bool=&SetBoolByFloatComparison(A=&VectorLength(Vector=$Engine.Owner.Velocity), B=400)
     True Float=1
     False Float=&MergeHelper_FloatSwitch(
      0=&FloatFromCurve(FloatCurve=DataInterface() {propertyValues {Curve {keys [<at 0 1 RCIM_Cubic>, {arriveTangent=-3.5186, interpMode=RCIM_Cubic, leaveTangent=-3.5186, tangentMode=RCTM_User, time=0.8, value=0}]}}})
      1=&FloatFromCurve(FloatCurve=DataInterface() {propertyValues {Curve {keys [<at 0 1 RCIM_Cubic>, {arriveTangent=-3.5186, interpMode=RCIM_Cubic, leaveTangent=-3.5186, tangentMode=RCTM_User, time=0.5, value=0}]}}})
      2=?
      3=?
      Index=&SpawnGroup()
      Num=::NewEnumerator0
     )
     Translation Contribution:NiagaraFloat=1
     Rotation Contribution:NiagaraFloat=1
    )) {module=BlendSpace002}
  ]
  rendererClasses [&NiagaraSpriteRendererProperties]
  renderers [
   {rendererClass=&NiagaraSpriteRendererProperties}
   {rendererClass=&NiagaraSpriteRendererProperties, rendererIndex=1}
  ]
 }
 Ground_fog001 {
  bEnabled=false
  emitterUpdateScript [
   &EmitterState(Inactive Response=::NewEnumerator1, Life Cycle Mode=::NewEnumerator1, Loop Duration=1, UseLoopDelay=false)
   &SpawnRate(
    Spawn Probability=0.5
    SpawnRate=&FN_FrontEndFloatModifier(
     Gameplay=&MakeCustomFloatFromBool(
      Bool=&SetBoolByFloatComparison(A=&VectorLength(Vector=$User.OwnerLocalVelocity), B=90)
      True Float=45
      False Float=37
     )
     Front End=7
    )
    Use Spawn Probability=true
   )
  ]
  particleSpawnScript [
   &InitializeParticle(
    Color=$System.Global color
    Lifetime=&RandomRangeFloat(Minimum=2.3, Maximum=2.7)
    Position Offset=<0 0 65>
    Sprite Size Max=<50 35>
    Sprite Size Min=<20 10>
    Sprite Size Mode=::NewEnumerator2
    UsePositionOffset=false
    Write Material Random=true
   )
   &AddVelocity(Velocity=&MakeVector(
     X=0
     Y=&FN_FrontEndFloatModifier(
      Gameplay=&MakeCustomFloatFromBool(
       Bool=&SetBoolByFloatComparison(A=&VectorLength(Vector=$User.OwnerLocalVelocity), B=90)
       True Float=-80
       False Float=-85
      )
      Front End=-80
     )
     Z=0
    ))
   &ShapeLocation(
    Box Size=<100 200 0>
    MaskSpawnGroup=false
    Non Uniform Scale=<1 1 0.1>
    Offset=<0 -40 -55>
    Offset Mode=::NewEnumerator0
    Override Seed=false
    Rotation Coordinate Space=::Simulation
    Shape Primitive=::NewEnumerator2
    Surface Only=false
    UseOwnerScale=true
   ) {enabled=false}
   &ShapeLocation(
    MaskSpawnGroup=false
    Non Uniform Scale=<1 1 0.1>
    Offset=<0 50 -50>
    Offset Coordinate Space=::Local
    Offset Mode=::NewEnumerator0
    Override Seed=false
    Sphere Radius=90
    UseOwnerScale=true
   ) {module=ShapeLocation001}
  ]
  particleUpdateScript [
   &ParticleState()
   &SubUVAnimation(
    End Frame=15
    SubUV Animation Mode:ENiagara_SubUVLookupMode=::NewEnumerator1
    SubUV Curve=DataInterface() {propertyValues {Curve {keys [{tangentMode=RCTM_User, time=0, value=0}, {tangentMode=RCTM_User, time=1.0138, value=0.388}]}}}
   )
   &Drag(
    Drag=&Multiply_Float(
     A=&RandomRangeFloat(Minimum=0.5, Maximum=1)
     B=&FloatFromCurve(FloatCurve=DataInterface() {propertyValues {Curve {keys [
          {interpMode=RCIM_Cubic, leaveTangent=0.248, tangentMode=RCTM_Break, time=0, value=0}
          {interpMode=RCIM_Cubic, tangentMode=RCTM_Break, time=1, value=1}
         ]}}})
    )
    Use Linear Drag=true
    Use Rotational Drag=false
   )
   &ScaleColor(Scale Alpha=$User.AnimCurve_fogIntensity, ScaleA=true, ScaleRGB=false) {enabled=false, module=ScaleColor001}
   &ScaleColor(
    Scale Alpha=&FloatFromCurve(
     FloatCurve=DataInterface() {propertyValues {Curve {keys [
         {arriveTangent=6.4271, interpMode=RCIM_Cubic, leaveTangent=6.4271, tangentMode=RCTM_User, time=0, value=0}
         {arriveTangent=0.5834, interpMode=RCIM_Cubic, leaveTangent=0.5834, tangentMode=RCTM_User, time=0.1413, value=0.7}
         {
          arriveTangent=0.1817
          arriveTangentWeight=0.1842
          interpMode=RCIM_Cubic
          leaveTangent=0.1817
          leaveTangentWeight=0.7369
          tangentMode=RCTM_User
          time=0.5037
          value=0.8
         }
         <at 1 0>
        ]}}}
     Scale Curve=&RandomRangeFloat(
      Minimum=&FNTimeOfDayScalar(Night Scale=0.035, Morning Scale=0.03, Day Scale=0.05, Evening Scale=0.04)
      Maximum=&FNTimeOfDayScalar(Night Scale=0.05, Morning Scale=0.05, Day Scale=0.07, Evening Scale=0.07)
     )
    )
    ScaleA=true
    ScaleRGB=true
   )
   &ScaleSpriteSize(
    Uniform Curve Scale=10
    Uniform Curve Sprite Scale=DataInterface() {propertyValues {Curve {defaultValue=1, keys [{tangentMode=RCTM_User, time=0, value=0.7}, {tangentMode=RCTM_User, time=1, value=1}]}}}
   )
   &DynamicMaterialParameters(
    Index 0 Param 2=1
    Index 0 Param 3=0.1
    Index 0 Param 4=1
    Module.Index0Param0Write=true
    Module.Index0Param1Write=true
    Module.Index0Param2Write=true
    Module.Index0Param3Write=true
   )
   &BlendSpace(Rotation Contribution=0.5, Translation Contribution=0.95) {enabled=false}
   &SolveForcesAndVelocity(Clamp Velocity=false, Limit Acceleration=false)
   &BlendSpace("Position Blend (World <-> Local)":NiagaraFloat=&MakeCustomFloatFromBool(
     Bool=&SetBoolByFloatComparison(A=&VectorLength(Vector=$Engine.Owner.Velocity), B=400)
     True Float=1
     False Float=&FloatFromCurve(FloatCurve=DataInterface() {propertyValues {Curve {keys [<at 0 1 RCIM_Cubic>, {arriveTangent=-3.5186, interpMode=RCIM_Cubic, leaveTangent=-3.5186, tangentMode=RCTM_User, time=0.5, value=0}]}}})
     Translation Contribution:NiagaraFloat=1
     Rotation Contribution:NiagaraFloat=1
    )) {module=BlendSpace001}
  ]
  rendererClasses [&NiagaraSpriteRendererProperties]
  renderers [{rendererClass=&NiagaraSpriteRendererProperties}]
 }
 Liquid_Trail {
  emitterUpdateScript [
   &EmitterState(
    Inactive Response=::NewEnumerator1
    Life Cycle Mode=::NewEnumerator1
    Loop Behavior=::NewEnumerator1
    Loop Delay=0.07
    Loop Duration Mode=::NewEnumerator1
    UseLoopDelay=false
   )
   &FN_SpawnAnimTrail(AnimTrailWidthMode=::FromFirst, Spawn Mode:ENiagaraSkeletalSpawning=::NewEnumerator0)
  ]
  particleSpawnScript [
   &FN_SetSpawnGroup(Number of Spawn Groups=1)
   &FN_InitializeAnimTrailBoneIndecies()
   &FN_InitializeAnimTrail(AnimTrailWidth=90, Lifetime=1)
   &SkeletalMeshLocation(
    Bone Sampling Mode=::NewEnumerator4
    MaskSpawnGroup=false
    Orientation Sampling=::NewEnumerator1
    Sampled Position Offset=<0 -30 -50>
    Sampled Scale Coordinate Space:ENiagaraCoordinateSpace=::Simulation
    Skeletal Mesh=$System.Skeletal Mesh
   )
   &Color(Color=&FNTimeOfDayColor(
     Night Color=&RandomRangeLinearColor(Minimum=<rgba 0.1881 0.2601 0.5698 1>, Maximum=<rgba 0.1335 0.2918 0.5698 1>)
     Morning Color=&RandomRangeLinearColor(Minimum=<rgba 0.5052 0.7402 1 1>, Maximum=<rgba 0.1335 0.2918 0.5698 1>)
     Day Color=&RandomRangeLinearColor(Minimum=<rgba 0.5052 0.7402 1 0.8>, Maximum=<rgba 0.6199 0.7308 0.849 0.5>)
     Evening Color=&RandomRangeLinearColor(Minimum=<rgba 0.5052 0.7402 1 0.5>, Maximum=<rgba 0.2252 0.394 0.6906 0.7>)
    ))
   &AddVelocity(Velocity=&MakeVector(
     X=0
     Y=&Lerp_Float(A=-270, B=-300, Alpha=$System.InterpolateOverTime.InterpolateOverTime_Float.Moving Average)
     Z=0
    )) {module=AddVelocity001}
  ]
  particleUpdateScript [
   &ParticleState()
   &SolveForcesAndVelocity(Clamp Velocity=false, Limit Acceleration=false)
   &ScaleColor(
    Scale Alpha=&FloatFromCurve(
     FloatCurve=DataInterface() {propertyValues {Curve {
        defaultValue=1
        keys [
         {interpMode=RCIM_Cubic, tangentMode=RCTM_User, time=0, value=0}
         {arriveTangent=-0.1462, interpMode=RCIM_Cubic, leaveTangent=-0.1462, tangentMode=RCTM_User, time=0.3, value=1}
         {arriveTangent=-5.4272, interpMode=RCIM_Cubic, leaveTangent=-5.4272, tangentMode=RCTM_User, time=1.0011, value=0}
        ]
       }}}
     Scale Curve=&FNTimeOfDayScalar(Night Scale=1.2, Morning Scale=0.5, Evening Scale=1)
    )
    Scale RGB=&VectorFromFloat(Value=&FNTimeOfDayScalar(Night Scale=2.2, Morning Scale=1.3, Day Scale=2, Evening Scale=1.8))
    ScaleA=true
    ScaleRGB=true
   ) {module=ScaleColor001}
   &FN_UpdateAnimTrail(SkeletalMesh=$System.Skeletal Mesh)
   &ScaleRibbonWidth(Ribbon Width Scale=&FloatFromCurve(FloatCurve=DataInterface() {propertyValues {Curve {keys [
         {arriveTangent=0.464, interpMode=RCIM_Cubic, leaveTangent=0.464, tangentMode=RCTM_User, time=0.0031, value=0.8936}
         {arriveTangent=-2.4767, interpMode=RCIM_Cubic, tangentMode=RCTM_Break, time=0.8, value=0.3359}
        ]}}}))
   &DynamicMaterialParameters(Module.Index0Float0Write=true, Module.Index0Float1Write=true, Module.Index0Float2Write=true, Module.Index0Float3Write=true)
   &BlendSpace("Position Blend (World <-> Local)"=0.5, Rotation Contribution=0, Translation Contribution=0.5) {enabled=false}
  ]
  rendererClasses [&NiagaraRibbonRendererProperties]
  renderers [
   {rendererClass=&NiagaraRibbonRendererProperties}
   {rendererClass=&NiagaraRibbonRendererProperties, rendererIndex=1}
  ]
 }
 Ribbons001 {
  emitterUpdateScript [
   &EmitterState(
    Inactive Response=::NewEnumerator1
    Life Cycle Mode=::NewEnumerator1
    Loop Behavior=::NewEnumerator1
    Loop Delay=1.5
    Loop Duration Mode=::NewEnumerator1
    UseLoopDelay=false
   )
   &SpawnBurst_Instantaneous(Spawn Count=2, Spawn Time=0.5, Use Spawn Probability=false, UseLoopCountLimit=false)
   &SpawnBurst_Instantaneous(Spawn Count=2, Spawn Group=1, Spawn Time=0.5, Use Spawn Probability=false, UseLoopCountLimit=false) {module=SpawnBurst_Instantaneous001}
  ]
  particleSpawnScript [
   /SetVariables(Particles.InitialSpawnGroup:NiagaraInt32=&SpawnGroup()) {isSetParameters=true}
   &FN_SetSpawnGroup()
   /SetVariables(Particles.VisibilityTag=&MergeHelper_IntSwitch(0=1, 1=0, 2=0, 3=1, 4=?, 5=?, Index=$Particles.SpawnGroup, Num=::NewEnumerator2)) {isSetParameters=true}
   &PartitionParticles()
   &InitializeParticle(
    Color=<rgba 0.4322 0.6339 0.849 0.1>
    Lifetime=0.2
    Mesh Scale=&Multiply_VectorByFloat(
     Vector=&MakeVector(
      X=&SelectFloatFromArray(Direct Array Index=$Particles.InitialSpawnGroup, Float Selection Array=DataInterface() {propertyValues {FloatData [1,-1]}})
      Y=1.1
      Z=1
     )
     Float=0.4
    )
    Mesh Scale Mode=::NewEnumerator4
    Position Mode=::NewEnumerator0
    Write Material Random=true
   ) {module=InitializeParticle001}
   &SocketLocation(
    SamplePosOffset=true
    Sampled Position Offset=&MergeHelper_VectorSwitch(0=<15 0 -35>, 1=<-30 0 -35>, 2=?, 3=?, Index=$Particles.PartitionParticles.Partition, Num=::NewEnumerator0)
    Socket Reader=$System.Socket Reader
    Socket Sampling Mode=::NewEnumerator3
    VelocitySpeedLimit=false
   ) {enabled=false}
   &CameraOffset(Camera Offset Amount=5)
   &InitialMeshOrientation(
    Mesh Orientation Mode=::NewEnumerator2
    Rotation=&SelectVectorFromArray(
     Direct Array Index=&SpawnGroup()
     Vector Selection Array=DataInterface() {propertyValues {FloatData [<0 0 0.9319>, <0 0 -0.9154>]}}
    )
    Use Rotation Vector=true
   )
  ]
  particleUpdateScript [
   &SkeletalMeshLocation(
    Bone / Socket Index=3
    Bone Sampling Mode=::NewEnumerator4
    MaskSpawnGroup=false
    Orientation Sampling=::NewEnumerator1
    Sampled Position Offset=&MergeHelper_VectorSwitch(0=<25 -15 -35>, 1=<-30 -15 -35>, 2=?, 3=?, Index=$Particles.PartitionParticles.Partition, Num=::NewEnumerator0)
    Skeletal Mesh=$System.Skeletal Mesh
   )
   &ParticleState(Kill Particles When Lifetime Has Elapsed=false)
   &SolveForcesAndVelocity(Clamp Velocity=false, Limit Acceleration=false)
   &InterpolateOverTime(
    Rate Of Change=10
    Target Vector Value:Vector3f=&TransformPosition(Position=$Particles.Position, Source Space=::Local, Destination Space=::World)
    Variable Type=::NewEnumerator2
   )
   &DynamicMaterialParameters(
    Index 0 Param 1=&RandomRangeFloat(Minimum=0, Maximum=1)
    Index 0 Param 3=&MakeFloatFromVector(Vector=<0 0 0>, Channel=::NewEnumerator0)
    Index 1 Param 1=&MakeFloatFromVector(Vector=$StackContext.InterpolateOverTime.InterpolateOverTime_V3_.Moving Average, Channel=::NewEnumerator0)
    Index 1 Param 2=&MakeFloatFromVector(Vector=$StackContext.InterpolateOverTime.InterpolateOverTime_V3_.Moving Average, Channel=::NewEnumerator1)
    Index 1 Param 3=&MakeFloatFromVector(Vector=$StackContext.InterpolateOverTime.InterpolateOverTime_V3_.Moving Average, Channel=::NewEnumerator2)
    Module.Index0Float0Write=true
    Module.Index0Float1Write=true
    Module.Index0Float2Write=true
    Module.Index0Float3Write=true
    Module.Index1Float0Write=true
    Module.Index1Float1Write=true
    Module.Index1Float2Write=true
    Module.Index1Float3Write=true
    Write Parameter Index 1=::NewEnumerator1
   ) {module=DynamicMaterialParameters001}
   &ScaleColor(
    Scale Alpha=&FloatFromCurve(
     FloatCurve=DataInterface() {propertyValues {Curve {keys [{interpMode=RCIM_Cubic, tangentMode=RCTM_User, time=0, value=0}, {interpMode=RCIM_Cubic, tangentMode=RCTM_User, time=1, value=1}]}}}
     Scale Curve=$User.AnimCurve_wakeOpacity
    )
    Scale RGB=&MakeVectorFromLinearColorRGB(Color=&FNTimeOfDayColor(Night Color=<rgba 0.334 0.556 1 1>, Morning Color=<rgba 1 1 1 1>, Day Color=<rgba 1 1 1 1>, Evening Color=<rgba 1 1 1 1>))
    ScaleA=true
    ScaleRGB=true
   )
  ]
  rendererClasses [&NiagaraMeshRendererProperties]
  renderers [
   {rendererClass=&NiagaraMeshRendererProperties, rendererIndex=0}
   {rendererClass=&NiagaraMeshRendererProperties, rendererIndex=1}
  ]
 }
 boat_ripples {
  emitterUpdateScript [
   &EmitterState(Inactive Response=::NewEnumerator1, Life Cycle Mode=::NewEnumerator1, Loop Duration=1, UseLoopDelay=false)
   &SpawnBurst_Instantaneous(Use Spawn Probability=false, UseLoopCountLimit=false) {enabled=false}
   &SpawnRate(SpawnRate=2, Use Spawn Probability=false)
  ]
  particleSpawnScript [
   &InitializeParticle(
    Color=$System.Ripples color
    Lifetime=&RandomRangeFloat(Minimum=1.5, Maximum=2)
    Mesh Scale=<1.2 0.85 0.05>
    Mesh Scale Mode=::NewEnumerator4
    Position Offset=<0 0 -50>
    UsePositionOffset=false
    Write Material Random=true
   )
   &FN_SetSpawnGroup(Number of Spawn Groups=1)
   &DynamicMaterialParameters(
    Index 0 Param 1=&RandomRangeFloat(Minimum=0, Maximum=1)
    Index 0 Param 3=1
    Module.Index0Param0Write=true
    Module.Index0Param1Write=true
    Module.Index0Param2Write=true
    Module.Index0Param3Write=false
   )
   &AddVelocity(Velocity=<0 -50 0>)
   &CameraOffset(Camera Offset Amount=-5)
  ]
  particleUpdateScript [
   &ParticleState()
   &SkeletalMeshLocation(
    Bone / Socket Index=1
    Bone Sampling Mode=::NewEnumerator4
    MaskSpawnGroup=false
    Sampled Position Offset=&MakeVector(
     X=0
     Y=0
     Z=&Multiply_Float(
      A=&MakeCustomFloatFromBool(
       Bool=&SetBoolByFloatComparison(
        A=&FN_PickFloatDependOnPlatform(
         TargetPlatform=DataInterface() {propertyValues {Platforms {
            qualityLevelMask=-29
            deviceProfileStates [
             {profileName=Windows, qualityLevelMask=0, setQualityLevelMask=2}
             {profileName=Mobile, qualityLevelMask=0, setQualityLevelMask=4294967295}
             {profileName=Mac, qualityLevelMask=0, setQualityLevelMask=2}
             {profileName=Linux, qualityLevelMask=0, setQualityLevelMask=2}
            ]
           }}}
         TargetPlatformActive=1
         TargetPlatformNoActive=0
        )
        B=0
       )
       True Float=-50
       False Float=-51
      )
      B=1
     )
    )
    Skeletal Mesh=$System.Skeletal Mesh
   )
   &ScaleColor(
    Scale Alpha=&FloatFromCurve(FloatCurve=DataInterface() {propertyValues {Curve {keys [
         {interpMode=RCIM_Cubic, tangentMode=RCTM_User, time=0, value=0}
         {interpMode=RCIM_Cubic, tangentMode=RCTM_User, time=0.5, value=1}
         {tangentMode=RCTM_User, time=1, value=0}
        ]}}})
    Scale RGB=&VectorFromFloat(Value=&FNTimeOfDayScalar(Night Scale=1.5, Morning Scale=1.3, Day Scale=1.5, Evening Scale=1.2))
    ScaleA=true
    ScaleRGB=true
   )
   &InterpolateOverTime(
    Rate Of Change=8
    Target Vector Value:Vector3f=&AddVectorToPosition(
     Position=$Particles.Position
     Vector=&RotateVectorByQuaternion(Vector To Rotate=<0 -100 0>, Quaternion=$Particles.SkeletalMeshLocation.SampledRotation)
    )
    Variable Type=::NewEnumerator2
   )
   &ScaleMeshSize(Scale Factor=&VectorFromCurve(
     VectorCurve=DataInterface() {propertyValues {
       XCurve {keys [{arriveTangent=1.898, interpMode=RCIM_Cubic, leaveTangent=1.898, tangentMode=RCTM_User, time=0, value=1}]}
       YCurve {keys [{arriveTangent=0.5312, interpMode=RCIM_Cubic, leaveTangent=0.5312, tangentMode=RCTM_User, time=0, value=1}, <at 1 1.2>]}
       ZCurve {keys [{arriveTangent=9.0725, interpMode=RCIM_Cubic, leaveTangent=9.0725, tangentMode=RCTM_User, time=0, value=1}]}
      }}
     Scale Curve=<1 1 1.5>
    )) {module=ScaleMeshSize001}
   &SolveForcesAndVelocity(Clamp Velocity=false, Limit Acceleration=false)
  ]
  rendererClasses [&NiagaraMeshRendererProperties]
  renderers [{rendererClass=&NiagaraMeshRendererProperties, rendererIndex=0}]
 }
}
summary {
 systemName=NS_Emote_GnatGalaPancake_Water_Traversal
 emitters [
  {bEnabled=false, emitterName=FOG, rendererClasses [&NiagaraSpriteRendererProperties]}
  {bEnabled=false, emitterName=Ground_fog001, rendererClasses [&NiagaraSpriteRendererProperties]}
  {emitterName=boat_ripples, rendererClasses [&NiagaraMeshRendererProperties]}
  {emitterName=Liquid_Trail, rendererClasses [&NiagaraRibbonRendererProperties]}
  {emitterName=Fog_system_Condensed, rendererClasses [&NiagaraSpriteRendererProperties]}
  {emitterName=Ribbons001, rendererClasses [&NiagaraMeshRendererProperties]}
 ]
 userVariables [
  User.AnimCurve_fogIntensity {type=NiagaraFloat, defaultValue=0, description=""}
  User.AnimCurve_wakeOpacity {type=NiagaraFloat, defaultValue=0, description=""}
  User.bIsFrontEnd {type=NiagaraBool}
  User.bIsFrontEndPreview {type=NiagaraBool}
  User.OwnerLocalVelocity {type=Vector3f, description="", defaultValue {struct=Vector3f, value=<0 0 0>}}
  User.RibbonIDOffset {type=NiagaraInt32}
 ]
}
