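# Niagara system for the llama water ripples in the GnatGala Pancake emote, built from two near-identical mesh emitters and two LWE_RIPPLE emitters that list no renderers.
# Timed spawns create short-lived, offset mesh particles that drift with drag, scale randomly along curves, and fade with time-of-day color scaling.
# Dynamic material parameters receive the lag between an offset particle point and its interpolated moving average.
# Skeletal mesh location modules are disabled, though updates still reference their sampled rotation.
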NiagaraSystem /BRCosmetics/Animation/Game/MainPlayer/Emotes/GnatGala_Pancake/FX/NS_Emote_GnatGalaPancake_Water_llama_ripples
dependencies {
 usedDataInterfaces [&NiagaraDataInterfaceCurve]
 usedDynamicInputs [
  &RemapRange
  &Add_Float
  &Abs_Float
  &MakeLinearColorFromVectorAndFloat
  &MakeVectorFromLinearColorRGB
  &FNTimeOfDayColor
  &RandomRangeLinearColor
 ]
 usedModules [
  &self:SystemSpawnScript.SystemScriptSource.SystemScriptGraph.NiagaraNodeAssignment_3.SetVariables_#1
  &SystemState
  &self:SystemSpawnScript.SystemScriptSource.SystemScriptGraph.NiagaraNodeAssignment_2.SetVariables_#2
  &self:SystemSpawnScript.SystemScriptSource.SystemScriptGraph.NiagaraNodeAssignment_4.SetVariables_#3
 ]
}
emitters {
 LWE_RIPPLE {rendererClasses [], renderers []}
 LWE_RIPPLE001 {rendererClasses [], renderers []}
 Llama_ripples {
  emitterUpdateScript [
   &EmitterState(Life Cycle Mode=::NewEnumerator1, Loop Behavior=::NewEnumerator1, Loop Delay=1, Loop Duration=3.5, UseLoopDelay=true)
   &SpawnBurst_Instantaneous(Spawn Time=3, Use Spawn Probability=false, UseLoopCountLimit=false)
   &SpawnRate(Use Spawn Probability=false)
  ]
  particleSpawnScript [
   &InitializeParticle(
    Color=$System.Ripples color
    Lifetime=&RandomRangeFloat(Minimum=1.5, Maximum=2)
    Mesh Scale Mode=::NewEnumerator1
    Mesh Uniform Scale Max=1
    Mesh Uniform Scale Min=0.5
    Position Offset=<35 -25 25>
    UsePositionOffset=true
    Write Material Random=true
   )
   &FN_SetSpawnGroup(Number of Spawn Groups=1)
   &DynamicMaterialParameters(
    Index 0 Param 1=&RandomRangeFloat(Minimum=0, Maximum=1)
    Index 0 Param 3=1
    Module.Index0Param0Write=true
    Module.Index0Param1Write=true
    Module.Index0Param2Write=true
    Module.Index0Param3Write=false
   )
   &SkeletalMeshLocation(
    MaskSpawnGroup=false
    Mesh Sampling Type=::NewEnumerator5
    Orientation Sampling=::NewEnumerator1
    Sampled Position Offset=<-10 0 0>
    Skeletal Mesh=$System.Skeletal Mesh
    Socket Sampling Mode=::NewEnumerator5
   ) {enabled=false}
   &AddVelocity(Velocity=<0 -50 0>)
  ]
  particleUpdateScript [
   &ParticleState()
   &ScaleColor(
    Scale Alpha=&FloatFromCurve(
     FloatCurve=DataInterface() {propertyValues {Curve {
        defaultValue=0
        keys [
         {interpMode=RCIM_Cubic, tangentMode=RCTM_User, time=0, value=0}
         {arriveTangent=-0.0455, interpMode=RCIM_Cubic, leaveTangent=-0.0455, tangentMode=RCTM_User, time=0.1103, value=0.8978}
         {interpMode=RCIM_Cubic, tangentMode=RCTM_User, time=0.5, value=1}
         {tangentMode=RCTM_User, time=0.89, value=0}
        ]
       }}}
     Scale Curve=&FNTimeOfDayScalar(Night Scale=5, Morning Scale=2.5, Day Scale=2.5, Evening Scale=1)
    )
    Scale RGB=&VectorFromFloat(Value=&FNTimeOfDayScalar(Night Scale=1.5, Morning Scale=1.3, Day Scale=1.5, Evening Scale=1.2))
    ScaleA=true
    ScaleRGB=true
   )
   &InterpolateOverTime(
    Rate Of Change=8
    Target Vector Value:Vector3f=&AddVectorToPosition(
     Position=$Particles.Position
     Vector=&RotateVectorByQuaternion(Vector To Rotate=<0 -100 0>, Quaternion=$Particles.SkeletalMeshLocation.SampledRotation)
    )
    Variable Type=::NewEnumerator2
   )
   &DynamicMaterialParameters(
    Index 2 Param 1=&MakeFloatFromVector(
     Vector=&Subtract_Position(
      A=&AddVectorToPosition(
       Position=$Particles.Position
       Vector=&RotateVectorByQuaternion(Vector To Rotate=<0 -100 0>, Quaternion=$Particles.SkeletalMeshLocation.SampledRotation)
      )
      B=Linked() {linkedVariable {name=Particles.InterpolateOverTime.InterpolateOverTime_V3_.Moving Average, type=Vector3f}}
     )
     Channel=::NewEnumerator0
    )
    Index 2 Param 2=&MakeFloatFromVector(
     Vector=&Subtract_Position(
      A=&AddVectorToPosition(
       Position=$Particles.Position
       Vector=&RotateVectorByQuaternion(Vector To Rotate=<0 -100 0>, Quaternion=$Particles.SkeletalMeshLocation.SampledRotation)
      )
      B=Linked() {linkedVariable {name=Particles.InterpolateOverTime.InterpolateOverTime_V3_.Moving Average, type=Vector3f}}
     )
     Channel=::NewEnumerator1
    )
    Index 2 Param 3=&MakeFloatFromVector(
     Vector=&Subtract_Position(
      A=&AddVectorToPosition(
       Position=$Particles.Position
       Vector=&RotateVectorByQuaternion(Vector To Rotate=<0 -100 0>, Quaternion=$Particles.SkeletalMeshLocation.SampledRotation)
      )
      B=Linked() {linkedVariable {name=Particles.InterpolateOverTime.InterpolateOverTime_V3_.Moving Average, type=Vector3f}}
     )
     Channel=::NewEnumerator2
    )
    Module.Param0WriteEnabled2=true
    Module.Param1WriteEnabled2=true
    Module.Param2WriteEnabled2=true
    Module.Param3WriteEnabled2=true
    Write Parameter Index 0=false
    Write Parameter Index 2=true
   ) {module=DynamicMaterialParameters001}
   &ScaleMeshSize(Scale Factor=&VectorFromFloat(Value=&RandomRangeFloat(
      Minimum=&FloatFromCurve(
       FloatCurve=DataInterface() {propertyValues {Curve {defaultValue=1, keys [{tangentMode=RCTM_User, time=0, value=0.5}, {tangentMode=RCTM_User, time=1, value=1}]}}}
       Scale Curve=1.5
      )
      Maximum=&FloatFromCurve(
       FloatCurve=DataInterface() {propertyValues {Curve {defaultValue=1, keys [{tangentMode=RCTM_User, time=0, value=0.5}, {tangentMode=RCTM_User, time=1, value=1}]}}}
       Scale Curve=2
      )
     ))) {module=ScaleMeshSize001}
   &Drag(Drag=2, Use Linear Drag=true, Use Rotational Drag=true)
   &SolveForcesAndVelocity(Clamp Velocity=false, Limit Acceleration=false)
  ]
  rendererClasses [&NiagaraMeshRendererProperties]
  renderers [{rendererClass=&NiagaraMeshRendererProperties, rendererIndex=0}]
 }
 Llama_ripples001 {
  emitterUpdateScript [
   &EmitterState(Life Cycle Mode=::NewEnumerator1, Loop Behavior=::NewEnumerator1, Loop Delay=7, Loop Duration=3, UseLoopDelay=false)
   &SpawnBurst_Instantaneous(Spawn Group=1, Spawn Time=0.02, Use Spawn Probability=false, UseLoopCountLimit=false) {enabled=false, module=SpawnBurst_Instantaneous001}
   &SpawnBurst_Instantaneous(Spawn Time=3, Use Spawn Probability=false, UseLoopCountLimit=false)
   &SpawnRate(Use Spawn Probability=false)
  ]
  particleSpawnScript [
   &InitializeParticle(
    Color=$System.Ripples color
    Lifetime=&RandomRangeFloat(Minimum=1, Maximum=2)
    Mesh Scale Mode=::NewEnumerator1
    Mesh Uniform Scale Max=1
    Mesh Uniform Scale Min=0.5
    Position Offset=<35 0 25>
    UsePositionOffset=true
    Write Material Random=true
   )
   &FN_SetSpawnGroup(Number of Spawn Groups=2)
   &DynamicMaterialParameters(
    Index 0 Param 1=&RandomRangeFloat(Minimum=0, Maximum=1)
    Index 0 Param 3=1
    Module.Index0Param0Write=true
    Module.Index0Param1Write=true
    Module.Index0Param2Write=true
    Module.Index0Param3Write=false
   )
   &SkeletalMeshLocation(
    MaskSpawnGroup=false
    Mesh Sampling Type=::NewEnumerator5
    Orientation Sampling=::NewEnumerator1
    Sampled Position Offset=<-10 0 0>
    Skeletal Mesh=$System.Skeletal Mesh
    Socket Sampling Mode=::NewEnumerator5
   ) {enabled=false}
   &AddVelocity(Velocity=<0 -50 0>)
  ]
  particleUpdateScript [
   &ParticleState()
   &ScaleColor(
    Scale Alpha=&FloatFromCurve(
     FloatCurve=DataInterface() {propertyValues {Curve {
        defaultValue=0
        keys [
         {interpMode=RCIM_Cubic, tangentMode=RCTM_User, time=0, value=0.5}
         {arriveTangent=-0.0455, interpMode=RCIM_Cubic, leaveTangent=-0.0455, tangentMode=RCTM_User, time=0.1103, value=0.8978}
         {interpMode=RCIM_Cubic, tangentMode=RCTM_User, time=0.5, value=1}
         {tangentMode=RCTM_User, time=0.89, value=0}
        ]
       }}}
     Scale Curve=&FNTimeOfDayScalar(Night Scale=5, Morning Scale=2.5, Day Scale=2.5, Evening Scale=1)
    )
    Scale RGB=&VectorFromFloat(Value=&FNTimeOfDayScalar(Night Scale=1.5, Morning Scale=1.3, Day Scale=1.5, Evening Scale=1.2))
    ScaleA=true
    ScaleRGB=true
   )
   &InterpolateOverTime(
    Rate Of Change=8
    Target Vector Value:Vector3f=&AddVectorToPosition(
     Position=$Particles.Position
     Vector=&RotateVectorByQuaternion(Vector To Rotate=<0 -100 0>, Quaternion=$Particles.SkeletalMeshLocation.SampledRotation)
    )
    Variable Type=::NewEnumerator2
   )
   &DynamicMaterialParameters(
    Index 2 Param 1=&MakeFloatFromVector(
     Vector=&Subtract_Position(
      A=&AddVectorToPosition(
       Position=$Particles.Position
       Vector=&RotateVectorByQuaternion(Vector To Rotate=<0 -100 0>, Quaternion=$Particles.SkeletalMeshLocation.SampledRotation)
      )
      B=Linked() {linkedVariable {name=Particles.InterpolateOverTime.InterpolateOverTime_V3_.Moving Average, type=Vector3f}}
     )
     Channel=::NewEnumerator0
    )
    Index 2 Param 2=&MakeFloatFromVector(
     Vector=&Subtract_Position(
      A=&AddVectorToPosition(
       Position=$Particles.Position
       Vector=&RotateVectorByQuaternion(Vector To Rotate=<0 -100 0>, Quaternion=$Particles.SkeletalMeshLocation.SampledRotation)
      )
      B=Linked() {linkedVariable {name=Particles.InterpolateOverTime.InterpolateOverTime_V3_.Moving Average, type=Vector3f}}
     )
     Channel=::NewEnumerator1
    )
    Index 2 Param 3=&MakeFloatFromVector(
     Vector=&Subtract_Position(
      A=&AddVectorToPosition(
       Position=$Particles.Position
       Vector=&RotateVectorByQuaternion(Vector To Rotate=<0 -100 0>, Quaternion=$Particles.SkeletalMeshLocation.SampledRotation)
      )
      B=Linked() {linkedVariable {name=Particles.InterpolateOverTime.InterpolateOverTime_V3_.Moving Average, type=Vector3f}}
     )
     Channel=::NewEnumerator2
    )
    Module.Param0WriteEnabled2=true
    Module.Param1WriteEnabled2=true
    Module.Param2WriteEnabled2=true
    Module.Param3WriteEnabled2=true
    Write Parameter Index 0=false
    Write Parameter Index 2=true
   ) {module=DynamicMaterialParameters001}
   &ScaleMeshSize(Scale Factor=&VectorFromFloat(Value=&RandomRangeFloat(
      Minimum=&FloatFromCurve(
       FloatCurve=DataInterface() {propertyValues {Curve {defaultValue=1, keys [{tangentMode=RCTM_User, time=0, value=0.5}, {tangentMode=RCTM_User, time=1, value=1}]}}}
       Scale Curve=1.5
      )
      Maximum=&FloatFromCurve(
       FloatCurve=DataInterface() {propertyValues {Curve {defaultValue=1, keys [{tangentMode=RCTM_User, time=0, value=0.5}, {tangentMode=RCTM_User, time=1, value=1}]}}}
       Scale Curve=2
      )
     ))) {module=ScaleMeshSize001}
   &Drag(Drag=2, Use Linear Drag=true, Use Rotational Drag=true)
   &SolveForcesAndVelocity(Clamp Velocity=false, Limit Acceleration=false)
  ]
  rendererClasses [&NiagaraMeshRendererProperties]
  renderers [
   {rendererClass=&NiagaraMeshRendererProperties, rendererIndex=0}
   {rendererClass=&NiagaraMeshRendererProperties, rendererIndex=1}
  ]
 }
}
summary {
 systemName=NS_Emote_GnatGalaPancake_Water_llama_ripples
 emitters [
  {emitterName=Llama_ripples001, rendererClasses [&NiagaraMeshRendererProperties]}
  {emitterName=Llama_ripples, rendererClasses [&NiagaraMeshRendererProperties]}
  {emitterName=LWE_RIPPLE, rendererClasses []}
  {emitterName=LWE_RIPPLE001, rendererClasses []}
 ]
 userVariables [
  User.bIsFrontEnd {type=NiagaraBool}
  User.bIsFrontEndPreview {type=NiagaraBool}
  User.CylinderLocationWidth {type=NiagaraFloat, defaultValue=0, description=""}
  User.OwnerLocalVelocity {type=Vector3f, description="", defaultValue {struct=Vector3f, value=<0 0 0>}}
  User.RibbonIDOffset {type=NiagaraInt32}
  User.StopHeight {type=NiagaraFloat, defaultValue=0, description=""}
 ]
}
