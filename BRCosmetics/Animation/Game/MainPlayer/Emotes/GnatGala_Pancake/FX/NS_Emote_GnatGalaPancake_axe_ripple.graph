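# Niagara effect for the axe ripple in the GnatGala Pancake emote, built from one mesh emitter, sword_ripples, with two mesh renderers.
# Spawn rate switches on an owner-speed threshold in gameplay, with a separate front-end rate.
# Short-lived, randomly sized meshes spawn slightly raised, take a system-level ripple color, then grow while alpha rises and fades out.
# Time-of-day scalars drive brightness and opacity, boosting alpha most during the day.
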NiagaraSystem /BRCosmetics/Animation/Game/MainPlayer/Emotes/GnatGala_Pancake/FX/NS_Emote_GnatGalaPancake_axe_ripple
dependencies {
 usedDataInterfaces [&NiagaraDataInterfaceCurve]
 usedDynamicInputs [&FNTimeOfDayColor, &RandomRangeLinearColor]
 usedModules [
  &self:SystemSpawnScript.SystemScriptSource.SystemScriptGraph.NiagaraNodeAssignment_3.SetVariables_#1
  &SystemState
 ]
}
emitters {sword_ripples {
  emitterUpdateScript [
   &EmitterState(Life Cycle Mode=::NewEnumerator1, Loop Behavior=::NewEnumerator1, Loop Duration=0.6, UseLoopDelay=true)
   &SpawnRate(
    SpawnRate=&FN_FrontEndFloatModifier(
     Gameplay=&MakeCustomFloatFromBool(
      Bool=&SetBoolByFloatComparison(A=&VectorLength(Vector=$User.OwnerLocalVelocity), B=90)
      True Float=4
      False Float=3
     )
     Front End=5
    )
    Use Spawn Probability=false
   )
  ]
  particleSpawnScript [
   &InitializeParticle(
    Color=$System.Ripple color
    Lifetime=1.3
    Mesh Scale Mode=::NewEnumerator1
    Mesh Uniform Scale Max=1
    Mesh Uniform Scale Min=0.5
    Position Offset=<0 0 18>
    UsePositionOffset=true
    Write Material Random=true
   )
   &FN_SetSpawnGroup(Number of Spawn Groups=1)
   &AddVelocity(Velocity=<0 0 0>)
  ]
  particleUpdateScript [
   &ParticleState()
   &ScaleMeshSize(Scale Factor=&VectorFromFloat(Value=&FloatFromCurve(
      FloatCurve=DataInterface() {propertyValues {Curve {
         defaultValue=1
         keys [{interpMode=RCIM_Cubic, tangentMode=RCTM_User, time=0, value=0.5}, {interpMode=RCIM_Cubic, tangentMode=RCTM_User, time=1, value=1}]
        }}}
      Scale Curve=1.5
     )))
   &ScaleColor(
    Scale Alpha=&FloatFromCurve(
     FloatCurve=DataInterface() {propertyValues {Curve {
        defaultValue=0
        keys [
         {interpMode=RCIM_Cubic, tangentMode=RCTM_User, time=0, value=0.5}
         {arriveTangent=-0.0455, interpMode=RCIM_Cubic, leaveTangent=-0.0455, tangentMode=RCTM_User, time=0.1103, value=1}
         {interpMode=RCIM_Cubic, tangentMode=RCTM_User, time=0.3, value=1}
         {tangentMode=RCTM_User, time=0.7, value=0}
        ]
       }}}
     Scale Curve=&FNTimeOfDayScalar(Night Scale=2, Morning Scale=2, Day Scale=4, Evening Scale=2)
    )
    Scale RGB=&VectorFromFloat(Value=&FNTimeOfDayScalar(Night Scale=1.5, Morning Scale=1.5, Day Scale=1.5, Evening Scale=1.5))
    ScaleA=true
    ScaleRGB=true
   )
   &SolveForcesAndVelocity(Clamp Velocity=false, Limit Acceleration=false)
   /SetVariables() {isSetParameters=true}
  ]
  rendererClasses [&NiagaraMeshRendererProperties]
  renderers [
   {rendererClass=&NiagaraMeshRendererProperties, rendererIndex=0}
   {rendererClass=&NiagaraMeshRendererProperties, rendererIndex=1}
  ]
 }}
summary {
 systemName=NS_Emote_GnatGalaPancake_axe_ripple
 emitters [{emitterName=sword_ripples, rendererClasses [&NiagaraMeshRendererProperties]}]
 userVariables [
  User.bIsFrontEnd {type=NiagaraBool}
  User.bIsFrontEndPreview {type=NiagaraBool}
  User.OwnerLocalVelocity {type=Vector3f, description="", defaultValue {struct=Vector3f, value=<0 0 0>}}
 ]
}
