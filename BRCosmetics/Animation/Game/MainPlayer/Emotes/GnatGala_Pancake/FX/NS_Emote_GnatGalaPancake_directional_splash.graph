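# Niagara system for a directional water splash in the GnatGala Pancake player emote, tinted by a system-level ripple color.
# Moving_water bursts short-lived meshes just above the origin, pushes them along negative Y and randomizes material parameters.
# Curves scale the meshes and fade alpha over lifetime, with time-of-day scalars adjusting brightness and opacity.
# A second emitter, LWE_RIPPLE, is listed without renderers.
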NiagaraSystem /BRCosmetics/Animation/Game/MainPlayer/Emotes/GnatGala_Pancake/FX/NS_Emote_GnatGalaPancake_directional_splash
dependencies {
 usedDataInterfaces [&NiagaraDataInterfaceCurve]
 usedDynamicInputs [&FNTimeOfDayColor, &RandomRangeLinearColor]
 usedModules [
  &self:SystemSpawnScript.SystemScriptSource.SystemScriptGraph.NiagaraNodeAssignment_2.SetVariables_#1
  &SystemState
 ]
}
emitters {
 LWE_RIPPLE {rendererClasses [], renderers []}
 Moving_water {
  emitterUpdateScript [
   &EmitterState()
   &SpawnBurst_Instantaneous(Spawn Count=2, Spawn Time=0.02, Use Spawn Probability=false, UseLoopCountLimit=false)
   &SpawnRate(SpawnRate=2, Use Spawn Probability=false) {enabled=false}
  ]
  particleSpawnScript [
   &InitializeParticle(
    Color=$System.Ripple color
    Lifetime Max=0.85
    Lifetime Min=0.65
    Lifetime Mode=::NewEnumerator1
    Mesh Scale Mode=::NewEnumerator1
    Mesh Uniform Scale Max=1
    Mesh Uniform Scale Min=0.5
    Position Offset=<0 0 15>
    UsePositionOffset=true
    Write Material Random=true
   )
   &FN_SetSpawnGroup(Number of Spawn Groups=1)
   &DynamicMaterialParameters(
    Index 0 Param 1=&RandomRangeFloat(Minimum=0, Maximum=1)
    Index 0 Param 3=1
    Module.Index0Param0Write=true
    Module.Index0Param1Write=true
    Module.Index0Param2Write=true
    Module.Index0Param3Write=false
   )
   &AddVelocity(Velocity=<0 -50 0>)
  ]
  particleUpdateScript [
   &ParticleState()
   &ScaleMeshSize(Scale Factor=&VectorFromFloat(Value=&RandomRangeFloat(
      Minimum=&FloatFromCurve(FloatCurve=DataInterface() {propertyValues {Curve {defaultValue=1, keys [{tangentMode=RCTM_User, time=0, value=0}, {tangentMode=RCTM_User, time=1, value=1}]}}})
      Maximum=&FloatFromCurve(
       FloatCurve=DataInterface() {propertyValues {Curve {defaultValue=1, keys [{tangentMode=RCTM_User, time=0, value=0}, {tangentMode=RCTM_User, time=1.5, value=1}]}}}
       Scale Curve=2.5
      )
     ))) {module=ScaleMeshSize001}
   &ScaleColor(
    Scale Alpha=&FloatFromCurve(
     FloatCurve=DataInterface() {propertyValues {Curve {
        defaultValue=0
        keys [
         {interpMode=RCIM_Cubic, tangentMode=RCTM_User, time=0, value=0}
         {arriveTangent=-0.0455, interpMode=RCIM_Cubic, leaveTangent=-0.0455, tangentMode=RCTM_User, time=0.1103, value=0.8978}
         {interpMode=RCIM_Cubic, tangentMode=RCTM_User, time=0.5, value=1}
         {tangentMode=RCTM_User, time=0.89, value=0}
        ]
       }}}
     Scale Curve=&FNTimeOfDayScalar(Night Scale=3, Morning Scale=2.5, Day Scale=5, Evening Scale=1)
    )
    Scale RGB=&VectorFromFloat(Value=&FNTimeOfDayScalar(Night Scale=1.5, Morning Scale=1.3, Day Scale=1.5, Evening Scale=1.2))
    ScaleA=true
    ScaleRGB=true
   ) {module=ScaleColor002}
   &SolveForcesAndVelocity(Clamp Velocity=false, Limit Acceleration=false)
  ]
  rendererClasses [&NiagaraMeshRendererProperties]
  renderers [{rendererClass=&NiagaraMeshRendererProperties, rendererIndex=0}]
 }
}
summary {
 systemName=NS_Emote_GnatGalaPancake_directional_splash
 emitters [{emitterName=Moving_water, rendererClasses [&NiagaraMeshRendererProperties]}, {emitterName=LWE_RIPPLE, rendererClasses []}]
 userVariables []
}
