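# Niagara water splash effect for the GnatGala Pancake player emote.
# In the visible part, Moving_water is the main active emitter: a burst mesh that drifts along -Y, grows and fades, tinted by the system Ripple color.
# Brightness and opacity follow time-of-day scalars.
# Many splash layers are present but disabled, including drips, a mesh splash, Water_Plasma001 and several SubUV sprite bursts sharing a Water Splashes color.
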NiagaraSystem /BRCosmetics/Animation/Game/MainPlayer/Emotes/GnatGala_Pancake/FX/NS_Emote_GnatGalaPancake_splash
dependencies {
 usedDataInterfaces [
  &NiagaraDataInterfaceSpriteRendererInfo
  &NiagaraDataInterfaceCurve
  &NiagaraDataInterfaceVector2DCurve
  &NiagaraDataInterfaceVectorCurve
 ]
 usedDynamicInputs [
  &Lerp_Float
  &FNTimeOfDayColor
  &RandomRangeLinearColor
  &MakeVectorFromLinearColorRGB
 ]
 usedModules [
  &self:SystemSpawnScript.SystemScriptSource.SystemScriptGraph.NiagaraNodeAssignment_3.SetVariables_#1
  &self:SystemSpawnScript.SystemScriptSource.SystemScriptGraph.NiagaraNodeAssignment_0.SetVariables_#2
  &SystemState
 ]
}
emitters {
 BuffSplash_Mesh {
  bEnabled=false
  emitterUpdateScript [
   &EmitterState(
    Enable Distance Culling=true
    Life Cycle Mode=::NewEnumerator1
    Loop Behavior=::NewEnumerator1
    Loop Delay=0.2
    Loop Duration=1
    MaxDistance=25000
    Module.bApplySpawnCountScaleByDistance=false
    Module.bUseMaxDistance=true
    Module.bUseMinDistance=false
    Scalability Mode=::NewEnumerator1
    Scale Spawn Count=true
    UseLoopDelay=false
   )
   &SpawnBurst_Instantaneous(Spawn Count=2, Use Spawn Probability=false, UseLoopCountLimit=false)
   &SpawnRate(SpawnRate=10, Use Spawn Probability=false) {enabled=false}
  ]
  particleSpawnScript [
   &InitializeParticle(
    Color=<rgba 0.4322 0.6339 0.849 1>
    Lifetime=0.4
    Mesh Scale Max=<1 1 2>
    Mesh Scale Min=<0.7 0.7 0.5>
    Mesh Scale Mode=::NewEnumerator2
    UsePositionOffset=true
    Write Material Random=true
   )
   &InitialMeshOrientation(
    Mesh Orientation Mode=::NewEnumerator2
    Rotation=&RandomRangeVector(Minimum=<0 0 -1>, Maximum=<0 0 1>)
    Use Rotation Vector=true
   )
   &CameraOffset(Camera Offset Amount=5)
   &ShapeLocation(MaskSpawnGroup=false, Override Seed=false, Sphere Radius=1, UseOwnerScale=true)
  ]
  particleUpdateScript [
   &ParticleState()
   &SolveForcesAndVelocity(Clamp Velocity=false, Limit Acceleration=false)
   &ScaleColor(
    Scale Alpha=&FloatFromCurve(FloatCurve=DataInterface() {propertyValues {Curve {keys [
         {arriveTangent=-0.6531, interpMode=RCIM_Cubic, time=0, value=1}
         {arriveTangent=-0.8638, interpMode=RCIM_Cubic, leaveTangent=-0.8638, tangentMode=RCTM_User, time=0.8, value=0.3}
         {arriveTangent=-0.4827, interpMode=RCIM_Cubic, leaveTangent=-0.4827, tangentMode=RCTM_User, time=1, value=0.1}
        ]}}})
    ScaleA=true
    ScaleRGB=false
   )
   &ScaleMeshSize(Scale Factor=&VectorFromCurve(
     VectorCurve=DataInterface() {propertyValues {
       XCurve {keys [
         {arriveTangent=1.8221, interpMode=RCIM_Cubic, leaveTangent=1.8221, tangentMode=RCTM_User, time=0, value=0}
         {arriveTangent=0.4715, interpMode=RCIM_Cubic, leaveTangent=0.4715, tangentMode=RCTM_User, time=1, value=1}
        ]}
       YCurve {keys [
         {arriveTangent=1.8221, interpMode=RCIM_Cubic, leaveTangent=1.8221, tangentMode=RCTM_User, time=0, value=0}
         {arriveTangent=0.4715, interpMode=RCIM_Cubic, leaveTangent=0.4715, tangentMode=RCTM_User, time=1, value=1}
        ]}
       ZCurve {keys [
         {arriveTangent=4.7636, interpMode=RCIM_Cubic, leaveTangent=4.7636, tangentMode=RCTM_User, time=0, value=0}
         {arriveTangentWeight=0.2848, interpMode=RCIM_Cubic, leaveTangentWeight=0.1221, tangentMode=RCTM_User, time=0.5, value=1}
         {arriveTangent=-5.505, interpMode=RCIM_Cubic, leaveTangent=-5.505, tangentMode=RCTM_User, time=1, value=0}
        ]}
      }}
     Scale Curve=&VectorFromFloat(Value=0.4)
    ))
   &DynamicMaterialParameters(
    Index 1 Param 2=&FloatFromCurve(
     FloatCurve=DataInterface() {propertyValues {Curve {keys [<at 0 -1.5 RCIM_Cubic>, {arriveTangent=1.2375, interpMode=RCIM_Cubic, leaveTangent=1.2375, tangentMode=RCTM_User, time=1, value=1}]}}}
     Scale Curve=0
    )
    Module.Param0WriteEnabled1=false
    Module.Param1WriteEnabled1=true
    Module.Param2WriteEnabled1=false
    Module.Param3WriteEnabled1=false
    Write Parameter Index 0=false
    Write Parameter Index 1=true
   )
   /SetVariables(Particles.MeshOrientation=&MultiplyQuaternion(
     Quaternion A=&MakeQuatFromAxisAngle(
      Axis=$Particles.Initial.ShapeLocation.ShapeBitangent
      AngleInDegrees=&AngleConversion(
       Angle:NiagaraFloat=&FloatFromCurve(
        FloatCurve=DataInterface() {propertyValues {Curve {keys [{interpMode=RCIM_Cubic, tangentMode=RCTM_User, time=0, value=0}, {interpMode=RCIM_Cubic, tangentMode=RCTM_User, time=1, value=1}]}}}
        Scale Curve=&RandomRangeFloat(Minimum=-0.1, Maximum=-0.225)
       )
       Angle Input:ENiagara_AngleInput=::NewEnumerator1
       Angle Output:ENiagara_AngleInput=::NewEnumerator0
      )
      NormalizeAxis=true
     )
     Quaternion B=$Particles.Initial.MeshOrientation
    )) {isSetParameters=true}
  ]
  rendererClasses [&NiagaraMeshRendererProperties]
  renderers [{rendererClass=&NiagaraMeshRendererProperties, rendererIndex=0}]
 }
 Drip_Drop {
  bEnabled=false
  emitterUpdateScript [
   &EmitterState(Life Cycle Mode=::NewEnumerator1, Loop Behavior=::NewEnumerator1, Loop Duration=1, UseLoopDelay=false)
   &SpawnBurst_Instantaneous(Spawn Count=15, Use Spawn Probability=false, UseLoopCountLimit=false)
   &SpawnRate(SpawnRate=35, Use Spawn Probability=false) {enabled=false}
  ]
  particleSpawnScript [
   &CameraOffset(Camera Offset Amount=3)
   &InitializeParticle(
    Color=<rgba 1 1 1 0.05>
    Lifetime Max=0.6
    Lifetime Min=0.2
    Lifetime Mode=::NewEnumerator1
    Sprite Size Max=<1.5 7>
    Sprite Size Min=<0.5 4>
    Sprite Size Mode=::NewEnumerator2
    UsePositionOffset=false
    Write Material Random=true
   ) {module=InitializeParticle001}
   &AddVelocity(
    Cone Angle=&RandomRangeFloat(Minimum=90, Maximum=0)
    Cone Axis=<0 0 5>
    Distribution Along Cone Axis=0.25
    Inner Cone Angle=90
    Module.UseConeFalloff=false
    Module.UseVelDistribution=false
    Override Seed=false
    Speed Falloff From Cone Axis=0.5
    Velocity Mode=::NewEnumerator2
    Velocity Speed=&RandomRangeFloat(Minimum=100, Maximum=300)
   ) {module=AddVelocity001}
  ]
  particleUpdateScript [
   &ParticleState()
   &GravityForce(Gravity=<0 0 -600>)
   &SolveForcesAndVelocity(Clamp Velocity=false, Limit Acceleration=false)
   &ScaleSpriteSize(Scale Factor=&Vector2DFromCurve(
     Vector2Curve=DataInterface() {propertyValues {
       XCurve {keys [
         {arriveTangent=5.9728, interpMode=RCIM_Cubic, leaveTangent=5.9728, tangentMode=RCTM_User, time=0, value=0.1}
         {arriveTangent=-0.1, interpMode=RCIM_Cubic, leaveTangent=-0.1, tangentMode=RCTM_User, time=0.18, value=0.99}
         {interpMode=RCIM_Cubic, tangentMode=RCTM_User, time=1, value=0}
        ]}
       YCurve {keys [
         {arriveTangent=6.2715, interpMode=RCIM_Cubic, leaveTangent=6.2715, tangentMode=RCTM_User, time=0, value=0.1}
         {arriveTangent=-0.1001, interpMode=RCIM_Cubic, leaveTangent=-0.1001, tangentMode=RCTM_User, time=0.18, value=0.99}
         {interpMode=RCIM_Cubic, tangentMode=RCTM_User, time=0.9989, value=0}
        ]}
      }}
     ScaleCurve=<2 2>
    ))
   &ScaleColor(
    Scale Alpha=&FloatFromCurve(FloatCurve=DataInterface() {propertyValues {Curve {
        defaultValue=1
        keys [{interpMode=RCIM_Cubic, tangentMode=RCTM_User, time=0, value=1}, {interpMode=RCIM_Cubic, tangentMode=RCTM_User, time=1, value=0}]
       }}})
    Scale RGB=&VectorFromFloat(Value=&FNTimeOfDayScalar(Night Scale=0.5, Morning Scale=1, Evening Scale=1))
    ScaleA=true
    ScaleRGB=true
   )
  ]
  rendererClasses [&NiagaraSpriteRendererProperties]
  renderers [{rendererClass=&NiagaraSpriteRendererProperties}]
 }
 Droplets {bEnabled=false, rendererClasses [], renderers []}
 Droplets001 {rendererClasses [], renderers []}
 LWE_RIPPLE {rendererClasses [], renderers []}
 LWE_RIPPLE001 {rendererClasses [], renderers []}
 Moving_water {
  emitterUpdateScript [
   &EmitterState(Life Cycle Mode=::NewEnumerator1, Loop Behavior=::NewEnumerator1, Loop Delay=0.5, Loop Duration=1.5, UseLoopDelay=false)
   &SpawnBurst_Instantaneous(Spawn Time=0.05, Use Spawn Probability=false, UseLoopCountLimit=false)
  ]
  particleSpawnScript [
   &InitializeParticle(
    Color=$System.Ripple color
    Lifetime=2
    Mesh Scale Mode=::NewEnumerator1
    Mesh Uniform Scale Max=1
    Mesh Uniform Scale Min=0.8
    Position Offset=<0 10 15>
    UsePositionOffset=true
    Write Material Random=true
   )
   &DynamicMaterialParameters(
    Index 0 Param 1=&RandomRangeFloat(Minimum=0, Maximum=1)
    Index 0 Param 3=1
    Module.Index0Param0Write=true
    Module.Index0Param1Write=true
    Module.Index0Param2Write=true
    Module.Index0Param3Write=false
   )
   &AddVelocity(Velocity=<0 -58 0>)
  ]
  particleUpdateScript [
   &ParticleState()
   &ScaleMeshSize(Scale Factor=&VectorFromFloat(Value=&FloatFromCurve(FloatCurve=DataInterface() {propertyValues {Curve {defaultValue=1, keys [{tangentMode=RCTM_User, time=0, value=0.5}, {tangentMode=RCTM_User, time=1, value=1}]}}}))) {module=ScaleMeshSize001}
   &ScaleColor(
    Scale Alpha=&FloatFromCurve(
     FloatCurve=DataInterface() {propertyValues {Curve {
        defaultValue=0
        keys [
         {interpMode=RCIM_Cubic, tangentMode=RCTM_User, time=0, value=0}
         {arriveTangent=-0.0455, interpMode=RCIM_Cubic, leaveTangent=-0.0455, tangentMode=RCTM_User, time=0.1103, value=0.8978}
         {interpMode=RCIM_Cubic, tangentMode=RCTM_User, time=0.5, value=1}
         {tangentMode=RCTM_User, time=0.89, value=0}
        ]
       }}}
     Scale Curve=&FNTimeOfDayScalar(Night Scale=3, Morning Scale=2.5, Day Scale=5, Evening Scale=1)
    )
    Scale RGB=&VectorFromFloat(Value=&FNTimeOfDayScalar(Night Scale=1.5, Morning Scale=1.3, Day Scale=1.5, Evening Scale=1.2))
    ScaleA=true
    ScaleRGB=true
   ) {module=ScaleColor002}
   &SolveForcesAndVelocity(Clamp Velocity=false, Limit Acceleration=false)
  ]
  rendererClasses [&NiagaraMeshRendererProperties]
  renderers [{rendererClass=&NiagaraMeshRendererProperties, rendererIndex=0}]
 }
 Splash_Both {
  bEnabled=false
  emitterUpdateScript [
   &EmitterState(Life Cycle Mode=::NewEnumerator1, Loop Behavior=::NewEnumerator1, Loop Delay=0.5, Loop Duration=1.5, UseLoopDelay=false)
   &SpawnBurst_Instantaneous(Spawn Count=12, Spawn Group=1, Use Spawn Probability=false, UseLoopCountLimit=false)
  ]
  particleSpawnScript [
   &InitializeParticle(
    Color=$System.Water Splashes color
    Lifetime Max=0.6
    Lifetime Min=0.35
    Lifetime Mode=::NewEnumerator1
    Sprite Rotation Angle=&RandomRangeFloat(Minimum=0, Maximum=0.2)
    Sprite Rotation Mode=::NewEnumerator1
    Sprite Size Mode=::NewEnumerator1
    Sprite UV Mode=::NewEnumerator3
    Uniform Sprite Size Max=30
    Uniform Sprite Size Min=18
    UsePositionOffset=false
    Write Material Random=true
   )
   &CameraOffset(Camera Offset Amount=3)
   &ShapeLocation(MaskSpawnGroup=false, Override Seed=false, Sphere Radius=5, UseOwnerScale=true)
   &SubUVAnimation(
    End Frame Range Override=3
    Evaluation Type=::NewEnumerator0
    "Sprite Renderer "=DataInterface() {propertyValues {SpriteRenderer=&self:Splash_Both_0.NiagaraSpriteRendererProperties_1}}
    SubUV Animation Mode:ENiagara_SubUVLookupModeV2=::NewEnumerator2
    UseEndFrame=true
    UseStartFrame=true
   )
   &AddVelocity(
    Module.Use Falloff Distance=false
    Origin Offset=<0 0 -7>
    Override Seed=false
    Velocity Mode=::NewEnumerator1
    Velocity Speed=&RandomRangeFloat(Minimum=75, Maximum=200)
   ) {module=AddVelocity001}
  ]
  particleUpdateScript [
   &ParticleState()
   &SpriteRotationRate(Rotation Rate=&MaskFloat(
     Mask Source=0
     Mask Parameter=?
     Mask=0
     Masked Float=&RandomRangeFloat(Minimum=-70, Maximum=-100)
     Passthrough Float=&MaskFloat(
      Mask Source=0
      Mask Parameter=?
      Mask=1
      Masked Float=&RandomRangeFloat(Minimum=70, Maximum=100)
      Passthrough Float=1
     )
    ))
   &GravityForce(Gravity=&RandomRangeVector(Minimum=<0 0 -500>, Maximum=<0 0 -980>))
   &Drag(Drag=&RandomRangeFloat(Minimum=0.2, Maximum=0.3), Use Linear Drag=true, Use Rotational Drag=false)
   &SolveForcesAndVelocity(Clamp Velocity=false, Limit Acceleration=false)
   &ScaleSpriteSize(Uniform Curve Sprite Scale=DataInterface() {propertyValues {Curve {keys [
        {arriveTangent=4.1785, interpMode=RCIM_Cubic, leaveTangent=4.1785, tangentMode=RCTM_User, time=0, value=0.15}
        {arriveTangent=1.1179, interpMode=RCIM_Cubic, leaveTangent=1.1179, tangentMode=RCTM_User, time=0.2042, value=0.6228}
        {arriveTangent=0.0224, interpMode=RCIM_Cubic, leaveTangent=0.0224, tangentMode=RCTM_User, time=1, value=1}
       ]}}})
   &ScaleSpriteSizeBySpeed(
    Sample Scale Factor By Curve=true
    Scale Factor Curve=DataInterface() {propertyValues {Curve {defaultValue=0.5941, keys [{tangentMode=RCTM_User, time=0, value=1}, {tangentMode=RCTM_User, time=1, value=0}]}}}
    Velocity Threshold=250
   )
   &ScaleColor(
    Scale Alpha=&FloatFromCurve(
     FloatCurve=DataInterface() {propertyValues {Curve {keys [
         {arriveTangent=-0.2723, interpMode=RCIM_Cubic, leaveTangent=14.5791, tangentMode=RCTM_Break, time=0, value=0}
         {interpMode=RCIM_Cubic, tangentMode=RCTM_Break, time=0.1, value=1}
         {arriveTangent=0.0378, interpMode=RCIM_Cubic, leaveTangent=0.0378, tangentMode=RCTM_User, time=0.8, value=1}
         {arriveTangent=-3.7547, interpMode=RCIM_Cubic, leaveTangent=-3.7547, tangentMode=RCTM_Break, time=1, value=0}
        ]}}}
     Scale Curve=&FNTimeOfDayScalar(Night Scale=0.12, Morning Scale=0.12, Day Scale=0.12, Evening Scale=0.12)
    )
    Scale RGB=&VectorFromFloat(Value=&FNTimeOfDayScalar(Night Scale=2, Morning Scale=2, Day Scale=2, Evening Scale=2))
    ScaleA=true
    ScaleRGB=true
   ) {enabled=false}
   &DynamicMaterialParameters(
    Index 0 Param 1=&FloatFromCurve(FloatCurve=DataInterface() {propertyValues {Curve {keys [{interpMode=RCIM_Cubic, tangentMode=RCTM_Break, time=0.6, value=1}, {interpMode=RCIM_Cubic, tangentMode=RCTM_Break, time=0.7, value=0.4}]}}})
    Module.Index0Float0Write=true
    Module.Index0Float1Write=false
    Module.Index0Float2Write=false
    Module.Index0Float3Write=false
   )
   &ScaleColor(
    Scale Alpha=&FloatFromCurve(
     FloatCurve=DataInterface() {propertyValues {Curve {keys [
         {arriveTangent=-0.2723, interpMode=RCIM_Cubic, leaveTangent=14.5791, tangentMode=RCTM_Break, time=0, value=0}
         {interpMode=RCIM_Cubic, tangentMode=RCTM_Break, time=0.1, value=1}
         {arriveTangent=0.0378, interpMode=RCIM_Cubic, leaveTangent=0.0378, tangentMode=RCTM_User, time=0.2648, value=1.0186}
         {arriveTangent=0.0889, interpMode=RCIM_Cubic, leaveTangent=-3.7547, tangentMode=RCTM_Break, time=0.995, value=0}
        ]}}}
     Scale Curve=&FNTimeOfDayScalar(Night Scale=0.3, Morning Scale=0.7, Day Scale=0.7, Evening Scale=1)
    )
    Scale RGB=&VectorFromFloat(Value=&FNTimeOfDayScalar(Night Scale=1.5, Morning Scale=2, Day Scale=1.5, Evening Scale=2))
    ScaleA=true
    ScaleRGB=true
   ) {module=ScaleColor001}
  ]
  rendererClasses [&NiagaraSpriteRendererProperties]
  renderers [{rendererClass=&NiagaraSpriteRendererProperties}]
 }
 Splash_Both001 {
  bEnabled=false
  emitterUpdateScript [
   &EmitterState(Life Cycle Mode=::NewEnumerator1, Loop Behavior=::NewEnumerator1, Loop Delay=0.5, Loop Duration=2.3, UseLoopDelay=false)
   &SpawnRate(SpawnRate=100, Use Spawn Probability=false) {enabled=false}
   &SpawnBurst_Instantaneous(Spawn Count=20, Spawn Group=1, Use Spawn Probability=false, UseLoopCountLimit=false)
  ]
  particleSpawnScript [
   &InitializeParticle(
    Color=$System.Water Splashes color
    Lifetime Max=0.8
    Lifetime Min=0.2
    Lifetime Mode=::NewEnumerator1
    Position Offset=<0 0 10>
    Sprite Rotation Angle=&RandomRangeFloat(Minimum=0, Maximum=0.2)
    Sprite Rotation Mode=::NewEnumerator1
    Sprite Size Mode=::NewEnumerator1
    Sprite UV Mode=::NewEnumerator3
    Uniform Sprite Size Max=30
    Uniform Sprite Size Min=18
    UsePositionOffset=false
    Write Material Random=true
   )
   &CameraOffset(Camera Offset Amount=3)
   &ShapeLocation(
    MaskSpawnGroup=false
    Offset=<0 0 -3>
    Offset Mode=::NewEnumerator0
    Override Seed=false
    Ring Radius=10
    Shape Primitive=::NewEnumerator4
    UseOwnerScale=true
   )
   &SubUVAnimation(
    End Frame Range Override=3
    Evaluation Type=::NewEnumerator0
    "Sprite Renderer "=DataInterface() {propertyValues {SpriteRenderer=&self:Splash_Both001_0.NiagaraSpriteRendererProperties_1}}
    SubUV Animation Mode:ENiagara_SubUVLookupModeV2=::NewEnumerator2
    UseEndFrame=true
    UseStartFrame=true
   )
   &AddVelocity(
    Module.Use Falloff Distance=false
    Origin Offset=<0 0 -17.5001>
    Origin Offset Coordinate Space=::Simulation
    Override Seed=false
    Velocity Mode=::NewEnumerator1
    Velocity Speed=&RandomRangeFloat(Minimum=75, Maximum=200)
   ) {module=AddVelocity001}
  ]
  particleUpdateScript [
   &ParticleState()
   &SpriteRotationRate(Rotation Rate=&MaskFloat(
     Mask Source=0
     Mask Parameter=?
     Mask=0
     Masked Float=&RandomRangeFloat(Minimum=-70, Maximum=-100)
     Passthrough Float=&MaskFloat(
      Mask Source=0
      Mask Parameter=?
      Mask=1
      Masked Float=&RandomRangeFloat(Minimum=70, Maximum=100)
      Passthrough Float=1
     )
    ))
   &GravityForce(Gravity=&RandomRangeVector(Minimum=<0 0 -200>, Maximum=<0 0 -400>))
   &Drag(Drag=&RandomRangeFloat(Minimum=0.2, Maximum=0.3), Use Linear Drag=true, Use Rotational Drag=false)
   &SolveForcesAndVelocity(Clamp Velocity=false, Limit Acceleration=false)
   &ScaleSpriteSize(Uniform Curve Sprite Scale=DataInterface() {propertyValues {Curve {keys [
        {arriveTangent=4.1785, interpMode=RCIM_Cubic, leaveTangent=4.1785, tangentMode=RCTM_User, time=0, value=0.15}
        {arriveTangent=1.1179, interpMode=RCIM_Cubic, leaveTangent=1.1179, tangentMode=RCTM_User, time=0.2042, value=0.6228}
        {arriveTangent=0.0224, interpMode=RCIM_Cubic, leaveTangent=0.0224, tangentMode=RCTM_User, time=1, value=1}
       ]}}})
   &ScaleSpriteSizeBySpeed(
    Sample Scale Factor By Curve=true
    Scale Factor Curve=DataInterface() {propertyValues {Curve {defaultValue=0.5941, keys [{tangentMode=RCTM_User, time=0, value=1}, {tangentMode=RCTM_User, time=1, value=0}]}}}
    Velocity Threshold=250
   )
   &ScaleColor(
    Scale Alpha=&FloatFromCurve(
     FloatCurve=DataInterface() {propertyValues {Curve {keys [
         {arriveTangent=-0.2723, interpMode=RCIM_Cubic, leaveTangent=14.5791, tangentMode=RCTM_Break, time=0, value=0}
         {interpMode=RCIM_Cubic, tangentMode=RCTM_Break, time=0.1, value=1}
         {arriveTangent=0.0378, interpMode=RCIM_Cubic, leaveTangent=0.0378, tangentMode=RCTM_User, time=0.6785, value=1.0282}
         {arriveTangent=-3.7547, interpMode=RCIM_Cubic, leaveTangent=-3.7547, tangentMode=RCTM_Break, time=0.8, value=0}
        ]}}}
     Scale Curve=&FNTimeOfDayScalar(Night Scale=1, Morning Scale=1, Evening Scale=1)
    )
    Scale RGB=&VectorFromFloat(Value=1)
    ScaleA=true
    ScaleRGB=true
   ) {enabled=false}
   &DynamicMaterialParameters(
    Index 0 Param 1=&FloatFromCurve(FloatCurve=DataInterface() {propertyValues {Curve {defaultValue=0, keys [{tangentMode=RCTM_User, time=0, value=1}, {tangentMode=RCTM_User, time=1, value=0}]}}})
    Module.Index0Float0Write=true
    Module.Index0Float1Write=false
    Module.Index0Float2Write=false
    Module.Index0Float3Write=false
   )
   &ScaleColor(
    Scale Alpha=&FloatFromCurve(
     FloatCurve=DataInterface() {propertyValues {Curve {keys [
         {arriveTangent=-0.2723, interpMode=RCIM_Cubic, leaveTangent=14.5791, tangentMode=RCTM_Break, time=0, value=0}
         {interpMode=RCIM_Cubic, tangentMode=RCTM_Break, time=0.1, value=1}
         {arriveTangent=0.0378, interpMode=RCIM_Cubic, leaveTangent=0.0378, tangentMode=RCTM_User, time=0.45, value=1.0282}
         {arriveTangent=-0.0539, interpMode=RCIM_Cubic, leaveTangent=-3.7547, tangentMode=RCTM_Break, time=0.7, value=0}
        ]}}}
     Scale Curve=&FNTimeOfDayScalar(Night Scale=0.3, Morning Scale=0.7, Day Scale=0.7, Evening Scale=1)
    )
    Scale RGB=&VectorFromFloat(Value=&FNTimeOfDayScalar(Night Scale=1.5, Morning Scale=2, Day Scale=1.8, Evening Scale=2))
    ScaleA=true
    ScaleRGB=true
   ) {module=ScaleColor001}
  ]
  rendererClasses [&NiagaraSpriteRendererProperties]
  renderers [{rendererClass=&NiagaraSpriteRendererProperties}]
 }
 Splash_Both002 {
  bEnabled=false
  emitterUpdateScript [
   &EmitterState(Life Cycle Mode=::NewEnumerator1, Loop Behavior=::NewEnumerator1, Loop Delay=0.5, Loop Duration=2.3, UseLoopDelay=false)
   &SpawnRate(SpawnRate=100, Use Spawn Probability=false) {enabled=false}
   &SpawnBurst_Instantaneous(Spawn Count=12, Spawn Group=1, Use Spawn Probability=false, UseLoopCountLimit=false)
  ]
  particleSpawnScript [
   &InitializeParticle(
    Color=$System.Water Splashes color
    Lifetime Max=0.8
    Lifetime Min=0.2
    Lifetime Mode=::NewEnumerator1
    Position Offset=<0 0 -10>
    Sprite Rotation Angle=&RandomRangeFloat(Minimum=0, Maximum=0.2)
    Sprite Rotation Mode=::NewEnumerator1
    Sprite Size Mode=::NewEnumerator1
    Sprite UV Mode=::NewEnumerator3
    Uniform Sprite Size Max=30
    Uniform Sprite Size Min=18
    UsePositionOffset=true
    Write Material Random=true
   )
   &CameraOffset(Camera Offset Amount=3)
   &ShapeLocation(
    Cone Length=30
    MaskSpawnGroup=false
    Offset=<0 0 -3>
    Offset Mode=::NewEnumerator0
    Override Seed=false
    Shape Primitive=::NewEnumerator5
    UseOwnerScale=true
   )
   &SubUVAnimation(
    End Frame Range Override=3
    Evaluation Type=::NewEnumerator0
    "Sprite Renderer "=DataInterface() {propertyValues {SpriteRenderer=&self:Splash_Both002_0.NiagaraSpriteRendererProperties_1}}
    SubUV Animation Mode:ENiagara_SubUVLookupModeV2=::NewEnumerator2
    UseEndFrame=true
    UseStartFrame=true
   )
   &AddVelocity(
    Module.Use Falloff Distance=false
    Origin Offset=<0 0 -7>
    Origin Offset Coordinate Space=::Simulation
    Override Seed=false
    Velocity Mode=::NewEnumerator1
    Velocity Speed=&RandomRangeFloat(Minimum=75, Maximum=200)
   ) {module=AddVelocity001}
  ]
  particleUpdateScript [
   &ParticleState()
   &SpriteRotationRate(Rotation Rate=&MaskFloat(
     Mask Source=0
     Mask Parameter=?
     Mask=0
     Masked Float=&RandomRangeFloat(Minimum=-70, Maximum=-100)
     Passthrough Float=&MaskFloat(
      Mask Source=0
      Mask Parameter=?
      Mask=1
      Masked Float=&RandomRangeFloat(Minimum=70, Maximum=100)
      Passthrough Float=1
     )
    ))
   &GravityForce(Gravity=&RandomRangeVector(Minimum=<0 0 -200>, Maximum=<0 0 -400>))
   &Drag(Drag=&RandomRangeFloat(Minimum=0.2, Maximum=0.3), Use Linear Drag=true, Use Rotational Drag=false)
   &SolveForcesAndVelocity(Clamp Velocity=false, Limit Acceleration=false)
   &ScaleSpriteSize(Uniform Curve Sprite Scale=DataInterface() {propertyValues {Curve {keys [
        {arriveTangent=4.1785, interpMode=RCIM_Cubic, leaveTangent=4.1785, tangentMode=RCTM_User, time=0, value=0.15}
        {arriveTangent=1.1179, interpMode=RCIM_Cubic, leaveTangent=1.1179, tangentMode=RCTM_User, time=0.2042, value=0.6228}
        {arriveTangent=0.0224, interpMode=RCIM_Cubic, leaveTangent=0.0224, tangentMode=RCTM_User, time=1, value=1}
       ]}}})
   &ScaleSpriteSizeBySpeed(
    Sample Scale Factor By Curve=true
    Scale Factor Curve=DataInterface() {propertyValues {Curve {defaultValue=0.5941, keys [{tangentMode=RCTM_User, time=0, value=1}, {tangentMode=RCTM_User, time=1, value=0}]}}}
    Velocity Threshold=250
   )
   &ScaleColor(
    Scale Alpha=&FloatFromCurve(
     FloatCurve=DataInterface() {propertyValues {Curve {keys [
         {arriveTangent=-0.2723, interpMode=RCIM_Cubic, leaveTangent=14.5791, tangentMode=RCTM_Break, time=0, value=0}
         {interpMode=RCIM_Cubic, tangentMode=RCTM_Break, time=0.1, value=1}
         {arriveTangent=0.0378, interpMode=RCIM_Cubic, leaveTangent=0.0378, tangentMode=RCTM_User, time=0.6785, value=1.0282}
         {arriveTangent=-3.7547, interpMode=RCIM_Cubic, leaveTangent=-3.7547, tangentMode=RCTM_Break, time=0.8, value=0}
        ]}}}
     Scale Curve=&FNTimeOfDayScalar(Night Scale=1, Morning Scale=1, Evening Scale=1)
    )
    Scale RGB=&VectorFromFloat(Value=1)
    ScaleA=true
    ScaleRGB=true
   ) {enabled=false}
   &DynamicMaterialParameters(
    Index 0 Param 1=&FloatFromCurve(FloatCurve=DataInterface() {propertyValues {Curve {defaultValue=0, keys [{tangentMode=RCTM_User, time=0, value=1}, {tangentMode=RCTM_User, time=1, value=0}]}}})
    Module.Index0Float0Write=true
    Module.Index0Float1Write=false
    Module.Index0Float2Write=false
    Module.Index0Float3Write=false
   )
   &ScaleColor(
    Scale Alpha=&FloatFromCurve(
     FloatCurve=DataInterface() {propertyValues {Curve {keys [
         {arriveTangent=-0.2723, interpMode=RCIM_Cubic, leaveTangent=14.5791, tangentMode=RCTM_Break, time=0, value=0}
         {interpMode=RCIM_Cubic, tangentMode=RCTM_Break, time=0.1, value=1}
         {arriveTangent=0.0378, interpMode=RCIM_Cubic, leaveTangent=0.0378, tangentMode=RCTM_User, time=0.45, value=1.0282}
         {arriveTangent=-0.0539, interpMode=RCIM_Cubic, leaveTangent=-3.7547, tangentMode=RCTM_Break, time=0.7, value=0}
        ]}}}
     Scale Curve=&FNTimeOfDayScalar(Night Scale=0.3, Morning Scale=0.7, Day Scale=0.7, Evening Scale=1)
    )
    Scale RGB=&VectorFromFloat(Value=&FNTimeOfDayScalar(Night Scale=1.5, Morning Scale=2, Day Scale=1.8, Evening Scale=2))
    ScaleA=true
    ScaleRGB=true
   ) {module=ScaleColor001}
  ]
  rendererClasses [&NiagaraSpriteRendererProperties]
  renderers [{rendererClass=&NiagaraSpriteRendererProperties}]
 }
 Splash_Both003 {
  bEnabled=false
  emitterUpdateScript [
   &EmitterState(Life Cycle Mode=::NewEnumerator1, Loop Behavior=::NewEnumerator1, Loop Delay=0.5, Loop Duration=2.3, UseLoopDelay=false)
   &SpawnBurst_Instantaneous(Spawn Count=20, Spawn Group=1, Use Spawn Probability=false, UseLoopCountLimit=false)
  ]
  particleSpawnScript [
   &InitializeParticle(
    Color=$System.Water Splashes color
    Lifetime Max=0.8
    Lifetime Min=0.2
    Lifetime Mode=::NewEnumerator1
    Sprite Rotation Angle=&RandomRangeFloat(Minimum=0, Maximum=0.2)
    Sprite Rotation Mode=::NewEnumerator1
    Sprite Size Mode=::NewEnumerator1
    Sprite UV Mode=::NewEnumerator3
    Uniform Sprite Size Max=25
    Uniform Sprite Size Min=10
    UsePositionOffset=false
    Write Material Random=true
   )
   &CameraOffset(Camera Offset Amount=3)
   &ShapeLocation(MaskSpawnGroup=false, Offset Mode=::NewEnumerator0, Override Seed=false, Ring Radius=5, Shape Primitive=::NewEnumerator4, UseOwnerScale=true)
   &SubUVAnimation(
    End Frame Range Override=3
    Evaluation Type=::NewEnumerator0
    "Sprite Renderer "=DataInterface() {propertyValues {SpriteRenderer=&self:Splash_Both003_0.NiagaraSpriteRendererProperties_1}}
    SubUV Animation Mode:ENiagara_SubUVLookupModeV2=::NewEnumerator2
    UseEndFrame=true
    UseStartFrame=true
   )
   &AddVelocity(
    Module.Use Falloff Distance=false
    Origin Offset=<0 0 -17.5001>
    Origin Offset Coordinate Space=::Simulation
    Override Seed=false
    Velocity Mode=::NewEnumerator1
    Velocity Speed=&RandomRangeFloat(Minimum=75, Maximum=200)
   ) {module=AddVelocity001}
  ]
  particleUpdateScript [
   &ParticleState()
   &SpriteRotationRate(Rotation Rate=&MaskFloat(
     Mask Source=0
     Mask Parameter=?
     Mask=0
     Masked Float=&RandomRangeFloat(Minimum=-70, Maximum=-100)
     Passthrough Float=&MaskFloat(
      Mask Source=0
      Mask Parameter=?
      Mask=1
      Masked Float=&RandomRangeFloat(Minimum=70, Maximum=100)
      Passthrough Float=1
     )
    ))
   &GravityForce(Gravity=&RandomRangeVector(Minimum=<0 0 -200>, Maximum=<0 0 -400>))
   &Drag(Drag=&RandomRangeFloat(Minimum=0.2, Maximum=0.3), Use Linear Drag=true, Use Rotational Drag=false)
   &SolveForcesAndVelocity(Clamp Velocity=false, Limit Acceleration=false)
   &ScaleSpriteSize(Uniform Curve Sprite Scale=DataInterface() {propertyValues {Curve {keys [
        {arriveTangent=4.1785, interpMode=RCIM_Cubic, leaveTangent=4.1785, tangentMode=RCTM_User, time=0, value=0.15}
        {arriveTangent=1.1179, interpMode=RCIM_Cubic, leaveTangent=1.1179, tangentMode=RCTM_User, time=0.2042, value=0.6228}
        {arriveTangent=0.0224, interpMode=RCIM_Cubic, leaveTangent=0.0224, tangentMode=RCTM_User, time=1, value=1}
       ]}}})
   &ScaleSpriteSizeBySpeed(
    Sample Scale Factor By Curve=true
    Scale Factor Curve=DataInterface() {propertyValues {Curve {defaultValue=0.5941, keys [{tangentMode=RCTM_User, time=0, value=1}, {tangentMode=RCTM_User, time=1, value=0}]}}}
    Velocity Threshold=250
   )
   &ScaleColor(
    Scale Alpha=&FloatFromCurve(
     FloatCurve=DataInterface() {propertyValues {Curve {keys [
         {arriveTangent=-0.2723, interpMode=RCIM_Cubic, leaveTangent=14.5791, tangentMode=RCTM_Break, time=0, value=0}
         {interpMode=RCIM_Cubic, tangentMode=RCTM_Break, time=0.1, value=1}
         {arriveTangent=0.0378, interpMode=RCIM_Cubic, leaveTangent=0.0378, tangentMode=RCTM_User, time=0.6785, value=1.0282}
         {arriveTangent=-3.7547, interpMode=RCIM_Cubic, leaveTangent=-3.7547, tangentMode=RCTM_Break, time=0.8, value=0}
        ]}}}
     Scale Curve=&FNTimeOfDayScalar(Night Scale=1, Morning Scale=1, Evening Scale=1)
    )
    Scale RGB=&VectorFromFloat(Value=1)
    ScaleA=true
    ScaleRGB=true
   ) {enabled=false}
   &DynamicMaterialParameters(
    Index 0 Param 1=&FloatFromCurve(FloatCurve=DataInterface() {propertyValues {Curve {defaultValue=0, keys [{tangentMode=RCTM_User, time=0, value=1}, {tangentMode=RCTM_User, time=1, value=0}]}}})
    Module.Index0Float0Write=true
    Module.Index0Float1Write=false
    Module.Index0Float2Write=false
    Module.Index0Float3Write=false
   )
   &ScaleColor(
    Scale Alpha=&FloatFromCurve(
     FloatCurve=DataInterface() {propertyValues {Curve {keys [
         {arriveTangent=-0.2723, interpMode=RCIM_Cubic, leaveTangent=14.5791, tangentMode=RCTM_Break, time=0, value=0}
         {interpMode=RCIM_Cubic, tangentMode=RCTM_Break, time=0.1, value=1}
         {arriveTangent=0.0378, interpMode=RCIM_Cubic, leaveTangent=0.0378, tangentMode=RCTM_User, time=0.45, value=1.0282}
         {arriveTangent=-0.0539, interpMode=RCIM_Cubic, leaveTangent=-3.7547, tangentMode=RCTM_Break, time=0.7, value=0}
        ]}}}
     Scale Curve=&FNTimeOfDayScalar(Night Scale=0.3, Morning Scale=0.7, Day Scale=0.7, Evening Scale=1)
    )
    Scale RGB=&VectorFromFloat(Value=&FNTimeOfDayScalar(Night Scale=1.5, Morning Scale=2, Day Scale=1.8, Evening Scale=2))
    ScaleA=true
    ScaleRGB=true
   ) {module=ScaleColor001}
  ]
  rendererClasses [&NiagaraSpriteRendererProperties]
  renderers [{rendererClass=&NiagaraSpriteRendererProperties}]
 }
 Splash_Both004 {
  bEnabled=false
  emitterUpdateScript [
   &EmitterState(Life Cycle Mode=::NewEnumerator1, Loop Behavior=::NewEnumerator1, Loop Delay=0.5, Loop Duration=2.3, UseLoopDelay=false)
   &SpawnBurst_Instantaneous(Spawn Count=12, Spawn Group=1, Use Spawn Probability=false, UseLoopCountLimit=false)
  ]
  particleSpawnScript [
   &InitializeParticle(
    Color=$System.Water Splashes color
    Lifetime Max=0.6
    Lifetime Min=0.2
    Lifetime Mode=::NewEnumerator1
    Position Offset=<0 0 -13>
    Sprite Rotation Angle=&RandomRangeFloat(Minimum=0, Maximum=0.2)
    Sprite Rotation Mode=::NewEnumerator1
    Sprite Size Mode=::NewEnumerator1
    Sprite UV Mode=::NewEnumerator3
    Uniform Sprite Size Max=20
    Uniform Sprite Size Min=10
    UsePositionOffset=true
    Write Material Random=true
   )
   &CameraOffset(Camera Offset Amount=3)
   &ShapeLocation(
    Cone Length=25
    MaskSpawnGroup=true
    Offset=<0 0 -3>
    Offset Mode=::NewEnumerator0
    Override Seed=false
    Shape Primitive=::NewEnumerator5
    Spawn Group Mask=1
    UseOwnerScale=true
   )
   &SubUVAnimation(
    End Frame Range Override=3
    Evaluation Type=::NewEnumerator0
    "Sprite Renderer "=DataInterface() {propertyValues {SpriteRenderer=&self:Splash_Both004_0.NiagaraSpriteRendererProperties_1}}
    SubUV Animation Mode:ENiagara_SubUVLookupModeV2=::NewEnumerator2
    UseEndFrame=true
    UseStartFrame=true
   )
   &AddVelocity(
    Module.Use Falloff Distance=false
    Origin Offset=<0 0 -10>
    Origin Offset Coordinate Space=::Simulation
    Override Seed=false
    Velocity Mode=::NewEnumerator1
    Velocity Speed=&RandomRangeFloat(Minimum=50, Maximum=100)
   ) {module=AddVelocity001}
  ]
  particleUpdateScript [
   &ParticleState()
   &SpriteRotationRate(Rotation Rate=&MaskFloat(
     Mask Source=0
     Mask Parameter=?
     Mask=0
     Masked Float=&RandomRangeFloat(Minimum=-70, Maximum=-100)
     Passthrough Float=&MaskFloat(
      Mask Source=0
      Mask Parameter=?
      Mask=1
      Masked Float=&RandomRangeFloat(Minimum=70, Maximum=100)
      Passthrough Float=1
     )
    ))
   &GravityForce(Gravity=&RandomRangeVector(Minimum=<0 0 -200>, Maximum=<0 0 -400>))
   &Drag(Drag=&RandomRangeFloat(Minimum=0.05, Maximum=0.1), Use Linear Drag=true, Use Rotational Drag=false)
   &SolveForcesAndVelocity(Clamp Velocity=false, Limit Acceleration=false)
   &ScaleSpriteSize(Uniform Curve Sprite Scale=DataInterface() {propertyValues {Curve {keys [
        {arriveTangent=4.1785, interpMode=RCIM_Cubic, leaveTangent=4.1785, tangentMode=RCTM_User, time=0, value=0.15}
        {arriveTangent=1.1179, interpMode=RCIM_Cubic, leaveTangent=1.1179, tangentMode=RCTM_User, time=0.2042, value=0.6228}
        {arriveTangent=0.0224, interpMode=RCIM_Cubic, leaveTangent=0.0224, tangentMode=RCTM_User, time=1, value=1}
       ]}}})
   &ScaleSpriteSizeBySpeed(
    Sample Scale Factor By Curve=true
    Scale Factor Curve=DataInterface() {propertyValues {Curve {defaultValue=0.5941, keys [{tangentMode=RCTM_User, time=0, value=1}, {tangentMode=RCTM_User, time=1, value=0}]}}}
    Velocity Threshold=250
   )
   &ScaleColor(
    Scale Alpha=&FloatFromCurve(
     FloatCurve=DataInterface() {propertyValues {Curve {keys [
         {arriveTangent=-0.2723, interpMode=RCIM_Cubic, leaveTangent=14.5791, tangentMode=RCTM_Break, time=0, value=0}
         {interpMode=RCIM_Cubic, tangentMode=RCTM_Break, time=0.1, value=1}
         {arriveTangent=0.0378, interpMode=RCIM_Cubic, leaveTangent=0.0378, tangentMode=RCTM_User, time=0.6785, value=1.0282}
         {arriveTangent=-3.7547, interpMode=RCIM_Cubic, leaveTangent=-3.7547, tangentMode=RCTM_Break, time=0.8, value=0}
        ]}}}
     Scale Curve=&FNTimeOfDayScalar(Night Scale=1, Morning Scale=1, Evening Scale=1)
    )
    Scale RGB=&VectorFromFloat(Value=1)
    ScaleA=true
    ScaleRGB=true
   ) {enabled=false}
   &DynamicMaterialParameters(
    Index 0 Param 1=&FloatFromCurve(FloatCurve=DataInterface() {propertyValues {Curve {defaultValue=0, keys [{tangentMode=RCTM_User, time=0, value=1}, {tangentMode=RCTM_User, time=1, value=0}]}}})
    Module.Index0Float0Write=true
    Module.Index0Float1Write=false
    Module.Index0Float2Write=false
    Module.Index0Float3Write=false
   )
   &ScaleColor(
    Scale Alpha=&FloatFromCurve(
     FloatCurve=DataInterface() {propertyValues {Curve {keys [
         {arriveTangent=-0.2723, interpMode=RCIM_Cubic, leaveTangent=14.5791, tangentMode=RCTM_Break, time=0, value=0}
         {interpMode=RCIM_Cubic, tangentMode=RCTM_Break, time=0.1, value=1}
         {arriveTangent=0.0378, interpMode=RCIM_Cubic, leaveTangent=0.0378, tangentMode=RCTM_User, time=0.45, value=1.0282}
         {arriveTangent=-0.0539, interpMode=RCIM_Cubic, leaveTangent=-3.7547, tangentMode=RCTM_Break, time=0.7, value=0}
        ]}}}
     Scale Curve=&FNTimeOfDayScalar(Night Scale=0.3, Morning Scale=0.7, Day Scale=0.7, Evening Scale=1)
    )
    Scale RGB=&VectorFromFloat(Value=&FNTimeOfDayScalar(Night Scale=1.5, Morning Scale=2, Day Scale=1.8, Evening Scale=2))
    ScaleA=true
    ScaleRGB=true
   ) {module=ScaleColor001}
  ]
  rendererClasses [&NiagaraSpriteRendererProperties]
  renderers [{rendererClass=&NiagaraSpriteRendererProperties}]
 }
 Water_Plasma001 {
  bEnabled=false
  emitterUpdateScript [
   &EmitterState(Life Cycle Mode=::NewEnumerator1, Loop Behavior=::NewEnumerator1, Loop Delay=0.5, Loop Duration=1.5, UseLoopDelay=false)
   &SpawnBurst_Instantaneous(Use Spawn Probability=false, UseLoopCountLimit=false)
   &SpawnRate(SpawnRate=2, Use Spawn Probability=false) {enabled=false}
  ]
  particleSpawnScript [
   &InitializeParticle(
    Color=$System.Ripple color
    Lifetime=0.75
    Mesh Scale Mode=::NewEnumerator1
    Mesh Uniform Scale Max=1
    Mesh Uniform Scale Min=0.5
    Position Offset=<0 0 15>
    UsePositionOffset=true
    Write Material Random=true
   )
   &FN_SetSpawnGroup(Number of Spawn Groups=1)
   &DynamicMaterialParameters(
    Index 0 Param 1=&RandomRangeFloat(Minimum=0, Maximum=1)
    Index 0 Param 3=1
    Module.Index0Param0Write=true
    Module.Index0Param1Write=true
    Module.Index0Param2Write=true
    Module.Index0Param3Write=false
   )
  ]
  particleUpdateScript [
   &ParticleState()
   &ScaleColor(
    Scale Alpha=&FloatFromCurve(
     FloatCurve=DataInterface() {propertyValues {Curve {
        defaultValue=0
        keys [
         {interpMode=RCIM_Cubic, tangentMode=RCTM_User, time=0, value=0.5}
         {arriveTangent=-0.0455, interpMode=RCIM_Cubic, leaveTangent=-0.0455, tangentMode=RCTM_User, time=0.1103, value=0.8978}
         {interpMode=RCIM_Cubic, tangentMode=RCTM_User, time=0.5, value=1}
         {tangentMode=RCTM_User, time=0.89, value=0}
        ]
       }}}
     Scale Curve=&FNTimeOfDayScalar(Night Scale=3, Morning Scale=2.5, Day Scale=2.5, Evening Scale=1)
    )
    Scale RGB=&VectorFromFloat(Value=&FNTimeOfDayScalar(Night Scale=1, Morning Scale=1, Evening Scale=1))
    ScaleA=true
    ScaleRGB=true
   )
   &InterpolateOverTime(
    Rate Of Change=8
    Target Vector Value:Vector3f=&AddVectorToPosition(
     Position=$Particles.Position
     Vector=&RotateVectorByQuaternion(Vector To Rotate=<0 -100 0>, Quaternion=$Particles.SkeletalMeshLocation.SampledRotation)
    )
    Variable Type=::NewEnumerator2
   )
   &DynamicMaterialParameters(
    Index 2 Param 1=&MakeFloatFromVector(
     Vector=&Subtract_Position(
      A=&AddVectorToPosition(
       Position=$Particles.Position
       Vector=&RotateVectorByQuaternion(Vector To Rotate=<0 -100 0>, Quaternion=$Particles.SkeletalMeshLocation.SampledRotation)
      )
      B=Linked() {linkedVariable {name=Particles.InterpolateOverTime.InterpolateOverTime_V3_.Moving Average, type=Vector3f}}
     )
     Channel=::NewEnumerator0
    )
    Index 2 Param 2=&MakeFloatFromVector(
     Vector=&Subtract_Position(
      A=&AddVectorToPosition(
       Position=$Particles.Position
       Vector=&RotateVectorByQuaternion(Vector To Rotate=<0 -100 0>, Quaternion=$Particles.SkeletalMeshLocation.SampledRotation)
      )
      B=Linked() {linkedVariable {name=Particles.InterpolateOverTime.InterpolateOverTime_V3_.Moving Average, type=Vector3f}}
     )
     Channel=::NewEnumerator1
    )
    Index 2 Param 3=&MakeFloatFromVector(
     Vector=&Subtract_Position(
      A=&AddVectorToPosition(
       Position=$Particles.Position
       Vector=&RotateVectorByQuaternion(Vector To Rotate=<0 -100 0>, Quaternion=$Particles.SkeletalMeshLocation.SampledRotation)
      )
      B=Linked() {linkedVariable {name=Particles.InterpolateOverTime.InterpolateOverTime_V3_.Moving Average, type=Vector3f}}
     )
     Channel=::NewEnumerator2
    )
    Module.Param0WriteEnabled2=true
    Module.Param1WriteEnabled2=true
    Module.Param2WriteEnabled2=true
    Module.Param3WriteEnabled2=true
    Write Parameter Index 0=false
    Write Parameter Index 2=true
   ) {module=DynamicMaterialParameters001}
   &ScaleMeshSize(Scale Factor=&VectorFromFloat(Value=&RandomRangeFloat(
      Minimum=&FloatFromCurve(FloatCurve=DataInterface() {propertyValues {Curve {defaultValue=1, keys [{tangentMode=RCTM_User, time=0, value=0.5}, {tangentMode=RCTM_User, time=1, value=1}]}}})
      Maximum=&FloatFromCurve(
       FloatCurve=DataInterface() {propertyValues {Curve {defaultValue=1, keys [{tangentMode=RCTM_User, time=0, value=0.5}, {tangentMode=RCTM_User, time=1, value=1}]}}}
       Scale Curve=2.5
      )
     ))) {module=ScaleMeshSize001}
  ]
  rendererClasses [&NiagaraMeshRendererProperties]
  renderers [{rendererClass=&NiagaraMeshRendererProperties, rendererIndex=0}]
 }
 splash {
  emitterUpdateScript [
   &EmitterState()
   &SpawnBurst_Instantaneous(Spawn Count=12, Spawn Group=1, Use Spawn Probability=false, UseLoopCountLimit=false) {module=SpawnBurst_Instantaneous001}
   &SpawnBurst_Instantaneous(Spawn Count=20, Use Spawn Probability=false, UseLoopCountLimit=false)
  ]
  particleSpawnScript [
   /SetVariables(Particles.VisibilityTag=&SpawnGroup()) {isSetParameters=true}
   &InitializeParticle(
    Color=$System.Water Splashes color
    Lifetime=&MergeHelper_FloatSwitch(
     0=&RandomRangeFloat(Minimum=0.2, Maximum=0.8)
     1=&RandomRangeFloat(Minimum=0.2, Maximum=0.6)
     2=?
     3=?
     Index=&SpawnGroup()
     Num=::NewEnumerator0
    )
    Sprite Rotation Angle=&RandomRangeFloat(Minimum=0, Maximum=0.2)
    Sprite Rotation Mode=::NewEnumerator1
    Sprite Size Mode=::NewEnumerator1
    Sprite UV Mode=::NewEnumerator3
    Uniform Sprite Size Max=25
    Uniform Sprite Size Min=&MergeHelper_FloatSwitch(
     0=&RandomRangeFloat(Minimum=10, Maximum=25)
     1=&RandomRangeFloat(Minimum=10, Maximum=20)
     2=?
     3=?
     Index=&SpawnGroup()
     Num=::NewEnumerator0
    )
    UsePositionOffset=false
    Write Material Random=true
   )
   &CameraOffset(Camera Offset Amount=&MergeHelper_FloatSwitch(0=3, 1=2, 2=?, 3=?, Index=&SpawnGroup(), Num=::NewEnumerator0))
   &ShapeLocation(MaskSpawnGroup=true, Override Seed=false, Sphere Radius=5, UseOwnerScale=true)
   &SubUVAnimation(
    End Frame Range Override=3
    Evaluation Type=::NewEnumerator0
    "Sprite Renderer "=DataInterface() {propertyValues {SpriteRenderer=&self:splash_0.NiagaraSpriteRendererProperties_1}}
    SubUV Animation Mode:ENiagara_SubUVLookupModeV2=::NewEnumerator2
    UseEndFrame=true
    UseStartFrame=true
   )
   &SubUVAnimation(
    End Frame Range Override=3
    Evaluation Type=::NewEnumerator0
    "Sprite Renderer "=DataInterface() {propertyValues {SpriteRenderer=&self:splash_0.NiagaraSpriteRendererProperties_2}}
    SubUV Animation Mode:ENiagara_SubUVLookupModeV2=::NewEnumerator2
    UseEndFrame=true
    UseStartFrame=true
   ) {module=SubUVAnimation001}
   &AddVelocity(
    Module.Use Falloff Distance=false
    Origin Offset=&MergeHelper_VectorSwitch(0=<0 0 -17.5001>, 1=<0 0 -10>, 2=?, 3=?, Index=&SpawnGroup(), Num=::NewEnumerator0)
    Override Seed=false
    Velocity Mode=::NewEnumerator1
    Velocity Speed=&MergeHelper_FloatSwitch(
     0=&RandomRangeFloat(Minimum=75, Maximum=200)
     1=&RandomRangeFloat(Minimum=50, Maximum=100)
     2=?
     3=?
     Index=&SpawnGroup()
     Num=::NewEnumerator0
    )
   )
   &ShapeLocation(
    Cone Length=25
    MaskSpawnGroup=true
    Offset=<0 0 -3>
    Offset Mode=::NewEnumerator0
    Override Seed=false
    Shape Primitive=::NewEnumerator5
    Spawn Group Mask=1
    UseOwnerScale=true
   ) {module=ShapeLocation001}
  ]
  particleUpdateScript [
   &ParticleState()
   &ScaleSpriteSize(Uniform Curve Sprite Scale=DataInterface() {propertyValues {Curve {keys [
        {arriveTangent=4.1785, interpMode=RCIM_Cubic, leaveTangent=4.1785, tangentMode=RCTM_User, time=0, value=0.15}
        {arriveTangent=1.1179, interpMode=RCIM_Cubic, leaveTangent=1.1179, tangentMode=RCTM_User, time=0.2042, value=0.6228}
        {arriveTangent=0.0224, interpMode=RCIM_Cubic, leaveTangent=0.0224, tangentMode=RCTM_User, time=1, value=1}
       ]}}})
   &SpriteRotationRate(Rotation Rate=&MaskFloat(
     Mask Source=0
     Mask Parameter=?
     Mask=0
     Masked Float=&RandomRangeFloat(Minimum=-70, Maximum=-100)
     Passthrough Float=&MaskFloat(
      Mask Source=0
      Mask Parameter=?
      Mask=1
      Masked Float=&RandomRangeFloat(Minimum=70, Maximum=100)
      Passthrough Float=1
     )
    ))
   &GravityForce(Gravity=&RandomRangeVector(Minimum=<0 0 -200>, Maximum=<0 0 -400>))
   &Drag(
    Drag=&MergeHelper_FloatSwitch(
     0=&RandomRangeFloat(Minimum=0.2, Maximum=0.3)
     1=&RandomRangeFloat(Minimum=0.05, Maximum=0.1)
     2=?
     3=?
     Index=&SpawnGroup()
     Num=::NewEnumerator0
    )
    Rotational Drag:NiagaraFloat=&MergeHelper_FloatSwitch(0=0, 1=0, 2=?, 3=?, Index=&SpawnGroup(), Num=::NewEnumerator0)
    Use Linear Drag=true
    Use Rotational Drag=false
   )
   &SolveForcesAndVelocity(Clamp Velocity=false, Limit Acceleration=false)
   &ScaleSpriteSizeBySpeed(
    Sample Scale Factor By Curve=true
    Scale Factor Curve=DataInterface() {propertyValues {Curve {defaultValue=0.5941, keys [{tangentMode=RCTM_User, time=0, value=1}, {tangentMode=RCTM_User, time=1, value=0}]}}}
    Velocity Threshold=250
   )
   &DynamicMaterialParameters(
    Index 0 Param 1=&FloatFromCurve(FloatCurve=DataInterface() {propertyValues {Curve {defaultValue=0, keys [{tangentMode=RCTM_User, time=0, value=1}, {tangentMode=RCTM_User, time=1, value=0}]}}})
    Module.Index0Float0Write=true
    Module.Index0Float1Write=false
    Module.Index0Float2Write=false
    Module.Index0Float3Write=false
   )
   &ScaleColor(
    Scale Alpha=&FloatFromCurve(
     FloatCurve=DataInterface() {propertyValues {Curve {keys [
         {arriveTangent=-0.2723, interpMode=RCIM_Cubic, leaveTangent=14.5791, tangentMode=RCTM_Break, time=0, value=0}
         {interpMode=RCIM_Cubic, tangentMode=RCTM_Break, time=0.1, value=1}
         {arriveTangent=0.0378, interpMode=RCIM_Cubic, leaveTangent=0.0378, tangentMode=RCTM_User, time=0.45, value=1.0282}
         {arriveTangent=-0.0539, interpMode=RCIM_Cubic, leaveTangent=-3.7547, tangentMode=RCTM_Break, time=0.7, value=0}
        ]}}}
     Scale Curve=&FNTimeOfDayScalar(Night Scale=0.3, Morning Scale=0.7, Day Scale=0.7, Evening Scale=1)
    )
    Scale RGB=&VectorFromFloat(Value=&FNTimeOfDayScalar(Night Scale=1.5, Morning Scale=2, Day Scale=1.8, Evening Scale=2))
    ScaleA=true
    ScaleRGB=true
   ) {module=ScaleColor001}
  ]
  rendererClasses [&NiagaraSpriteRendererProperties]
  renderers [
   {rendererClass=&NiagaraSpriteRendererProperties}
   {rendererClass=&NiagaraSpriteRendererProperties, rendererIndex=1}
  ]
 }
 thicc_drops {rendererClasses [], renderers []}
 water_ripple {
  bEnabled=false
  emitterUpdateScript [
   &EmitterState(Life Cycle Mode=::NewEnumerator1, Loop Behavior=::NewEnumerator1, Loop Delay=0.5, Loop Duration=1.5, UseLoopDelay=false)
   &SpawnBurst_Instantaneous(Spawn Time=0.01, Use Spawn Probability=false, UseLoopCountLimit=false)
   &SpawnRate(SpawnRate=2, Use Spawn Probability=false) {enabled=false}
  ]
  particleSpawnScript [
   &InitializeParticle(
    Color=$System.Ripple color
    Lifetime=0.75
    Mesh Scale Mode=::NewEnumerator3
    Mesh Uniform Scale=0.75
    Position Offset=<0 0 15>
    UsePositionOffset=true
    Write Material Random=true
   )
   &FN_SetSpawnGroup(Number of Spawn Groups=1)
  ]
  particleUpdateScript [
   &ParticleState()
   &ScaleColor(
    Scale Alpha=&FloatFromCurve(
     FloatCurve=DataInterface() {propertyValues {Curve {
        defaultValue=0
        keys [
         {interpMode=RCIM_Cubic, tangentMode=RCTM_User, time=0, value=0.5}
         {arriveTangent=-0.0455, interpMode=RCIM_Cubic, leaveTangent=-0.0455, tangentMode=RCTM_User, time=0.1103, value=0.8978}
         {interpMode=RCIM_Cubic, tangentMode=RCTM_User, time=0.5, value=1}
         {tangentMode=RCTM_User, time=0.89, value=0}
        ]
       }}}
     Scale Curve=&FNTimeOfDayScalar(Night Scale=3, Morning Scale=2.5, Evening Scale=1)
    )
    Scale RGB=&VectorFromFloat(Value=&FNTimeOfDayScalar(Night Scale=1.5, Morning Scale=1.3, Day Scale=1.5, Evening Scale=1.2))
    ScaleA=true
    ScaleRGB=true
   )
   &ScaleMeshSize(Scale Factor=&VectorFromFloat(Value=&RandomRangeFloat(
      Minimum=&FloatFromCurve(FloatCurve=DataInterface() {propertyValues {Curve {defaultValue=1, keys [{tangentMode=RCTM_User, time=0, value=0.3}, {tangentMode=RCTM_User, time=1, value=1}]}}})
      Maximum=&FloatFromCurve(
       FloatCurve=DataInterface() {propertyValues {Curve {defaultValue=1, keys [{tangentMode=RCTM_User, time=0, value=0.2}, {tangentMode=RCTM_User, time=1, value=1}]}}}
       Scale Curve=2.5
      )
     ))) {module=ScaleMeshSize001}
  ]
  rendererClasses [&NiagaraMeshRendererProperties]
  renderers [{rendererClass=&NiagaraMeshRendererProperties, rendererIndex=0}]
 }
}
summary {
 systemName=NS_Emote_GnatGalaPancake_splash
 emitters [
  {bEnabled=false, emitterName=Splash_Both, rendererClasses [&NiagaraSpriteRendererProperties]}
  {bEnabled=false, emitterName=Droplets, rendererClasses []}
  {bEnabled=false, emitterName=Water_Plasma001, rendererClasses [&NiagaraMeshRendererProperties]}
  {bEnabled=false, emitterName=Splash_Both001, rendererClasses [&NiagaraSpriteRendererProperties]}
  {bEnabled=false, emitterName=Splash_Both002, rendererClasses [&NiagaraSpriteRendererProperties]}
  {emitterName=Droplets001, rendererClasses []}
  {bEnabled=false, emitterName=water_ripple, rendererClasses [&NiagaraMeshRendererProperties]}
  {bEnabled=false, emitterName=Splash_Both003, rendererClasses [&NiagaraSpriteRendererProperties]}
  {bEnabled=false, emitterName=Splash_Both004, rendererClasses [&NiagaraSpriteRendererProperties]}
  {bEnabled=false, emitterName=Drip_Drop, rendererClasses [&NiagaraSpriteRendererProperties]}
  {bEnabled=false, emitterName=BuffSplash_Mesh, rendererClasses [&NiagaraMeshRendererProperties]}
  {emitterName=LWE_RIPPLE, rendererClasses []}
  {emitterName=thicc_drops, rendererClasses []}
  {emitterName=splash, rendererClasses [&NiagaraSpriteRendererProperties]}
  {emitterName=Moving_water, rendererClasses [&NiagaraMeshRendererProperties]}
  {emitterName=LWE_RIPPLE001, rendererClasses []}
 ]
 userVariables []
}
