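# Niagara system for the GnatGala Pancake emote that sprays water splash and droplet sprites from points sampled on a skeletal mesh.
# Drip_Drop makes droplets, while Splashes_compressed (GPU) and Splashes_compressed001 each merge several splash types into one emitter.
# Spawn groups drive MergeHelper switches for each type's look and motion, with VisibilityTag set for multiple renderers.
# Time-of-day scalars adjust brightness; standalone BOTTOM and SplashFat emitters are disabled.
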NiagaraSystem /BRCosmetics/Animation/Game/MainPlayer/Emotes/GnatGala_Pancake/FX/NS_Emote_GnatGalaPancake_splashes
dependencies {
 usedDataInterfaces [
  &NiagaraDataInterfaceSpriteRendererInfo
  &NiagaraDataInterfaceCurve
  &NiagaraDataInterfaceVector2DCurve
  &NiagaraDataInterfaceSkeletalMesh
 ]
 usedDynamicInputs [&FNTimeOfDayColor, &RandomRangeLinearColor]
 usedModules [
  &self:SystemSpawnScript.SystemScriptSource.SystemScriptGraph.NiagaraNodeAssignment_3.SetVariables_#1
  &SystemState
  &self:SystemSpawnScript.SystemScriptSource.SystemScriptGraph.NiagaraNodeAssignment_0.SetVariables_#2
 ]
}
emitters {
 BOTTOM {
  bEnabled=false
  emitterUpdateScript [
   &EmitterState(Life Cycle Mode=::NewEnumerator1, Loop Behavior=::NewEnumerator1, Loop Delay=0.5, Loop Duration=3, UseLoopDelay=false)
   &SpawnRate(SpawnRate=150, Use Spawn Probability=false)
   &SpawnBurst_Instantaneous(Spawn Count=20, Spawn Group=1, Use Spawn Probability=false, UseLoopCountLimit=false) {enabled=false}
  ]
  particleSpawnScript [
   &InitializeParticle(
    Color=$System.Water splash color
    Lifetime Max=0.55
    Lifetime Min=0.2
    Lifetime Mode=::NewEnumerator1
    Position Offset=<0 0 10>
    Sprite Rotation Angle=&RandomRangeFloat(Minimum=0, Maximum=0.2)
    Sprite Rotation Mode=::NewEnumerator1
    Sprite Size Mode=::NewEnumerator1
    Sprite UV Mode=::NewEnumerator3
    Uniform Sprite Size Max=35
    Uniform Sprite Size Min=20
    UsePositionOffset=false
    Write Material Random=true
   )
   &CameraOffset(Camera Offset Amount=&RandomRangeFloat(Minimum=6, Maximum=15))
   &SkeletalMeshLocation(
    MaskSpawnGroup=false
    Mesh Sampling Type=::NewEnumerator5
    Orientation Sampling=::NewEnumerator1
    Sampled Position Offset=<-5 0 0>
    Skeletal Mesh=$System.Skeletal Mesh
    Socket Sampling Mode=::NewEnumerator5
   )
   &ShapeLocation(
    Disc Coverage=0.3405
    MaskSpawnGroup=false
    Offset=<0 0 -5>
    Offset Mode=::NewEnumerator0
    Override Seed=false
    Ring Radius=10
    Shape Primitive=::NewEnumerator4
    UseOwnerScale=true
   )
   &SubUVAnimation(
    End Frame Range Override=3
    Evaluation Type=::NewEnumerator0
    "Sprite Renderer "=DataInterface() {propertyValues {SpriteRenderer=&self:BOTTOM_0.NiagaraSpriteRendererProperties_1}}
    SubUV Animation Mode:ENiagara_SubUVLookupModeV2=::NewEnumerator2
    UseEndFrame=true
    UseStartFrame=true
   )
   &AddVelocity(
    Module.Use Falloff Distance=false
    Origin Offset=<0 0 -35>
    Origin Offset Coordinate Space=::Simulation
    Override Seed=false
    Velocity Mode=::NewEnumerator1
    Velocity Speed=&RandomRangeFloat(Minimum=75, Maximum=250)
   ) {module=AddVelocity001}
   &AddVelocity(Velocity=&RandomRangeVector(Minimum=<-75 -75 75>, Maximum=<75 75 250>)) {enabled=false}
  ]
  particleUpdateScript [
   &ParticleState()
   &SpriteRotationRate(Rotation Rate=&RandomRangeFloat(Minimum=-70, Maximum=-100))
   &GravityForce(Gravity=&RandomRangeVector(Minimum=<0 0 -500>, Maximum=<0 0 -980>))
   &Drag(Drag=&RandomRangeFloat(Minimum=0.2, Maximum=0.3), Use Linear Drag=true, Use Rotational Drag=false)
   &SolveForcesAndVelocity(Clamp Velocity=false, Limit Acceleration=false)
   &ScaleSpriteSize(Uniform Curve Sprite Scale=DataInterface() {propertyValues {Curve {keys [
        {arriveTangent=4.1785, interpMode=RCIM_Cubic, leaveTangent=4.1785, tangentMode=RCTM_User, time=0, value=0.15}
        {arriveTangent=1.1179, interpMode=RCIM_Cubic, leaveTangent=1.1179, tangentMode=RCTM_User, time=0.2042, value=0.6228}
        {arriveTangent=0.0224, interpMode=RCIM_Cubic, leaveTangent=0.0224, tangentMode=RCTM_User, time=1, value=1}
       ]}}})
   &ScaleSpriteSizeBySpeed(
    Sample Scale Factor By Curve=true
    Scale Factor Curve=DataInterface() {propertyValues {Curve {defaultValue=0.5941, keys [{tangentMode=RCTM_User, time=0, value=1}, {tangentMode=RCTM_User, time=1, value=0}]}}}
    Velocity Threshold=250
   )
   &ScaleColor(
    Scale Alpha=&FloatFromCurve(
     FloatCurve=DataInterface() {propertyValues {Curve {keys [
         {arriveTangent=-0.2723, interpMode=RCIM_Cubic, leaveTangent=14.5791, tangentMode=RCTM_Break, time=0, value=0}
         {interpMode=RCIM_Cubic, tangentMode=RCTM_Break, time=0.1, value=1}
         {arriveTangent=0.0378, interpMode=RCIM_Cubic, leaveTangent=0.0378, tangentMode=RCTM_User, time=0.6785, value=1.0282}
         {arriveTangent=-3.7547, interpMode=RCIM_Cubic, leaveTangent=-3.7547, tangentMode=RCTM_Break, time=0.8, value=0}
        ]}}}
     Scale Curve=&FNTimeOfDayScalar(Night Scale=1, Morning Scale=0.7, Day Scale=2, Evening Scale=0.5)
    )
    Scale RGB=&VectorFromFloat(Value=&FNTimeOfDayScalar(Night Scale=1.5, Morning Scale=2, Day Scale=1.2, Evening Scale=2))
    ScaleA=true
    ScaleRGB=true
   )
   &DynamicMaterialParameters(
    Index 0 Param 1=&FloatFromCurve(FloatCurve=DataInterface() {propertyValues {Curve {defaultValue=0, keys [{tangentMode=RCTM_User, time=0, value=1}, {tangentMode=RCTM_User, time=1, value=0}]}}})
    Module.Index0Float0Write=true
    Module.Index0Float1Write=false
    Module.Index0Float2Write=false
    Module.Index0Float3Write=false
   )
  ]
  rendererClasses [&NiagaraSpriteRendererProperties]
  renderers [{rendererClass=&NiagaraSpriteRendererProperties}]
 }
 Drip_Drop {
  emitterUpdateScript [
   &EmitterState(Life Cycle Mode=::NewEnumerator1, Loop Behavior=::NewEnumerator1, Loop Duration=3, UseLoopDelay=false)
   &SpawnRate(SpawnRate=35, Use Spawn Probability=false)
  ]
  particleSpawnScript [
   &InitializeParticle(
    Color=$System.DropletColors
    Lifetime Max=0.6
    Lifetime Min=0.2
    Lifetime Mode=::NewEnumerator1
    Sprite Size Max=<2.5 10>
    Sprite Size Min=<1.5 5>
    Sprite Size Mode=::NewEnumerator2
    UsePositionOffset=false
    Write Material Random=true
   ) {module=InitializeParticle001}
   &AddVelocity(
    Cone Angle=&RandomRangeFloat(Minimum=90, Maximum=0)
    Cone Axis=<0 0 5>
    Distribution Along Cone Axis=0.25
    Inner Cone Angle=90
    Module.UseConeFalloff=false
    Module.UseVelDistribution=false
    Override Seed=false
    Speed Falloff From Cone Axis=0.5
    Velocity Mode=::NewEnumerator2
    Velocity Speed=&RandomRangeFloat(Minimum=100, Maximum=300)
   ) {module=AddVelocity001}
   &SkeletalMeshLocation(
    MaskSpawnGroup=false
    Mesh Sampling Type=::NewEnumerator5
    Orientation Sampling=::NewEnumerator1
    Sampled Position Offset=<-10 0 0>
    Skeletal Mesh=$System.Skeletal Mesh
    Socket Sampling Mode=::NewEnumerator5
   )
   &ShapeLocation(
    Cylinder Distribution=::NewEnumerator2
    Cylinder Height=5
    Cylinder Radius=20
    MaskSpawnGroup=false
    Shape Primitive=::NewEnumerator1
    UseOwnerScale=true
   )
  ]
  particleUpdateScript [
   &CameraOffset(Camera Offset Amount=3)
   &ParticleState()
   &GravityForce(Gravity=<0 0 -600>)
   &SolveForcesAndVelocity(Clamp Velocity=false, Limit Acceleration=false)
   &ScaleSpriteSize(Scale Factor=&Vector2DFromCurve(
     Vector2Curve=DataInterface() {propertyValues {
       XCurve {keys [
         {arriveTangent=5.9728, interpMode=RCIM_Cubic, leaveTangent=5.9728, tangentMode=RCTM_User, time=0, value=0.1}
         {arriveTangent=-0.1, interpMode=RCIM_Cubic, leaveTangent=-0.1, tangentMode=RCTM_User, time=0.18, value=0.99}
         {interpMode=RCIM_Cubic, tangentMode=RCTM_User, time=1, value=0}
        ]}
       YCurve {keys [
         {arriveTangent=6.2715, interpMode=RCIM_Cubic, leaveTangent=6.2715, tangentMode=RCTM_User, time=0, value=0.1}
         {arriveTangent=-0.1001, interpMode=RCIM_Cubic, leaveTangent=-0.1001, tangentMode=RCTM_User, time=0.18, value=0.99}
         {interpMode=RCIM_Cubic, tangentMode=RCTM_User, time=0.9989, value=0}
        ]}
      }}
     ScaleCurve=<2 2>
    ))
   &ScaleColor(
    Scale Alpha=&FloatFromCurve(
     FloatCurve=DataInterface() {propertyValues {Curve {
        defaultValue=1
        keys [{interpMode=RCIM_Cubic, tangentMode=RCTM_User, time=0, value=1}, {interpMode=RCIM_Cubic, tangentMode=RCTM_User, time=1, value=0}]
       }}}
     Scale Curve=&FNTimeOfDayScalar(Night Scale=0.3, Morning Scale=0.3, Day Scale=0.5, Evening Scale=0.1)
    )
    Scale RGB=&VectorFromFloat(Value=&FNTimeOfDayScalar(Night Scale=0.5, Morning Scale=0.5, Day Scale=0.5, Evening Scale=0.3))
    ScaleA=true
    ScaleRGB=true
   )
  ]
  rendererClasses [&NiagaraSpriteRendererProperties]
  renderers [{rendererClass=&NiagaraSpriteRendererProperties}]
 }
 SplashFat {
  bEnabled=false
  emitterUpdateScript [
   &EmitterState(Life Cycle Mode=::NewEnumerator1, Loop Behavior=::NewEnumerator1, Loop Delay=0.5, Loop Duration=3, UseLoopDelay=false)
   &SpawnRate(SpawnRate=30, Use Spawn Probability=false)
   &SpawnBurst_Instantaneous(Spawn Count=10, Use Spawn Probability=false, UseLoopCountLimit=false) {enabled=false}
  ]
  particleSpawnScript [
   /SetVariables(Particles.Scaler:NiagaraFloat=&FloatFromCurve(
     FloatCurve=DataInterface() {propertyValues {Curve {keys [
         <at 0 0 RCIM_Cubic>
         {arriveTangent=2.9389, interpMode=RCIM_Cubic, leaveTangent=2.9389, tangentMode=RCTM_User, time=0.5, value=0.5}
         <at 1 1 RCIM_Cubic>
        ]}}}
     CurveIndex=&RandomRangeFloat(Minimum=0.5, Maximum=1)
    )) {isSetParameters=true, module=SetVariables_#3}
   &InitializeParticle(
    Color=$System.Water splash color
    Lifetime Max=0.5
    Lifetime Min=0.35
    Lifetime Mode=::NewEnumerator1
    Sprite Rotation Mode=::NewEnumerator2
    Sprite Size Mode=::NewEnumerator3
    Uniform Sprite Size=&Lerp_Float(
     A=&RandomRangeFloat(Minimum=25, Maximum=50)
     B=&RandomRangeFloat(Minimum=50, Maximum=80)
     Alpha=$Particles.Scaler
    )
    UsePositionOffset=false
    Write Material Random=true
   )
   &ShapeLocation(MaskSpawnGroup=false, Override Seed=false, Sphere Radius=20, UseOwnerScale=true)
   &AddVelocity(
    Cone Angle=60
    Cone Axis=<0 0 1>
    Module.UseConeFalloff=false
    Module.UseVelDistribution=false
    Override Seed=false
    Rotation Coordinate Space=::World
    Velocity Mode=::NewEnumerator2
    Velocity Speed=&RandomRangeFloat(Minimum=75, Maximum=150)
   )
   &SkeletalMeshLocation(
    MaskSpawnGroup=false
    Mesh Sampling Type=::NewEnumerator5
    Sampled Position Offset=<0 0 5>
    Skeletal Mesh=$System.Skeletal Mesh
    Socket Sampling Mode=::NewEnumerator5
   )
   &SubUVAnimation(
    "Sprite Renderer "=DataInterface() {propertyValues {SpriteRenderer=&self:NiagaraEmitter_1.NiagaraSpriteRendererProperties_1}}
    SubUV Animation Mode:ENiagara_SubUVLookupModeV2=::NewEnumerator2
    UseEndFrame=false
    UseStartFrame=false
   )
   &CameraOffset(Camera Offset Amount=&RandomRangeFloat(Minimum=0, Maximum=10))
  ]
  particleUpdateScript [
   &ParticleState()
   &ScaleSpriteSize(
    Scale Sprite Size Mode=::NewEnumerator0
    Uniform Scale Factor=&Lerp_Float(
     A=&FloatFromCurve(FloatCurve=DataInterface() {propertyValues {Curve {
         defaultValue=0
         keys [{interpMode=RCIM_Cubic, tangentMode=RCTM_User, time=0, value=0}, {interpMode=RCIM_Cubic, tangentMode=RCTM_User, time=1, value=1}]
        }}})
     B=&FloatFromCurve(FloatCurve=DataInterface() {propertyValues {Curve {
         defaultValue=0
         keys [{interpMode=RCIM_Cubic, tangentMode=RCTM_User, time=0, value=0}, {interpMode=RCIM_Cubic, tangentMode=RCTM_User, time=0.7817, value=0.9686}]
        }}})
     Alpha=$Particles.Scaler
    )
   ) {module=ScaleSpriteSize001}
   &SpriteRotationRate(Rotation Rate=&RandomRangeFloat(Minimum=-0.5, Maximum=0.5))
   &Drag(
    Drag=&FloatFromCurve(FloatCurve=DataInterface() {propertyValues {Curve {keys [{interpMode=RCIM_Cubic, tangentMode=RCTM_User, time=0, value=0}, {interpMode=RCIM_Cubic, tangentMode=RCTM_User, time=1, value=1}]}}})
    Use Linear Drag=true
    Use Rotational Drag=true
   )
   &DynamicMaterialParameters(
    Index 0 Param 1=&Lerp_Float(
     A=&FloatFromCurve(FloatCurve=DataInterface() {propertyValues {Curve {
         defaultValue=0
         keys [{interpMode=RCIM_Cubic, tangentMode=RCTM_User, time=0, value=1}, {interpMode=RCIM_Cubic, tangentMode=RCTM_User, time=0.4, value=0}]
        }}})
     B=&FloatFromCurve(FloatCurve=DataInterface() {propertyValues {Curve {
         defaultValue=0
         keys [{interpMode=RCIM_Cubic, tangentMode=RCTM_User, time=0, value=1}, {interpMode=RCIM_Cubic, tangentMode=RCTM_User, time=0.6, value=-0.1553}]
        }}})
     Alpha=$Particles.Scaler
    )
    Index 0 Param 2=0.5
    Module.Index0Param0Write=true
    Module.Index0Param1Write=true
    Module.Index0Param2Write=false
    Module.Index0Param3Write=false
   ) {enabled=false, module=DynamicMaterialParameters001}
   &ScaleColor(
    Scale Alpha=&FloatFromCurve(FloatCurve=DataInterface() {propertyValues {Curve {keys [
         {arriveTangent=6.8822, interpMode=RCIM_Cubic, leaveTangent=6.8822, tangentMode=RCTM_User, time=0, value=0}
         {interpMode=RCIM_Cubic, tangentMode=RCTM_User, time=0.2, value=1}
         {arriveTangent=-0.4524, tangentMode=RCTM_User, time=1, value=0.4}
        ]}}})
    Scale RGB=&VectorFromFloat(Value=&FNTimeOfDayScalar(Night Scale=3, Morning Scale=1, Day Scale=50, Evening Scale=1))
    ScaleA=true
    ScaleRGB=true
   )
   &GravityForce(Gravity=<0 0 -500>)
   &SolveForcesAndVelocity(Clamp Velocity=false, Limit Acceleration=false)
   &DynamicMaterialParameters(
    Index 0 Param 1=1
    Index 0 Param 2=1
    Index 0 Param 3=5
    Module.Index0Float0Write=true
    Module.Index0Float1Write=true
    Module.Index0Float2Write=true
    Module.Index0Float3Write=true
   ) {enabled=false}
   &DynamicMaterialParameters(
    Index 0 Param 1=&FloatFromCurve(FloatCurve=DataInterface() {propertyValues {Curve {keys [
         {interpMode=RCIM_Cubic, tangentMode=RCTM_User, time=0, value=1}
         {arriveTangent=-3.7889, interpMode=RCIM_Cubic, leaveTangent=-3.7889, tangentMode=RCTM_User, time=0.8, value=0}
        ]}}})
    Module.Param0WriteEnabled=true
    Module.Param1WriteEnabled=true
    Module.Param2WriteEnabled=true
    Module.Param3WriteEnabled=true
   ) {module=DynamicMaterialParameters002}
  ]
  rendererClasses [&NiagaraSpriteRendererProperties]
  renderers [{rendererClass=&NiagaraSpriteRendererProperties}]
 }
 SplashFat002 {
  bEnabled=false
  emitterUpdateScript [
   &EmitterState(Life Cycle Mode=::NewEnumerator1, Loop Behavior=::NewEnumerator1, Loop Delay=0.05, Loop Duration=3, UseLoopDelay=true)
   &SpawnRate(SpawnRate=100, Use Spawn Probability=false)
   &SpawnBurst_Instantaneous(Spawn Count=10, Use Spawn Probability=false, UseLoopCountLimit=false) {enabled=false}
  ]
  particleSpawnScript [
   /SetVariables(Particles.Scaler:NiagaraFloat=&FloatFromCurve(
     FloatCurve=DataInterface() {propertyValues {Curve {keys [
         <at 0 0 RCIM_Cubic>
         {arriveTangent=2.9389, interpMode=RCIM_Cubic, leaveTangent=2.9389, tangentMode=RCTM_User, time=0.5, value=0.5}
         <at 1 1 RCIM_Cubic>
        ]}}}
     CurveIndex=&RandomRangeFloat(Minimum=0, Maximum=1)
    )) {isSetParameters=true, module=SetVariables_#3}
   &InitializeParticle(
    Lifetime Max=0.55
    Lifetime Min=0.35
    Lifetime Mode=::NewEnumerator1
    Position Offset=<0 0 10>
    Sprite Rotation Mode=::NewEnumerator2
    Sprite Size Mode=::NewEnumerator3
    Uniform Sprite Size=&Lerp_Float(
     A=&RandomRangeFloat(Minimum=25, Maximum=45)
     B=&RandomRangeFloat(Minimum=35, Maximum=60)
     Alpha=$Particles.Scaler
    )
    UsePositionOffset=false
    Write Material Random=true
   )
   &AddVelocity(
    Cone Angle=60
    Cone Axis=<-0.2 0 1>
    Module.UseConeFalloff=false
    Module.UseVelDistribution=false
    Override Seed=false
    Rotation Coordinate Space=::World
    Velocity Mode=::NewEnumerator2
    Velocity Speed=&RandomRangeFloat(Minimum=75, Maximum=150)
   )
   &SkeletalMeshLocation(
    MaskSpawnGroup=false
    Mesh Sampling Type=::NewEnumerator5
    Orientation Sampling=::NewEnumerator1
    Sampled Position Offset=<-8 5 0>
    Skeletal Mesh=$System.Skeletal Mesh
    Socket Sampling Mode=::NewEnumerator5
   ) {module=SkeletalMeshLocation001}
   &CameraOffset(Camera Offset Amount=15)
   &ShapeLocation(
    MaskSpawnGroup=false
    Non Uniform Scale=<1 1 0.3>
    Offset=<0 0 -2>
    Offset Mode=::NewEnumerator0
    Override Seed=false
    Sphere Radius=15
    Sphere Surface Distribution=1
    UseOwnerScale=true
   )
  ]
  particleUpdateScript [
   &ParticleState()
   &ScaleSpriteSize(
    Scale Sprite Size Mode=::NewEnumerator0
    Uniform Scale Factor=&Lerp_Float(
     A=&FloatFromCurve(FloatCurve=DataInterface() {propertyValues {Curve {keys [
          {arriveTangent=2.9966, interpMode=RCIM_Cubic, leaveTangent=2.9966, tangentMode=RCTM_User, time=0, value=0}
          {arriveTangent=1, interpMode=RCIM_Cubic, leaveTangent=1, time=0.3, value=0.6}
          {arriveTangent=0.5346, interpMode=RCIM_Cubic, leaveTangent=0.5346, tangentMode=RCTM_User, time=1, value=1}
         ]}}})
     B=&FloatFromCurve(FloatCurve=DataInterface() {propertyValues {Curve {keys [
          {arriveTangent=1.5909, interpMode=RCIM_Cubic, leaveTangent=1.5909, tangentMode=RCTM_User, time=0, value=0}
          {arriveTangent=0.5346, interpMode=RCIM_Cubic, leaveTangent=0.5346, tangentMode=RCTM_User, time=1, value=1}
         ]}}})
     Alpha=$Particles.Scaler
    )
   ) {module=ScaleSpriteSize001}
   &SpriteRotationRate(Rotation Rate=&RandomRangeFloat(Minimum=-0.5, Maximum=0.5))
   &Drag(
    Drag=&FloatFromCurve(FloatCurve=DataInterface() {propertyValues {Curve {keys [{interpMode=RCIM_Cubic, tangentMode=RCTM_User, time=0, value=0}, {interpMode=RCIM_Cubic, tangentMode=RCTM_User, time=1, value=1}]}}})
    Use Linear Drag=true
    Use Rotational Drag=true
   )
   &ScaleColor(
    Scale Alpha=&FloatFromCurve(FloatCurve=DataInterface() {propertyValues {Curve {keys [
         {arriveTangent=6.8822, interpMode=RCIM_Cubic, leaveTangent=6.8822, tangentMode=RCTM_User, time=0, value=0}
         {interpMode=RCIM_Cubic, tangentMode=RCTM_User, time=0.105, value=0.9906}
         {arriveTangent=-0.4524, tangentMode=RCTM_User, time=1, value=0.4}
        ]}}})
    Scale RGB=&VectorFromFloat(Value=&FNTimeOfDayScalar(Night Scale=3, Morning Scale=2.5, Day Scale=1.2, Evening Scale=3))
    ScaleA=true
    ScaleRGB=true
   )
   &GravityForce(Gravity=<0 0 -500>)
   &SolveForcesAndVelocity(Clamp Velocity=false, Limit Acceleration=false)
   &DynamicMaterialParameters(
    Index 0 Param 1=&Lerp_Float(
     A=&FloatFromCurve(FloatCurve=DataInterface() {propertyValues {Curve {keys [
          {interpMode=RCIM_Cubic, leaveTangentWeight=0.4309, tangentMode=RCTM_User, time=0, value=0}
          {arriveTangent=4.2419, tangentMode=RCTM_User, time=1, value=10}
         ]}}})
     B=&FloatFromCurve(FloatCurve=DataInterface() {propertyValues {Curve {keys [
          {interpMode=RCIM_Cubic, leaveTangentWeight=0.4309, tangentMode=RCTM_User, time=0, value=0}
          {arriveTangent=25.6845, tangentMode=RCTM_User, time=1, value=10}
         ]}}})
     Alpha=$Particles.Scaler
    )
    Index 0 Param 2=1
    Module.Index0Param0Write=true
    Module.Index0Param1Write=true
    Module.Index0Param2Write=false
    Module.Index0Param3Write=false
   ) {enabled=false, module=DynamicMaterialParameters001}
   &DynamicMaterialParameters(
    Index 0 Param 1=&FloatFromCurve(FloatCurve=DataInterface() {propertyValues {Curve {keys [
         {interpMode=RCIM_Cubic, tangentMode=RCTM_User, time=0, value=1}
         {arriveTangent=-3.7889, interpMode=RCIM_Cubic, leaveTangent=-3.7889, tangentMode=RCTM_User, time=0.8, value=0}
        ]}}})
    Module.Param0WriteEnabled=true
    Module.Param1WriteEnabled=true
    Module.Param2WriteEnabled=true
    Module.Param3WriteEnabled=true
   ) {module=DynamicMaterialParameters002}
  ]
  rendererClasses [&NiagaraSpriteRendererProperties]
  renderers [{rendererClass=&NiagaraSpriteRendererProperties}]
 }
 Splashes_compressed {
  simTarget=GPUComputeSim
  emitterUpdateScript [
   &EmitterState(Life Cycle Mode=::NewEnumerator1, Loop Behavior=::NewEnumerator1, Loop Duration=3, UseLoopDelay=false)
   &SpawnRate(SpawnGroup=3, SpawnRate=30, Use Spawn Probability=false) {module=SpawnRate003}
   &SpawnRate(SpawnGroup=2, SpawnRate=100, Use Spawn Probability=false) {module=SpawnRate002}
   &SpawnRate(SpawnGroup=1, SpawnRate=150, Use Spawn Probability=false) {module=SpawnRate001}
   &SpawnRate(SpawnRate=100, Use Spawn Probability=false)
  ]
  particleSpawnScript [
   /SetVariables(Particles.VisibilityTag=&SpawnGroup()) {isSetParameters=true, module=SetVariables_#4}
   /SetVariables(Particles.Scaler:NiagaraFloat=&FloatFromCurve(
     FloatCurve=DataInterface() {propertyValues {Curve {keys [
         <at 0 0 RCIM_Cubic>
         {arriveTangent=2.9389, interpMode=RCIM_Cubic, leaveTangent=2.9389, tangentMode=RCTM_User, time=0.5, value=0.5}
         <at 1 1 RCIM_Cubic>
        ]}}}
     CurveIndex=&RandomRangeFloat(Minimum=0, Maximum=1)
    )) {isSetParameters=true, module=SetVariables_#5}
   &InitializeParticle(
    Color=&MergeHelper_ColorSwitch(
     0=<rgba 1 1 1 1>
     1=$System.Water splash color
     2=$System.Water splash color
     3=$System.Water splash color
     4=$System.DropletColors
     Index=&SpawnGroup()
     Num=::NewEnumerator3
    )
    Lifetime=&MergeHelper_FloatSwitch(
     0=&RandomRangeFloat(Minimum=0.35, Maximum=0.55)
     1=&RandomRangeFloat(Minimum=0.2, Maximum=0.55)
     2=&RandomRangeFloat(Minimum=0.2, Maximum=0.55)
     3=&RandomRangeFloat(Minimum=0.35, Maximum=0.5)
     4=&RandomRangeFloat(Minimum=0, Maximum=1)
     Index=&SpawnGroup()
     Num=::NewEnumerator3
    )
    Sprite Rotation Angle=&MergeHelper_FloatSwitch(
     0=&RandomRangeFloat(Minimum=0, Maximum=1)
     1=&RandomRangeFloat(Minimum=0, Maximum=0.2)
     2=&RandomRangeFloat(Minimum=0, Maximum=0.5)
     3=&RandomRangeFloat(Minimum=0, Maximum=1)
     4=0
     Index=&SpawnGroup()
     Num=::NewEnumerator3
    )
    Sprite Rotation Mode=::NewEnumerator1
    Sprite Size Mode=::NewEnumerator3
    Uniform Sprite Size=&MergeHelper_FloatSwitch(
     0=&Lerp_Float(
      A=&RandomRangeFloat(Minimum=25, Maximum=45)
      B=&RandomRangeFloat(Minimum=35, Maximum=60)
      Alpha=$Particles.Scaler
     )
     1=&RandomRangeFloat(Minimum=20, Maximum=35)
     2=&RandomRangeFloat(Minimum=22, Maximum=37)
     3=&Lerp_Float(
      A=&RandomRangeFloat(Minimum=25, Maximum=50)
      B=&RandomRangeFloat(Minimum=50, Maximum=80)
      Alpha=$Particles.Scaler
     )
     4=0
     Index=&SpawnGroup()
     Num=::NewEnumerator3
    )
    UsePositionOffset=false
    Write Material Random=true
   )
   &AddVelocity(Velocity=&MergeHelper_VectorSwitch(
     0=&RandomRangeVector(Minimum=<-75 -75 75>, Maximum=<75 75 150>)
     1=&RandomRangeVector(Minimum=<-75 -75 75>, Maximum=<75 75 250>)
     2=&RandomRangeVector(Minimum=<-75 -75 75>, Maximum=<75 75 250>)
     3=&RandomRangeVector(Minimum=<-75 -75 75>, Maximum=<75 75 150>)
     4=<0 0 0>
     Index=&SpawnGroup()
     Num=::NewEnumerator3
    ))
   &CameraOffset(Camera Offset Amount=&MergeHelper_FloatSwitch(
     0=15
     1=&RandomRangeFloat(Minimum=6, Maximum=15)
     2=&RandomRangeFloat(Minimum=6, Maximum=15)
     3=&RandomRangeFloat(Minimum=0, Maximum=10)
     4=0
     Index=&SpawnGroup()
     Num=::NewEnumerator3
    ))
   &SkeletalMeshLocation(
    MaskSpawnGroup=false
    Mesh Sampling Type=::NewEnumerator5
    Orientation Sampling=::NewEnumerator1
    Sampled Position Offset=&MergeHelper_VectorSwitch(0=<-8 5 0>, 1=<-5 0 0>, 2=<-5 0 15>, 3=<0 0 5>, 4=<0 0 0>, Index=&SpawnGroup(), Num=::NewEnumerator3)
    Skeletal Mesh=$System.Skeletal Mesh
    Socket Sampling Mode=::NewEnumerator5
   )
   &ShapeLocation(
    MaskSpawnGroup=false
    Non Uniform Scale=&MergeHelper_VectorSwitch(0=<1 1 0.3>, 1=<1 1 0.1>, 2=<1 1 0.1>, 3=<0 0 0>, 4=<0 0 0>, Index=&SpawnGroup(), Num=::NewEnumerator3)
    Offset=&MergeHelper_VectorSwitch(0=<0 0 -2>, 1=<0 0 0>, 2=<0 0 0>, 3=<0 0 0>, 4=<0 0 0>, Index=&SpawnGroup(), Num=::NewEnumerator3)
    Offset Mode=::NewEnumerator0
    Override Seed=false
    Sphere Radius=&MergeHelper_FloatSwitch(0=15, 1=12, 2=15, 3=20, 4=0, Index=&SpawnGroup(), Num=::NewEnumerator3)
    UseOwnerScale=true
   )
   &SubUVAnimation(
    End Frame Range Override=3
    Evaluation Type=::NewEnumerator0
    "Sprite Renderer "=DataInterface() {propertyValues {SpriteRenderer=&self:Splashes_compressed_0.NiagaraSpriteRendererProperties_3}}
    SubUV Animation Mode:ENiagara_SubUVLookupModeV2=::NewEnumerator2
    UseEndFrame=true
    UseStartFrame=true
   )
   &SubUVAnimation(
    End Frame Range Override=3
    Evaluation Type=::NewEnumerator0
    "Sprite Renderer "=DataInterface() {propertyValues {SpriteRenderer=&self:Splashes_compressed_0.NiagaraSpriteRendererProperties_4}}
    SubUV Animation Mode:ENiagara_SubUVLookupModeV2=::NewEnumerator2
    UseEndFrame=true
    UseStartFrame=true
   ) {module=SubUVAnimation001}
   &SubUVAnimation(
    "Sprite Renderer "=DataInterface() {propertyValues {SpriteRenderer=&self:Splashes_compressed_0.NiagaraSpriteRendererProperties_5}}
    SubUV Animation Mode:ENiagara_SubUVLookupModeV2=::NewEnumerator2
    UseEndFrame=false
    UseStartFrame=false
   ) {module=SubUVAnimation002}
  ]
  particleUpdateScript [
   &ParticleState()
   &ScaleSpriteSize(
    Scale Sprite Size Mode=::NewEnumerator0
    Uniform Scale Factor=&MergeHelper_FloatSwitch(
     0=&Lerp_Float(
      A=&FloatFromCurve(FloatCurve=DataInterface() {propertyValues {Curve {keys [
           {arriveTangent=2.9966, interpMode=RCIM_Cubic, leaveTangent=2.9966, tangentMode=RCTM_User, time=0, value=0}
           {arriveTangent=1, interpMode=RCIM_Cubic, leaveTangent=1, time=0.3, value=0.6}
           {arriveTangent=0.5346, interpMode=RCIM_Cubic, leaveTangent=0.5346, tangentMode=RCTM_User, time=1, value=1}
          ]}}})
      B=&FloatFromCurve(FloatCurve=DataInterface() {propertyValues {Curve {keys [
           {arriveTangent=1.5909, interpMode=RCIM_Cubic, leaveTangent=1.5909, tangentMode=RCTM_User, time=0, value=0}
           {arriveTangent=0.5346, interpMode=RCIM_Cubic, leaveTangent=0.5346, tangentMode=RCTM_User, time=1, value=1}
          ]}}})
      Alpha=$Particles.Scaler
     )
     1=&FloatFromCurve(FloatCurve=DataInterface() {propertyValues {Curve {keys [
          {arriveTangent=4.1785, interpMode=RCIM_Cubic, leaveTangent=4.1785, tangentMode=RCTM_User, time=0, value=0.15}
          {arriveTangent=1.1179, interpMode=RCIM_Cubic, leaveTangent=1.1179, tangentMode=RCTM_User, time=0.2042, value=0.6228}
          {arriveTangent=0.0224, interpMode=RCIM_Cubic, leaveTangent=0.0224, tangentMode=RCTM_User, time=1, value=1}
         ]}}})
     2=&FloatFromCurve(FloatCurve=DataInterface() {propertyValues {Curve {keys [
          {arriveTangent=4.1785, interpMode=RCIM_Cubic, leaveTangent=4.1785, tangentMode=RCTM_User, time=0, value=0.15}
          {arriveTangent=1.1179, interpMode=RCIM_Cubic, leaveTangent=1.1179, tangentMode=RCTM_User, time=0.2042, value=0.6228}
          {arriveTangent=0.0224, interpMode=RCIM_Cubic, leaveTangent=0.0224, tangentMode=RCTM_User, time=1, value=1}
         ]}}})
     3=&Lerp_Float(
      A=&FloatFromCurve(FloatCurve=DataInterface() {propertyValues {Curve {
          defaultValue=0
          keys [{interpMode=RCIM_Cubic, tangentMode=RCTM_User, time=0, value=0}, {interpMode=RCIM_Cubic, tangentMode=RCTM_User, time=1, value=1}]
         }}})
      B=&FloatFromCurve(FloatCurve=DataInterface() {propertyValues {Curve {
          defaultValue=0
          keys [{interpMode=RCIM_Cubic, tangentMode=RCTM_User, time=0, value=0}, {interpMode=RCIM_Cubic, tangentMode=RCTM_User, time=0.7817, value=0.9686}]
         }}})
      Alpha=$Particles.Scaler
     )
     4=0
     Index=&SpawnGroup()
     Num=::NewEnumerator3
    )
   )
   &SpriteRotationRate(Rotation Rate=&MergeHelper_FloatSwitch(
     0=&RandomRangeFloat(Minimum=-0.5, Maximum=0.5)
     1=&RandomRangeFloat(Minimum=-70, Maximum=-100)
     2=&RandomRangeFloat(Minimum=-50, Maximum=-150)
     3=&RandomRangeFloat(Minimum=-0.5, Maximum=0.5)
     4=0
     Index=&SpawnGroup()
     Num=::NewEnumerator3
    ))
   &ScaleColor(
    Scale Alpha=&MergeHelper_FloatSwitch(
     0=&FloatFromCurve(FloatCurve=DataInterface() {propertyValues {Curve {keys [
          {arriveTangent=6.8822, interpMode=RCIM_Cubic, leaveTangent=6.8822, tangentMode=RCTM_User, time=0, value=0}
          {interpMode=RCIM_Cubic, tangentMode=RCTM_User, time=0.105, value=0.9906}
          {arriveTangent=-0.4524, tangentMode=RCTM_User, time=1, value=0.4}
         ]}}})
     1=&FloatFromCurve(
      FloatCurve=DataInterface() {propertyValues {Curve {keys [
          {arriveTangent=-0.2723, interpMode=RCIM_Cubic, leaveTangent=14.5791, tangentMode=RCTM_Break, time=0, value=0}
          {interpMode=RCIM_Cubic, tangentMode=RCTM_Break, time=0.1, value=1}
          {arriveTangent=0.0378, interpMode=RCIM_Cubic, leaveTangent=0.0378, tangentMode=RCTM_User, time=0.6785, value=1.0282}
          {arriveTangent=-3.7547, interpMode=RCIM_Cubic, leaveTangent=-3.7547, tangentMode=RCTM_Break, time=0.8, value=0}
         ]}}}
      Scale Curve=&FNTimeOfDayScalar(Night Scale=1, Morning Scale=0.7, Day Scale=2, Evening Scale=0.5)
     )
     2=&FloatFromCurve(
      FloatCurve=DataInterface() {propertyValues {Curve {keys [
          {arriveTangent=-0.2723, interpMode=RCIM_Cubic, leaveTangent=14.5791, tangentMode=RCTM_Break, time=0, value=0}
          {interpMode=RCIM_Cubic, tangentMode=RCTM_Break, time=0.1, value=1}
          {arriveTangent=0.0378, interpMode=RCIM_Cubic, leaveTangent=0.0378, tangentMode=RCTM_User, time=0.6785, value=1.0282}
          {arriveTangent=-3.7547, interpMode=RCIM_Cubic, leaveTangent=-3.7547, tangentMode=RCTM_Break, time=0.8, value=0}
         ]}}}
      Scale Curve=&FNTimeOfDayScalar(Night Scale=1, Morning Scale=0.7, Day Scale=2, Evening Scale=0.5)
     )
     3=&FloatFromCurve(FloatCurve=DataInterface() {propertyValues {Curve {keys [
          {arriveTangent=6.8822, interpMode=RCIM_Cubic, leaveTangent=6.8822, tangentMode=RCTM_User, time=0, value=0}
          {interpMode=RCIM_Cubic, tangentMode=RCTM_User, time=0.2, value=1}
          {arriveTangent=-0.4524, tangentMode=RCTM_User, time=1, value=0.4}
         ]}}})
     4=&FloatFromCurve(FloatCurve=DataInterface() {propertyValues {Curve {keys [
          {arriveTangent=6.8822, interpMode=RCIM_Cubic, leaveTangent=6.8822, tangentMode=RCTM_User, time=0, value=0}
          {interpMode=RCIM_Cubic, tangentMode=RCTM_User, time=0.2, value=1}
          {arriveTangent=-0.4524, tangentMode=RCTM_User, time=1, value=0.4}
         ]}}})
     Index=&SpawnGroup()
     Num=::NewEnumerator2
    )
    Scale RGB=&MergeHelper_VectorSwitch(
     0=&VectorFromFloat(Value=&FNTimeOfDayScalar(Night Scale=3, Morning Scale=2.5, Day Scale=1.2, Evening Scale=3))
     1=&VectorFromFloat(Value=&FNTimeOfDayScalar(Night Scale=1.5, Morning Scale=2, Day Scale=1.2, Evening Scale=2))
     2=&VectorFromFloat(Value=&FNTimeOfDayScalar(Night Scale=1.5, Morning Scale=2, Day Scale=1.2, Evening Scale=2))
     3=&VectorFromFloat(Value=&FNTimeOfDayScalar(Night Scale=3, Morning Scale=1, Day Scale=50, Evening Scale=1))
     4=<1 1 1>
     Index=&SpawnGroup()
     Num=::NewEnumerator3
    )
    ScaleA=true
    ScaleRGB=true
   )
   &GravityForce(Gravity=&MergeHelper_VectorSwitch(
     0=<0 0 -500>
     1=&RandomRangeVector(Minimum=<0 0 -500>, Maximum=<1 1 -980>)
     2=&RandomRangeVector(Minimum=<0 0 -500>, Maximum=<0 0 -980>)
     3=<0 0 -500>
     4=<0 0 0>
     Index=&SpawnGroup()
     Num=::NewEnumerator3
    ))
   &Drag(
    Drag=&MergeHelper_FloatSwitch(
     0=0
     1=&RandomRangeFloat(Minimum=0.2, Maximum=0.3)
     2=&RandomRangeFloat(Minimum=0.2, Maximum=0.4)
     3=&FloatFromCurve(FloatCurve=DataInterface() {propertyValues {Curve {keys [{interpMode=RCIM_Cubic, tangentMode=RCTM_User, time=0, value=0}, {interpMode=RCIM_Cubic, tangentMode=RCTM_User, time=1, value=1}]}}})
     4=0
     Index=&SpawnGroup()
     Num=::NewEnumerator3
    )
    Rotational Drag:NiagaraFloat=&MergeHelper_FloatSwitch(0=1, 1=1, 2=1, 3=1, 4=1, Index=&SpawnGroup(), Num=::NewEnumerator3)
    Use Linear Drag=true
    Use Rotational Drag=true
   )
   &SolveForcesAndVelocity(Clamp Velocity=false, Limit Acceleration=false)
   &ScaleSpriteSizeBySpeed(
    Max Scale Factor=&MergeHelper_Vector2Switch(0=<1 1>, 1=<2 2>, 2=<2 2>, 3=<1.5 1.5>, 4=<2 2>, Index=&SpawnGroup(), Num=::NewEnumerator3)
    Min Scale Factor=&MergeHelper_Vector2Switch(0=<1 1>, 1=<1 1>, 2=<1 1>, 3=<1 1>, 4=<1 1>, Index=&SpawnGroup(), Num=::NewEnumerator3)
    Sample Scale Factor By Curve=true
    Scale Factor Curve=DataInterface() {propertyValues {Curve {defaultValue=0.5941, keys [{tangentMode=RCTM_User, time=0, value=1}, {tangentMode=RCTM_User, time=1, value=0}]}}}
    Velocity Threshold=250
   )
   /SetVariables(
    Particles.0 - Fat splash DynParam:Vector4f=&MakeVector4(
     X=&FloatFromCurve(FloatCurve=DataInterface() {propertyValues {Curve {keys [
          {interpMode=RCIM_Cubic, tangentMode=RCTM_User, time=0, value=1}
          {arriveTangent=-3.7889, interpMode=RCIM_Cubic, leaveTangent=-3.7889, tangentMode=RCTM_User, time=0.8, value=0}
         ]}}})
     Y=0
     Z=0
     W=0
    )
    Particles.1 - Thick Splash DynParam:Vector4f=&MakeVector4(
     X=&FloatFromCurve(FloatCurve=DataInterface() {propertyValues {Curve {defaultValue=0, keys [{tangentMode=RCTM_User, time=0, value=1}, {tangentMode=RCTM_User, time=1, value=0}]}}})
     Y=0
     Z=0
     W=0
    )
    Particles.2 - Thick Splash Bottom DynParam:Vector4f=&MakeVector4(
     X=&FloatFromCurve(FloatCurve=DataInterface() {propertyValues {Curve {defaultValue=0, keys [{tangentMode=RCTM_User, time=0, value=1}, {tangentMode=RCTM_User, time=1, value=0}]}}})
     Y=0
     Z=0
     W=0
    )
    Particles.3 - Super Thick Splash DynParam:Vector4f=&MakeVector4(
     X=&FloatFromCurve(FloatCurve=DataInterface() {propertyValues {Curve {keys [
          {interpMode=RCIM_Cubic, tangentMode=RCTM_User, time=0, value=1}
          {arriveTangent=-3.7889, interpMode=RCIM_Cubic, leaveTangent=-3.7889, tangentMode=RCTM_User, time=0.8, value=0}
         ]}}})
     Y=0
     Z=0
     W=0
    )
   ) {isSetParameters=true, module=SetVariables_#6}
  ]
  rendererClasses [&NiagaraSpriteRendererProperties]
  renderers [
   {rendererClass=&NiagaraSpriteRendererProperties}
   {rendererClass=&NiagaraSpriteRendererProperties, rendererIndex=1}
   {rendererClass=&NiagaraSpriteRendererProperties, rendererIndex=2}
   {rendererClass=&NiagaraSpriteRendererProperties, rendererIndex=3}
  ]
 }
 Splashes_compressed001 {
  emitterUpdateScript [
   &EmitterState(Life Cycle Mode=::NewEnumerator1, Loop Behavior=::NewEnumerator1, Loop Duration=3, UseLoopDelay=false)
   &SpawnRate(SpawnGroup=3, SpawnRate=30, Use Spawn Probability=false) {module=SpawnRate003}
   &SpawnRate(SpawnGroup=2, SpawnRate=100, Use Spawn Probability=false) {module=SpawnRate002}
   &SpawnRate(SpawnGroup=1, SpawnRate=150, Use Spawn Probability=false) {module=SpawnRate001}
   &SpawnRate(SpawnRate=100, Use Spawn Probability=false)
  ]
  particleSpawnScript [
   /SetVariables(Particles.VisibilityTag=&SpawnGroup()) {isSetParameters=true, module=SetVariables_#4}
   /SetVariables(Particles.Scaler:NiagaraFloat=&FloatFromCurve(
     FloatCurve=DataInterface() {propertyValues {Curve {keys [
         <at 0 0 RCIM_Cubic>
         {arriveTangent=2.9389, interpMode=RCIM_Cubic, leaveTangent=2.9389, tangentMode=RCTM_User, time=0.5, value=0.5}
         <at 1 1 RCIM_Cubic>
        ]}}}
     CurveIndex=&RandomRangeFloat(Minimum=0, Maximum=1)
    )) {isSetParameters=true, module=SetVariables_#5}
   &InitializeParticle(
    Color=&MergeHelper_ColorSwitch(
     0=<rgba 1 1 1 1>
     1=$System.Water splash color
     2=$System.Water splash color
     3=$System.Water splash color
     4=$System.DropletColors
     Index=&SpawnGroup()
     Num=::NewEnumerator3
    )
    Lifetime=&MergeHelper_FloatSwitch(
     0=&RandomRangeFloat(Minimum=0.35, Maximum=0.55)
     1=&RandomRangeFloat(Minimum=0.2, Maximum=0.55)
     2=&RandomRangeFloat(Minimum=0.2, Maximum=0.55)
     3=&RandomRangeFloat(Minimum=0.35, Maximum=0.5)
     4=&RandomRangeFloat(Minimum=0, Maximum=1)
     Index=&SpawnGroup()
     Num=::NewEnumerator3
    )
    Sprite Rotation Angle=&MergeHelper_FloatSwitch(
     0=&RandomRangeFloat(Minimum=0, Maximum=1)
     1=&RandomRangeFloat(Minimum=0, Maximum=0.2)
     2=&RandomRangeFloat(Minimum=0, Maximum=0.5)
     3=&RandomRangeFloat(Minimum=0, Maximum=1)
     4=0
     Index=&SpawnGroup()
     Num=::NewEnumerator3
    )
    Sprite Rotation Mode=::NewEnumerator1
    Sprite Size Mode=::NewEnumerator3
    Uniform Sprite Size=&MergeHelper_FloatSwitch(
     0=&Lerp_Float(
      A=&RandomRangeFloat(Minimum=25, Maximum=45)
      B=&RandomRangeFloat(Minimum=35, Maximum=60)
      Alpha=$Particles.Scaler
     )
     1=&RandomRangeFloat(Minimum=20, Maximum=35)
     2=&RandomRangeFloat(Minimum=22, Maximum=37)
     3=&Lerp_Float(
      A=&RandomRangeFloat(Minimum=25, Maximum=50)
      B=&RandomRangeFloat(Minimum=50, Maximum=80)
      Alpha=$Particles.Scaler
     )
     4=0
     Index=&SpawnGroup()
     Num=::NewEnumerator3
    )
    UsePositionOffset=false
    Write Material Random=true
   )
   &AddVelocity(Velocity=&MergeHelper_VectorSwitch(
     0=&RandomRangeVector(Minimum=<-75 -75 75>, Maximum=<75 75 150>)
     1=&RandomRangeVector(Minimum=<-75 -75 75>, Maximum=<75 75 250>)
     2=&RandomRangeVector(Minimum=<-75 -75 75>, Maximum=<75 75 250>)
     3=&RandomRangeVector(Minimum=<-75 -75 75>, Maximum=<75 75 150>)
     4=<0 0 0>
     Index=&SpawnGroup()
     Num=::NewEnumerator3
    ))
   &CameraOffset(Camera Offset Amount=&MergeHelper_FloatSwitch(
     0=15
     1=&RandomRangeFloat(Minimum=6, Maximum=15)
     2=&RandomRangeFloat(Minimum=6, Maximum=15)
     3=&RandomRangeFloat(Minimum=0, Maximum=10)
     4=0
     Index=&SpawnGroup()
     Num=::NewEnumerator3
    ))
   &SkeletalMeshLocation(
    MaskSpawnGroup=false
    Mesh Sampling Type=::NewEnumerator5
    Orientation Sampling=::NewEnumerator1
    Sampled Position Offset=&MergeHelper_VectorSwitch(0=<-8 5 0>, 1=<-5 0 0>, 2=<-5 0 15>, 3=<0 0 5>, 4=<0 0 0>, Index=&SpawnGroup(), Num=::NewEnumerator3)
    Skeletal Mesh=$System.Skeletal Mesh
    Socket Sampling Mode=::NewEnumerator5
   )
   &ShapeLocation(
    MaskSpawnGroup=false
    Non Uniform Scale=&MergeHelper_VectorSwitch(0=<1 1 0.3>, 1=<1 1 0.1>, 2=<1 1 0.1>, 3=<0 0 0>, 4=<0 0 0>, Index=&SpawnGroup(), Num=::NewEnumerator3)
    Offset=&MergeHelper_VectorSwitch(0=<0 0 -2>, 1=<0 0 0>, 2=<0 0 0>, 3=<0 0 0>, 4=<0 0 0>, Index=&SpawnGroup(), Num=::NewEnumerator3)
    Offset Mode=::NewEnumerator0
    Override Seed=false
    Sphere Radius=&MergeHelper_FloatSwitch(0=15, 1=12, 2=15, 3=20, 4=0, Index=&SpawnGroup(), Num=::NewEnumerator3)
    UseOwnerScale=true
   )
   &SubUVAnimation(
    End Frame Range Override=3
    Evaluation Type=::NewEnumerator0
    "Sprite Renderer "=DataInterface() {propertyValues {SpriteRenderer=&self:Splashes_compressed001_0.NiagaraSpriteRendererProperties_3}}
    SubUV Animation Mode:ENiagara_SubUVLookupModeV2=::NewEnumerator2
    UseEndFrame=true
    UseStartFrame=true
   )
   &SubUVAnimation(
    End Frame Range Override=3
    Evaluation Type=::NewEnumerator0
    "Sprite Renderer "=DataInterface() {propertyValues {SpriteRenderer=&self:Splashes_compressed001_0.NiagaraSpriteRendererProperties_4}}
    SubUV Animation Mode:ENiagara_SubUVLookupModeV2=::NewEnumerator2
    UseEndFrame=true
    UseStartFrame=true
   ) {module=SubUVAnimation001}
   &SubUVAnimation(
    "Sprite Renderer "=DataInterface() {propertyValues {SpriteRenderer=&self:Splashes_compressed001_0.NiagaraSpriteRendererProperties_5}}
    SubUV Animation Mode:ENiagara_SubUVLookupModeV2=::NewEnumerator2
    UseEndFrame=false
    UseStartFrame=false
   ) {module=SubUVAnimation002}
  ]
  particleUpdateScript [
   &ParticleState()
   &ScaleSpriteSize(
    Scale Sprite Size Mode=::NewEnumerator0
    Uniform Scale Factor=&MergeHelper_FloatSwitch(
     0=&Lerp_Float(
      A=&FloatFromCurve(FloatCurve=DataInterface() {propertyValues {Curve {keys [
           {arriveTangent=2.9966, interpMode=RCIM_Cubic, leaveTangent=2.9966, tangentMode=RCTM_User, time=0, value=0}
           {arriveTangent=1, interpMode=RCIM_Cubic, leaveTangent=1, time=0.3, value=0.6}
           {arriveTangent=0.5346, interpMode=RCIM_Cubic, leaveTangent=0.5346, tangentMode=RCTM_User, time=1, value=1}
          ]}}})
      B=&FloatFromCurve(FloatCurve=DataInterface() {propertyValues {Curve {keys [
           {arriveTangent=1.5909, interpMode=RCIM_Cubic, leaveTangent=1.5909, tangentMode=RCTM_User, time=0, value=0}
           {arriveTangent=0.5346, interpMode=RCIM_Cubic, leaveTangent=0.5346, tangentMode=RCTM_User, time=1, value=1}
          ]}}})
      Alpha=$Particles.Scaler
     )
     1=&FloatFromCurve(FloatCurve=DataInterface() {propertyValues {Curve {keys [
          {arriveTangent=4.1785, interpMode=RCIM_Cubic, leaveTangent=4.1785, tangentMode=RCTM_User, time=0, value=0.15}
          {arriveTangent=1.1179, interpMode=RCIM_Cubic, leaveTangent=1.1179, tangentMode=RCTM_User, time=0.2042, value=0.6228}
          {arriveTangent=0.0224, interpMode=RCIM_Cubic, leaveTangent=0.0224, tangentMode=RCTM_User, time=1, value=1}
         ]}}})
     2=&FloatFromCurve(FloatCurve=DataInterface() {propertyValues {Curve {keys [
          {arriveTangent=4.1785, interpMode=RCIM_Cubic, leaveTangent=4.1785, tangentMode=RCTM_User, time=0, value=0.15}
          {arriveTangent=1.1179, interpMode=RCIM_Cubic, leaveTangent=1.1179, tangentMode=RCTM_User, time=0.2042, value=0.6228}
          {arriveTangent=0.0224, interpMode=RCIM_Cubic, leaveTangent=0.0224, tangentMode=RCTM_User, time=1, value=1}
         ]}}})
     3=&Lerp_Float(
      A=&FloatFromCurve(FloatCurve=DataInterface() {propertyValues {Curve {
          defaultValue=0
          keys [{interpMode=RCIM_Cubic, tangentMode=RCTM_User, time=0, value=0}, {interpMode=RCIM_Cubic, tangentMode=RCTM_User, time=1, value=1}]
         }}})
      B=&FloatFromCurve(FloatCurve=DataInterface() {propertyValues {Curve {
          defaultValue=0
          keys [{interpMode=RCIM_Cubic, tangentMode=RCTM_User, time=0, value=0}, {interpMode=RCIM_Cubic, tangentMode=RCTM_User, time=0.7817, value=0.9686}]
         }}})
      Alpha=$Particles.Scaler
     )
     4=0
     Index=&SpawnGroup()
     Num=::NewEnumerator3
    )
   )
   &SpriteRotationRate(Rotation Rate=&MergeHelper_FloatSwitch(
     0=&RandomRangeFloat(Minimum=-0.5, Maximum=0.5)
     1=&RandomRangeFloat(Minimum=-70, Maximum=-100)
     2=&RandomRangeFloat(Minimum=-50, Maximum=-150)
     3=&RandomRangeFloat(Minimum=-0.5, Maximum=0.5)
     4=0
     Index=&SpawnGroup()
     Num=::NewEnumerator3
    ))
   &ScaleColor(
    Scale Alpha=&MergeHelper_FloatSwitch(
     0=&FloatFromCurve(FloatCurve=DataInterface() {propertyValues {Curve {keys [
          {arriveTangent=6.8822, interpMode=RCIM_Cubic, leaveTangent=6.8822, tangentMode=RCTM_User, time=0, value=0}
          {interpMode=RCIM_Cubic, tangentMode=RCTM_User, time=0.105, value=0.9906}
          {arriveTangent=-0.4524, tangentMode=RCTM_User, time=1, value=0.4}
         ]}}})
     1=&FloatFromCurve(
      FloatCurve=DataInterface() {propertyValues {Curve {keys [
          {arriveTangent=-0.2723, interpMode=RCIM_Cubic, leaveTangent=14.5791, tangentMode=RCTM_Break, time=0, value=0}
          {interpMode=RCIM_Cubic, tangentMode=RCTM_Break, time=0.1, value=1}
          {arriveTangent=0.0378, interpMode=RCIM_Cubic, leaveTangent=0.0378, tangentMode=RCTM_User, time=0.6785, value=1.0282}
          {arriveTangent=-3.7547, interpMode=RCIM_Cubic, leaveTangent=-3.7547, tangentMode=RCTM_Break, time=0.8, value=0}
         ]}}}
      Scale Curve=&FNTimeOfDayScalar(Night Scale=1, Morning Scale=0.7, Day Scale=2, Evening Scale=0.5)
     )
     2=&FloatFromCurve(
      FloatCurve=DataInterface() {propertyValues {Curve {keys [
          {arriveTangent=-0.2723, interpMode=RCIM_Cubic, leaveTangent=14.5791, tangentMode=RCTM_Break, time=0, value=0}
          {interpMode=RCIM_Cubic, tangentMode=RCTM_Break, time=0.1, value=1}
          {arriveTangent=0.0378, interpMode=RCIM_Cubic, leaveTangent=0.0378, tangentMode=RCTM_User, time=0.6785, value=1.0282}
          {arriveTangent=-3.7547, interpMode=RCIM_Cubic, leaveTangent=-3.7547, tangentMode=RCTM_Break, time=0.8, value=0}
         ]}}}
      Scale Curve=&FNTimeOfDayScalar(Night Scale=1, Morning Scale=0.7, Day Scale=2, Evening Scale=0.5)
     )
     3=&FloatFromCurve(FloatCurve=DataInterface() {propertyValues {Curve {keys [
          {arriveTangent=6.8822, interpMode=RCIM_Cubic, leaveTangent=6.8822, tangentMode=RCTM_User, time=0, value=0}
          {interpMode=RCIM_Cubic, tangentMode=RCTM_User, time=0.2, value=1}
          {arriveTangent=-0.4524, tangentMode=RCTM_User, time=1, value=0.4}
         ]}}})
     4=&FloatFromCurve(FloatCurve=DataInterface() {propertyValues {Curve {keys [
          {arriveTangent=6.8822, interpMode=RCIM_Cubic, leaveTangent=6.8822, tangentMode=RCTM_User, time=0, value=0}
          {interpMode=RCIM_Cubic, tangentMode=RCTM_User, time=0.2, value=1}
          {arriveTangent=-0.4524, tangentMode=RCTM_User, time=1, value=0.4}
         ]}}})
     Index=&SpawnGroup()
     Num=::NewEnumerator2
    )
    Scale RGB=&MergeHelper_VectorSwitch(
     0=&VectorFromFloat(Value=&FNTimeOfDayScalar(Night Scale=3, Morning Scale=2.5, Day Scale=1.2, Evening Scale=3))
     1=&VectorFromFloat(Value=&FNTimeOfDayScalar(Night Scale=1.5, Morning Scale=2, Day Scale=1.2, Evening Scale=2))
     2=&VectorFromFloat(Value=&FNTimeOfDayScalar(Night Scale=1.5, Morning Scale=2, Day Scale=1.2, Evening Scale=2))
     3=&VectorFromFloat(Value=&FNTimeOfDayScalar(Night Scale=3, Morning Scale=1, Day Scale=50, Evening Scale=1))
     4=<1 1 1>
     Index=&SpawnGroup()
     Num=::NewEnumerator3
    )
    ScaleA=true
    ScaleRGB=true
   )
   &GravityForce(Gravity=&MergeHelper_VectorSwitch(
     0=<0 0 -500>
     1=&RandomRangeVector(Minimum=<0 0 -500>, Maximum=<1 1 -980>)
     2=&RandomRangeVector(Minimum=<0 0 -500>, Maximum=<0 0 -980>)
     3=<0 0 -500>
     4=<0 0 0>
     Index=&SpawnGroup()
     Num=::NewEnumerator3
    ))
   &Drag(
    Drag=&MergeHelper_FloatSwitch(
     0=0
     1=&RandomRangeFloat(Minimum=0.2, Maximum=0.3)
     2=&RandomRangeFloat(Minimum=0.2, Maximum=0.4)
     3=&FloatFromCurve(FloatCurve=DataInterface() {propertyValues {Curve {keys [{interpMode=RCIM_Cubic, tangentMode=RCTM_User, time=0, value=0}, {interpMode=RCIM_Cubic, tangentMode=RCTM_User, time=1, value=1}]}}})
     4=0
     Index=&SpawnGroup()
     Num=::NewEnumerator3
    )
    Rotational Drag:NiagaraFloat=&MergeHelper_FloatSwitch(0=1, 1=1, 2=1, 3=1, 4=1, Index=&SpawnGroup(), Num=::NewEnumerator3)
    Use Linear Drag=true
    Use Rotational Drag=true
   )
   &SolveForcesAndVelocity(Clamp Velocity=false, Limit Acceleration=false)
   &ScaleSpriteSizeBySpeed(
    Max Scale Factor=&MergeHelper_Vector2Switch(0=<1 1>, 1=<2 2>, 2=<2 2>, 3=<1.5 1.5>, 4=<2 2>, Index=&SpawnGroup(), Num=::NewEnumerator3)
    Min Scale Factor=&MergeHelper_Vector2Switch(0=<1 1>, 1=<1 1>, 2=<1 1>, 3=<1 1>, 4=<1 1>, Index=&SpawnGroup(), Num=::NewEnumerator3)
    Sample Scale Factor By Curve=true
    Scale Factor Curve=DataInterface() {propertyValues {Curve {defaultValue=0.5941, keys [{tangentMode=RCTM_User, time=0, value=1}, {tangentMode=RCTM_User, time=1, value=0}]}}}
    Velocity Threshold=250
   )
   /SetVariables(
    Particles.0 - Fat splash DynParam:Vector4f=&MakeVector4(
     X=&FloatFromCurve(FloatCurve=DataInterface() {propertyValues {Curve {keys [
          {interpMode=RCIM_Cubic, tangentMode=RCTM_User, time=0, value=1}
          {arriveTangent=-3.7889, interpMode=RCIM_Cubic, leaveTangent=-3.7889, tangentMode=RCTM_User, time=0.8, value=0}
         ]}}})
     Y=0
     Z=0
     W=0
    )
    Particles.1 - Thick Splash DynParam:Vector4f=&MakeVector4(
     X=&FloatFromCurve(FloatCurve=DataInterface() {propertyValues {Curve {defaultValue=0, keys [{tangentMode=RCTM_User, time=0, value=1}, {tangentMode=RCTM_User, time=1, value=0}]}}})
     Y=0
     Z=0
     W=0
    )
    Particles.2 - Thick Splash Bottom DynParam:Vector4f=&MakeVector4(
     X=&FloatFromCurve(FloatCurve=DataInterface() {propertyValues {Curve {defaultValue=0, keys [{tangentMode=RCTM_User, time=0, value=1}, {tangentMode=RCTM_User, time=1, value=0}]}}})
     Y=0
     Z=0
     W=0
    )
    Particles.3 - Super Thick Splash DynParam:Vector4f=&MakeVector4(
     X=&FloatFromCurve(FloatCurve=DataInterface() {propertyValues {Curve {keys [
          {interpMode=RCIM_Cubic, tangentMode=RCTM_User, time=0, value=1}
          {arriveTangent=-3.7889, interpMode=RCIM_Cubic, leaveTangent=-3.7889, tangentMode=RCTM_User, time=0.8, value=0}
         ]}}})
     Y=0
     Z=0
     W=0
    )
   ) {isSetParameters=true, module=SetVariables_#6}
  ]
  rendererClasses [&NiagaraSpriteRendererProperties]
  renderers [
   {rendererClass=&NiagaraSpriteRendererProperties}
   {rendererClass=&NiagaraSpriteRendererProperties, rendererIndex=1}
   {rendererClass=&NiagaraSpriteRendererProperties, rendererIndex=2}
   {rendererClass=&NiagaraSpriteRendererProperties, rendererIndex=3}
  ]
 }
 TOP_WATER_SPLASHED {
  bEnabled=false
  emitterUpdateScript [
   &EmitterState(Life Cycle Mode=::NewEnumerator1, Loop Behavior=::NewEnumerator1, Loop Delay=0.5, Loop Duration=3, UseLoopDelay=false)
   &SpawnRate(SpawnRate=100, Use Spawn Probability=false)
   &SpawnBurst_Instantaneous(Spawn Count=20, Spawn Group=1, Use Spawn Probability=false, UseLoopCountLimit=false) {enabled=false}
  ]
  particleSpawnScript [
   &InitializeParticle(
    Color=$System.Water splash color
    Lifetime Max=0.55
    Lifetime Min=0.2
    Lifetime Mode=::NewEnumerator1
    Position Offset=<0 0 10>
    Sprite Rotation Angle=&RandomRangeFloat(Minimum=0, Maximum=0.2)
    Sprite Rotation Mode=::NewEnumerator1
    Sprite Size Mode=::NewEnumerator1
    Sprite UV Mode=::NewEnumerator3
    Uniform Sprite Size Max=35
    Uniform Sprite Size Min=20
    UsePositionOffset=false
    Write Material Random=true
   )
   &CameraOffset(Camera Offset Amount=&RandomRangeFloat(Minimum=6, Maximum=15))
   &SkeletalMeshLocation(
    MaskSpawnGroup=false
    Mesh Sampling Type=::NewEnumerator5
    Orientation Sampling=::NewEnumerator1
    Sampled Position Offset=<-5 0 15>
    Skeletal Mesh=$System.Skeletal Mesh
    Socket Sampling Mode=::NewEnumerator5
   )
   &ShapeLocation(
    Disc Coverage=0.3405
    MaskSpawnGroup=false
    Offset=<0 0 -5>
    Offset Mode=::NewEnumerator0
    Override Seed=false
    Ring Radius=15
    Shape Primitive=::NewEnumerator4
    UseOwnerScale=true
   )
   &SubUVAnimation(
    End Frame Range Override=3
    Evaluation Type=::NewEnumerator0
    "Sprite Renderer "=DataInterface() {propertyValues {SpriteRenderer=&self:TOP_WATER_SPLASHED_0.NiagaraSpriteRendererProperties_1}}
    SubUV Animation Mode:ENiagara_SubUVLookupModeV2=::NewEnumerator2
    UseEndFrame=true
    UseStartFrame=true
   )
   &AddVelocity(
    Module.Use Falloff Distance=false
    Origin Offset=<0 0 -35>
    Origin Offset Coordinate Space=::Simulation
    Override Seed=false
    Velocity Mode=::NewEnumerator1
    Velocity Speed=&RandomRangeFloat(Minimum=75, Maximum=250)
   ) {module=AddVelocity001}
   &AddVelocity(Velocity=&RandomRangeVector(Minimum=<-75 -75 75>, Maximum=<75 75 250>)) {enabled=false}
  ]
  particleUpdateScript [
   &ParticleState()
   &SpriteRotationRate(Rotation Rate=&MaskFloat(
     Mask Source=0
     Mask Parameter=?
     Mask=0
     Masked Float=&RandomRangeFloat(Minimum=-70, Maximum=-100)
     Passthrough Float=&MaskFloat(
      Mask Source=0
      Mask Parameter=?
      Mask=1
      Masked Float=&RandomRangeFloat(Minimum=70, Maximum=100)
      Passthrough Float=1
     )
    ))
   &GravityForce(Gravity=&RandomRangeVector(Minimum=<0 0 -500>, Maximum=<0 0 -980>))
   &Drag(Drag=&RandomRangeFloat(Minimum=0.2, Maximum=0.3), Use Linear Drag=true, Use Rotational Drag=false)
   &SolveForcesAndVelocity(Clamp Velocity=false, Limit Acceleration=false)
   &ScaleSpriteSize(Uniform Curve Sprite Scale=DataInterface() {propertyValues {Curve {keys [
        {arriveTangent=4.1785, interpMode=RCIM_Cubic, leaveTangent=4.1785, tangentMode=RCTM_User, time=0, value=0.15}
        {arriveTangent=1.1179, interpMode=RCIM_Cubic, leaveTangent=1.1179, tangentMode=RCTM_User, time=0.2042, value=0.6228}
        {arriveTangent=0.0224, interpMode=RCIM_Cubic, leaveTangent=0.0224, tangentMode=RCTM_User, time=1, value=1}
       ]}}})
   &ScaleSpriteSizeBySpeed(
    Sample Scale Factor By Curve=true
    Scale Factor Curve=DataInterface() {propertyValues {Curve {defaultValue=0.5941, keys [{tangentMode=RCTM_User, time=0, value=1}, {tangentMode=RCTM_User, time=1, value=0}]}}}
    Velocity Threshold=250
   )
   &ScaleColor(
    Scale Alpha=&FloatFromCurve(
     FloatCurve=DataInterface() {propertyValues {Curve {keys [
         {arriveTangent=-0.2723, interpMode=RCIM_Cubic, leaveTangent=14.5791, tangentMode=RCTM_Break, time=0, value=0}
         {interpMode=RCIM_Cubic, tangentMode=RCTM_Break, time=0.1, value=1}
         {arriveTangent=0.0378, interpMode=RCIM_Cubic, leaveTangent=0.0378, tangentMode=RCTM_User, time=0.6785, value=1.0282}
         {arriveTangent=-3.7547, interpMode=RCIM_Cubic, leaveTangent=-3.7547, tangentMode=RCTM_Break, time=0.8, value=0}
        ]}}}
     Scale Curve=&FNTimeOfDayScalar(Night Scale=1, Morning Scale=0.7, Day Scale=2, Evening Scale=0.5)
    )
    Scale RGB=&VectorFromFloat(Value=&FNTimeOfDayScalar(Night Scale=1.5, Morning Scale=2, Day Scale=1.2, Evening Scale=2))
    ScaleA=true
    ScaleRGB=true
   )
   &DynamicMaterialParameters(
    Index 0 Param 1=&FloatFromCurve(FloatCurve=DataInterface() {propertyValues {Curve {defaultValue=0, keys [{tangentMode=RCTM_User, time=0, value=1}, {tangentMode=RCTM_User, time=1, value=0}]}}})
    Module.Index0Float0Write=true
    Module.Index0Float1Write=false
    Module.Index0Float2Write=false
    Module.Index0Float3Write=false
   )
  ]
  rendererClasses [&NiagaraSpriteRendererProperties]
  renderers [{rendererClass=&NiagaraSpriteRendererProperties}]
 }
}
summary {
 systemName=NS_Emote_GnatGalaPancake_splashes
 emitters [
  {bEnabled=false, emitterName=BOTTOM, rendererClasses [&NiagaraSpriteRendererProperties]}
  {bEnabled=false, emitterName=SplashFat, rendererClasses [&NiagaraSpriteRendererProperties]}
  {emitterName=Drip_Drop, rendererClasses [&NiagaraSpriteRendererProperties]}
  {bEnabled=false, emitterName=SplashFat002, rendererClasses [&NiagaraSpriteRendererProperties]}
  {bEnabled=false, emitterName=TOP_WATER_SPLASHED, rendererClasses [&NiagaraSpriteRendererProperties]}
  {emitterName=Splashes_compressed, simTarget=GPUComputeSim, rendererClasses [&NiagaraSpriteRendererProperties]}
  {emitterName=Splashes_compressed001, rendererClasses [&NiagaraSpriteRendererProperties]}
 ]
 userVariables [
  User.bIsFrontEnd {type=NiagaraBool}
  User.bIsFrontEndPreview {type=NiagaraBool}
  User.OwnerLocalVelocity {type=Vector3f, description="", defaultValue {struct=Vector3f, value=<0 0 0>}}
 ]
}
