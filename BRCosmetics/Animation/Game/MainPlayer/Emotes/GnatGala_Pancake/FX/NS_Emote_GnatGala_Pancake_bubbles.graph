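# Niagara system for the bubbling and splashing effect of the GnatGala Pancake player emote.
# Short-lived mesh bubbles spawn in a ring, grow and shrink, and act as the source for sprite droplets spawned through a particle attribute read.
# A separate sprite emitter throws splashes upward under gravity, with time-of-day brightness scaling across the effect.
# Emitters share a system loop length; the center, water and compressed ripple emitters are disabled.
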NiagaraSystem /BRCosmetics/Animation/Game/MainPlayer/Emotes/GnatGala_Pancake/FX/NS_Emote_GnatGala_Pancake_bubbles
dependencies {
 usedDataInterfaces [
  &NiagaraDataInterfaceCurve
  &NiagaraDataInterfaceColorCurve
  &NiagaraDataInterfaceParticleRead
  &NiagaraDataInterfaceVectorCurve
  &NiagaraDataInterfaceSpriteRendererInfo
 ]
 usedDynamicInputs [&RandomRangeLinearColor]
 usedModules [
  &self:SystemSpawnScript.SystemScriptSource.SystemScriptGraph.NiagaraNodeAssignment_2.SetVariables_#1
  &self:SystemSpawnScript.SystemScriptSource.SystemScriptGraph.NiagaraNodeAssignment_1.SetVariables_#2
  &self:SystemSpawnScript.SystemScriptSource.SystemScriptGraph.NiagaraNodeAssignment_0.SetVariables_#3
  &self:SystemSpawnScript.SystemScriptSource.SystemScriptGraph.NiagaraNodeAssignment_3.SetVariables_#4
  &SystemState
 ]
}
emitters {
 Bubbles {
  emitterUpdateScript [
   &EmitterState(Life Cycle Mode=::NewEnumerator1, Loop Behavior=::NewEnumerator1, Loop Duration=$System.loop length, UseLoopDelay=false)
   &SpawnRate(SpawnRate=15, Use Spawn Probability=false)
  ]
  particleSpawnScript [
   /SetVariables(Particles.VisibilityTag=&SpawnGroup()) {enabled=false, isSetParameters=true}
   &FN_SetSpawnGroup()
   &InitializeParticle(
    Color Maximum=<rgba 0.1365 0.1365 0.1365 0.5>
    Color Minimum=<rgba 0.4322 0.634 0.849 0.5>
    Color Mode=::NewEnumerator2
    Lifetime Max=0.4
    Lifetime Min=0.2
    Lifetime Mode=::NewEnumerator1
    Mesh Scale Mode=::NewEnumerator1
    Mesh Uniform Scale Max=0.07
    Mesh Uniform Scale Min=0.0123
    Sprite Rotation Mode=::NewEnumerator2
    UsePositionOffset=false
    Write Material Random=true
   )
   &ShapeLocation(Disc Coverage=1, MaskSpawnGroup=false, Override Seed=false, Ring Radius=15, Shape Primitive=::NewEnumerator4, UseOwnerScale=true)
   &CameraOffset(Camera Offset Amount=11)
  ]
  particleUpdateScript [
   &ParticleState()
   &ScaleMeshSize(Scale Factor=&VectorFromFloat(Value=&FloatFromCurve(FloatCurve=DataInterface() {propertyValues {Curve {keys [<at 0 0.2>,<at 0.8 1>,<at 1 0>]}}})))
   &DynamicMaterialParameters(
    Index 0 Param 1=&FloatFromCurve(FloatCurve=DataInterface() {propertyValues {Curve {keys [
         <at 0 0>
         {interpMode=RCIM_Cubic, leaveTangent=1.2072, tangentMode=RCTM_Break, time=0.55, value=0}
         {arriveTangent=3.283, tangentMode=RCTM_User, time=1, value=1}
        ]}}})
    Index 0 Param 4=1.1
    Module.Index0Param0Write=true
    Module.Index0Param1Write=true
    Module.Index0Param2Write=true
    Module.Index0Param3Write=true
   )
   &GravityForce(Gravity=&MakeVector(
     X=0
     Y=0
     Z=&FloatFromCurve(FloatCurve=DataInterface() {propertyValues {Curve {defaultValue=1, keys [{tangentMode=RCTM_User, time=0, value=0}, {tangentMode=RCTM_User, time=1, value=1}]}}})
    ))
   &SolveForcesAndVelocity(Clamp Velocity=false, Limit Acceleration=false)
   /SetVariables(
    Particles.SplashDynamicParamenter:Vector4f=&MakeVector4(
     X=&FloatFromCurve(FloatCurve=DataInterface() {propertyValues {Curve {keys [
          <at 0 1>
          {interpMode=RCIM_Cubic, leaveTangent=3.8935, tangentMode=RCTM_Break, time=0.8, value=1.5}
          {arriveTangent=171.8661, tangentMode=RCTM_User, time=1, value=17}
         ]}}})
     Y=0.5
     Z=0
     W=0
    )
    Particles.SplashIndex:NiagaraFloat=&RandomRangeFloat(Minimum=0, Maximum=3)
    Particles.SplashSize:Vector2f=&Vector2DFromFloat(Value=&FloatFromCurve(
      FloatCurve=DataInterface() {propertyValues {Curve {keys [<at 0 0>,<at 0.7 0>,<at 0.85 1>]}}}
      Scale Curve=&RandomRangeFloat(Minimum=15, Maximum=20)
     ))
   ) {isSetParameters=true}
   /SetVariables(
    Particles.WaveColor:LinearColor=&ColorFromCurve(ColorCurve=DataInterface() {propertyValues {
       AlphaCurve {keys [<at 0.6765 0>,<at 0.7462 1>,<at 0.9338 0 RCIM_Cubic>]}
       BlueCurve {keys [<at 0.8093 0.849 RCIM_Cubic>,<at 1 0.1365 RCIM_Cubic>]}
       GreenCurve {keys [<at 0.8093 0.634 RCIM_Cubic>,<at 1 0.1365 RCIM_Cubic>]}
       RedCurve {keys [<at 0.8093 0.4322 RCIM_Cubic>,<at 1 0.1365 RCIM_Cubic>]}
      }})
    Particles.WaveSize:Vector2f=&Vector2DFromFloat(Value=&MakeFloatFromVector2d(Vector 2D=&Vector2DFromFloat(Value=&FloatFromCurve(
        FloatCurve=DataInterface() {propertyValues {Curve {keys [<at 0 0>,<at 0.7 0>,<at 0.8 1>]}}}
        Scale Curve=&RandomRangeFloat(Minimum=25, Maximum=30)
       ))))
   ) {isSetParameters=true}
   /SetVariables(
    Particles.RippleColor:LinearColor=&MultiplyLinearColor(
     Linear Color=<rgba 1 1 1 0.1>
     Scale Factor=&MakeVector4_FromLinearColor(LinearColor=&FNTimeOfDayColor(Night Color=<rgba 0.556 0.5008 1 1>, Morning Color=<rgba 1 0.6695 0.6095 1>, Day Color=<rgba 1 1 1 1>, Evening Color=<rgba 1 1 1 1>))
    )
    Particles.RippleFacing:Vector3f=<0 0 1>
    Particles.RippleSize:Vector2f=&Vector2DFromFloat(Value=&FloatFromCurve(FloatCurve=DataInterface() {propertyValues {Curve {keys [<at 0.5 0>,<at 1 1>]}}}, Scale Curve=35))
   ) {isSetParameters=true}
   /SetVariables(Particles.bubble:LinearColor=<rgba 0.7782 1 0.9558 0.5>) {isSetParameters=true}
   &ScaleColor(
    Scale RGB=&VectorFromFloat(Value=&FNTimeOfDayScalar(Night Scale=2, Morning Scale=2, Day Scale=2, Evening Scale=2))
    ScaleA=true
    ScaleRGB=true
   )
  ]
  rendererClasses [&NiagaraMeshRendererProperties]
  renderers [{rendererClass=&NiagaraMeshRendererProperties, rendererIndex=0}]
 }
 Center_Ripple001 {
  bEnabled=false
  emitterUpdateScript [
   &EmitterState(Life Cycle Mode=::NewEnumerator1, Loop Behavior=::NewEnumerator1, Loop Delay=0.05, Loop Duration=$System.loop length, UseLoopDelay=false)
   &SpawnRate(SpawnRate=4.8543, Use Spawn Probability=false)
  ]
  particleSpawnScript [
   &CameraOffset(Camera Offset Amount=-3)
   &InitializeParticle(
    Color=&FNTimeOfDayColor(
     Night Color=<rgba 0.6199 0.7308 0.849 0.5>
     Morning Color=<rgba 0.6199 0.7308 0.849 0.7>
     Day Color=<rgba 0.6199 0.7308 0.849 0.5>
     Evening Color=<rgba 0.6199 0.7308 0.849 0.5>
    )
    Lifetime=0.5
    Mesh Scale Mode=::NewEnumerator3
    Mesh Uniform Scale=0.35
    Position Offset=<0 0 -2>
    UsePositionOffset=true
    Write Material Random=true
   )
  ]
  particleUpdateScript [
   &ParticleState()
   &ScaleColor(
    Scale Alpha=&FloatFromCurve(
     FloatCurve=DataInterface() {propertyValues {Curve {keys [
         {arriveTangent=9.0725, interpMode=RCIM_Cubic, leaveTangent=9.0725, tangentMode=RCTM_User, time=0, value=0}
         {arriveTangent=-2.5249, interpMode=RCIM_Cubic, leaveTangent=-2.5249, tangentMode=RCTM_User, time=0.1408, value=0.9847}
         {arriveTangent=-0.0405, interpMode=RCIM_Cubic, leaveTangent=-0.0405, tangentMode=RCTM_User, time=1, value=0}
        ]}}}
     Scale Curve=&FNTimeOfDayScalar(Night Scale=1, Morning Scale=0.3, Day Scale=0.5, Evening Scale=0.3)
    )
    Scale RGB=&VectorFromFloat(Value=&FNTimeOfDayScalar(Night Scale=1, Morning Scale=2, Day Scale=2, Evening Scale=0.7))
    ScaleA=true
    ScaleRGB=true
   )
   &ScaleMeshSize(Scale Factor=&VectorFromCurve(VectorCurve=DataInterface() {propertyValues {XCurve {keys [<at 0 1.0241>,<at 1 3>]}, YCurve {keys [<at 0 1>,<at 1 3>]}, ZCurve {keys [<at 0 1>,<at 1 1.0066>]}}}))
   &DynamicMaterialParameters(
    Index 0 Param 1=&RandomRangeFloat(Minimum=0.1, Maximum=0.55)
    Index 0 Param 2=&FloatFromCurve(
     FloatCurve=DataInterface() {propertyValues {Curve {keys [
         {interpMode=RCIM_Cubic, tangentMode=RCTM_Break, time=0, value=0}
         {interpMode=RCIM_Cubic, tangentMode=RCTM_Break, time=0.5, value=40}
         {interpMode=RCIM_Cubic, tangentMode=RCTM_Break, time=1, value=0}
        ]}}}
     Scale Curve=0.65
    )
    Module.Param0WriteEnabled=true
    Module.Param1WriteEnabled=true
    Module.Param2WriteEnabled=true
    Module.Param3WriteEnabled=true
   )
  ]
  rendererClasses [&NiagaraMeshRendererProperties]
  renderers [{rendererClass=&NiagaraMeshRendererProperties, rendererIndex=0}]
 }
 Minimal {
  emitterUpdateScript [
   &EmitterState(Life Cycle Mode=::NewEnumerator1, Loop Duration=$System.loop length, UseLoopDelay=false)
   &SpawnParticlesFromOtherEmitter(
    Attribute Reader=DataInterface() {propertyValues {EmitterBinding {bindingMode=Other, emitterName=Bubbles}}}
    Spawn Rate=&FloatFromCurve(
     FloatCurve=DataInterface() {propertyValues {Curve {keys [<at 0.1002 -0.0376>,<at 0.3486 0>,<at 0.499 0.8437>]}}}
     CurveIndex=$Emitter.NormalizedLoopAge
     Scale Curve=65
    )
    Spawn Rate Per Particle Cap=35
    Use Maximum Age=false
    Use Minimum Age=false
   )
  ]
  particleSpawnScript [
   &InitializeParticle(
    Color=$System.DropletColors
    Lifetime Max=0.1
    Lifetime Min=0.05
    Lifetime Mode=::NewEnumerator1
    Sprite Size Mode=::NewEnumerator1
    Uniform Sprite Size Max=0.8
    Uniform Sprite Size Min=0.5
    UsePositionOffset=false
    Write Material Random=true
   )
   &SampleParticlesFromOtherEmitter(
    Age Sampling:ENiagara_DirectReadSamplingMode=::NewEnumerator0
    Applied Position Offset:Vector3f=<0 0 0.9536>
    Lifetime Application:ENiagara_LifetimeInheritanceOptions=::NewEnumerator1
    Lifetime Sampling:ENiagara_DirectReadSamplingMode=::NewEnumerator0
   )
   &AddVelocity(Velocity=&RandomRangeVector(Minimum=<-30 -30 60>, Maximum=<30 30 80>))
   &ShapeLocation(MaskSpawnGroup=false, Override Seed=false, Sphere Radius=1, Sphere Surface Distribution=0.319, UseOwnerScale=true) {enabled=false}
  ]
  particleUpdateScript [
   &ParticleState()
   &ScaleSpriteSize(Scale Factor=&Vector2DFromFloat(Value=/Game/Effects/Niagara/DynamicInputs/FN_PickFloatFromSpawnGroupCutoff(
      Over Cutoff Float:NiagaraFloat=0
      Under Cutoff Float:NiagaraFloat=&FloatFromCurve(FloatCurve=DataInterface() {propertyValues {Curve {keys [
           {interpMode=RCIM_Cubic, tangentMode=RCTM_Break, time=0.5, value=0.9489}
           {arriveTangent=-3.7049, interpMode=RCIM_Cubic, tangentMode=RCTM_Break, time=1.0011, value=0.002}
          ]}}})
      SpawnGroup:NiagaraInt32=$Emitter.SpawnGroup
      Cutoff:NiagaraInt32=2
     )))
   &GravityForce(Gravity=&RandomRangeVector(Minimum=<0 0 -200>, Maximum=<1 1 -400>))
   &Drag(Drag=5, Use Linear Drag=true, Use Rotational Drag=false)
   &SolveForcesAndVelocity(Clamp Velocity=false, Limit Acceleration=false)
   &ScaleSpriteSizeBySpeed(Max Scale Factor=<0.9 6>, Min Scale Factor=<0.7 4>, Sample Scale Factor By Curve=true, Velocity Threshold=100)
   &ScaleColor(
    Scale Alpha=&FloatFromCurve(FloatCurve=DataInterface() {propertyValues {Curve {
        defaultValue=1.0113
        keys [
         {interpMode=RCIM_Cubic, tangentMode=RCTM_User, time=0.1, value=0.019}
         {interpMode=RCIM_Cubic, tangentMode=RCTM_User, time=0.5, value=1}
         {tangentMode=RCTM_User, time=1, value=0}
        ]
       }}})
    Scale RGB=&VectorFromFloat(Value=&FNTimeOfDayScalar(Night Scale=2, Morning Scale=3, Day Scale=5, Evening Scale=2))
    ScaleA=true
    ScaleRGB=true
   )
  ]
  rendererClasses [&NiagaraSpriteRendererProperties]
  renderers [
   {rendererClass=&NiagaraSpriteRendererProperties}
   {rendererClass=&NiagaraSpriteRendererProperties, rendererIndex=1}
  ]
 }
 Splash_Both001 {
  emitterUpdateScript [
   &EmitterState(Life Cycle Mode=::NewEnumerator1, Loop Behavior=::NewEnumerator1, Loop Delay=0.5, Loop Duration=$System.loop length, UseLoopDelay=false)
   &SpawnRate(SpawnRate=100, Use Spawn Probability=false)
  ]
  particleSpawnScript [
   &InitializeParticle(
    Color=$System.Water splash color
    Lifetime=&RandomRangeFloat(Minimum=0.3, Maximum=0.6)
    Position Offset=<0 0 -5>
    Sprite Size=&Multiply_Vector2D_ByFloat(Float=0.45, Vector2D=&RandomRangeVector2D(Minimum=<5 5>, Maximum=<5 5>))
    Sprite Size Mode=::NewEnumerator4
    UsePositionOffset=true
    Write Material Random=true
   ) {module=InitializeParticle001}
   &ShapeLocation(
    MaskSpawnGroup=false
    Offset=<0 0 -3>
    Offset Mode=::NewEnumerator0
    Override Seed=false
    Ring Radius=3
    Shape Primitive=::NewEnumerator4
    UseOwnerScale=true
   )
   &AddVelocity(
    Cone Axis=<0 0 1>
    Module.UseConeFalloff=false
    Module.UseVelDistribution=false
    Override Seed=false
    Velocity Mode=::NewEnumerator2
    Velocity Speed=&Multiply_Float(A=&RandomRangeFloat(Minimum=150, Maximum=200), B=1)
   ) {module=AddVelocity002}
   &AddVelocity(Velocity=&Multiply_VectorByFloat(Vector=&RandomRangeVector(Minimum=<0 0 10>, Maximum=<0 0 50>), Float=1)) {module=AddVelocity001}
   &CameraOffset(Camera Offset Amount=-2)
   &SubUVAnimation(End Frame=3, SubUV Animation Mode:ENiagara_SubUVLookupMode=::NewEnumerator2)
  ]
  particleUpdateScript [
   &ParticleState()
   &SpriteRotationRate(Rotation Rate=&Lerp_Float(A=-360, B=306, Alpha=$Particles.Initial.MaterialRandom))
   &ScaleSpriteSize(Scale Factor=&Vector2DFromFloat(Value=&FloatFromCurve(
      FloatCurve=DataInterface() {propertyValues {Curve {keys [
          {arriveTangent=10.0275, interpMode=RCIM_Cubic, leaveTangent=10.0275, tangentMode=RCTM_User, time=0, value=0.4376}
          {arriveTangent=1.6841, interpMode=RCIM_Cubic, leaveTangent=1.6841, tangentMode=RCTM_User, time=1, value=5}
         ]}}}
      Scale Curve=2
     )))
   &GravityForce(Gravity=&Multiply_VectorByFloat(Vector=<0 0 -980>, Float=&RandomRangeFloat(Minimum=1, Maximum=2)))
   &SolveForcesAndVelocity(Clamp Velocity=false, Limit Acceleration=false)
   &ScaleColor(
    Scale Alpha=&FloatFromCurve(
     FloatCurve=DataInterface() {propertyValues {Curve {keys [
         {arriveTangent=-0.2723, interpMode=RCIM_Cubic, leaveTangent=14.5791, tangentMode=RCTM_Break, time=0, value=0}
         {interpMode=RCIM_Cubic, tangentMode=RCTM_Break, time=0.1, value=1}
         {arriveTangent=-3.7547, interpMode=RCIM_Cubic, leaveTangent=-3.7547, tangentMode=RCTM_Break, time=0.3823, value=0.0753}
         {arriveTangent=0.0378, interpMode=RCIM_Cubic, leaveTangent=0.0378, tangentMode=RCTM_User, time=0.5404, value=0}
        ]}}}
     Scale Curve=&FNTimeOfDayScalar(Night Scale=0.5, Morning Scale=0.7, Day Scale=0.7, Evening Scale=1)
    )
    Scale RGB=&VectorFromFloat(Value=&FNTimeOfDayScalar(Night Scale=1.5, Morning Scale=2, Day Scale=2, Evening Scale=2))
    ScaleA=true
    ScaleRGB=true
   ) {module=ScaleColor001}
   &DynamicMaterialParameters(
    Index 0 Param 1=&FloatFromCurve(FloatCurve=DataInterface() {propertyValues {Curve {keys [
         {interpMode=RCIM_Cubic, tangentMode=RCTM_User, time=0, value=1}
         {arriveTangent=-3.7889, interpMode=RCIM_Cubic, leaveTangent=-3.7889, tangentMode=RCTM_User, time=0.8, value=0}
        ]}}})
    Module.Param0WriteEnabled=true
    Module.Param1WriteEnabled=true
    Module.Param2WriteEnabled=true
    Module.Param3WriteEnabled=true
   ) {module=DynamicMaterialParameters002}
  ]
  rendererClasses [&NiagaraSpriteRendererProperties]
  renderers [{rendererClass=&NiagaraSpriteRendererProperties}]
 }
 WATER_RIPPLE_lwe {rendererClasses [], renderers []}
 ripple_compressed001 {
  bEnabled=false
  emitterUpdateScript [
   &EmitterState(Life Cycle Mode=::NewEnumerator1, Loop Behavior=::NewEnumerator1, Loop Delay=0.5, Loop Duration=$System.loop length, UseLoopDelay=false)
   &SpawnRate(SpawnGroup=1, SpawnRate=5, Use Spawn Probability=false) {module=SpawnRate001}
   &SpawnRate(SpawnRate=100, Use Spawn Probability=false)
  ]
  particleSpawnScript [
   /SetVariables(Particles.VisibilityTag=&SpawnGroup()) {isSetParameters=true}
   &InitializeParticle(
    Color=&MergeHelper_ColorSwitch(
     0=$System.Water splash color
     1=&FNTimeOfDayColor(
      Night Color=<rgba 0.6199 0.7308 0.849 0.5>
      Morning Color=<rgba 0.6199 0.7308 0.849 0.7>
      Day Color=<rgba 0.6199 0.7308 0.849 0.5>
      Evening Color=<rgba 0.6199 0.7308 0.849 0.5>
     )
     2=?
     3=?
     Index=&SpawnGroup()
     Num=::NewEnumerator0
    )
    Lifetime=&MergeHelper_FloatSwitch(
     0=&RandomRangeFloat(Minimum=0.3, Maximum=0.6)
     1=0.5
     2=&RandomRangeFloat(Minimum=0.2, Maximum=0.4)
     3=&RandomRangeFloat(Minimum=0.05, Maximum=0.1)
     Index=&SpawnGroup()
     Num=::NewEnumerator2
    )
    Mesh Scale Mode=::NewEnumerator3
    Mesh Uniform Scale=0.35
    Position Offset=<0 0 -5>
    Sprite Size=&Multiply_Vector2D_ByFloat(Float=0.45, Vector2D=&RandomRangeVector2D(Minimum=<5 5>, Maximum=<5 5>))
    Sprite Size Mode=::NewEnumerator4
    UsePositionOffset=true
    Write Material Random=true
   )
   &ShapeLocation(
    MaskSpawnGroup=false
    Override Seed=false
    Ring Radius=&MergeHelper_FloatSwitch(0=3, 1=0, 2=0, 3=0, Index=&SpawnGroup(), Num=::NewEnumerator2)
    Shape Primitive=::NewEnumerator4
    UseOwnerScale=true
   )
   &AddVelocity(
    Cone Axis=<0 0 1>
    Module.UseConeFalloff=false
    Module.UseVelDistribution=false
    Override Seed=false
    Velocity Mode=::NewEnumerator2
    Velocity Speed=&MergeHelper_FloatSwitch(
     0=&Multiply_Float(A=&RandomRangeFloat(Minimum=150, Maximum=200), B=1)
     1=0
     2=0
     3=0
     4=0
     Index=$Particles.VisibilityTag
     Num=::NewEnumerator3
    )
   )
   &AddVelocity(Velocity=&MergeHelper_VectorSwitch(
     0=&Multiply_VectorByFloat(Vector=&RandomRangeVector(Minimum=<0 0 10>, Maximum=<0 0 50>), Float=1)
     1=<0 0 0>
     2=<0 0 0>
     3=<0 0 0>
     Index=&SpawnGroup()
     Num=::NewEnumerator2
    )) {module=AddVelocity001}
   &ApplyInitialForces()
   &CameraOffset(Camera Offset Amount=&MergeHelper_FloatSwitch(0=-2, 1=-3, 2=0, 3=0, Index=&SpawnGroup(), Num=::NewEnumerator2))
   &SubUVAnimation(
    "Sprite Renderer "=DataInterface() {propertyValues {SpriteRenderer=&self:ripple_compressed001_0.NiagaraSpriteRendererProperties_1}}
    SubUV Animation Mode:ENiagara_SubUVLookupModeV2=::NewEnumerator2
    UseEndFrame=false
    UseStartFrame=false
   )
   /SetVariables(
    Particles.Ripple Scale:Vector3f=<0.35 0.35 35>
    Particles.ripple lifetime=0.5
    Particles.ripple position:NiagaraPosition=&AddVectorToPosition(Position=<0 0 -2>, Vector=<0 0 0>)
   ) {isSetParameters=true}
  ]
  particleUpdateScript [
   &ParticleState()
   &ScaleMeshSize(Scale Factor=&VectorFromCurve(VectorCurve=DataInterface() {propertyValues {XCurve {keys [<at 0 1>,<at 1 3>]}, YCurve {keys [<at 0 1>,<at 1 3>]}, ZCurve {keys [<at 0 1>,<at 1 1>]}}}))
   &ScaleSpriteSize(
    Scale Sprite Size Mode=::NewEnumerator0
    Uniform Scale Factor=&FloatFromCurve(
     FloatCurve=DataInterface() {propertyValues {Curve {keys [
         {arriveTangent=10.0275, interpMode=RCIM_Cubic, leaveTangent=10.0275, tangentMode=RCTM_User, time=0, value=0.4376}
         {arriveTangent=1.6841, interpMode=RCIM_Cubic, leaveTangent=1.6841, tangentMode=RCTM_User, time=1, value=5}
        ]}}}
     Scale Curve=2
    )
   )
   &GravityForce(Gravity=&MergeHelper_VectorSwitch(
     0=&Multiply_VectorByFloat(Vector=<0 0 -980>, Float=&RandomRangeFloat(Minimum=1, Maximum=2))
     1=<0 0 500>
     2=<0 0 0>
     3=&RandomRangeVector(Minimum=<0 0 -200>, Maximum=<1 1 -400>)
     Index=&SpawnGroup()
     Num=::NewEnumerator2
    ))
   &SolveForcesAndVelocity(Clamp Velocity=false, Limit Acceleration=false)
   &ScaleColor(
    Scale Alpha=&MergeHelper_FloatSwitch(
     0=&FloatFromCurve(
      FloatCurve=DataInterface() {propertyValues {Curve {keys [
          {arriveTangent=-0.2723, interpMode=RCIM_Cubic, leaveTangent=14.5791, tangentMode=RCTM_Break, time=0, value=0}
          {interpMode=RCIM_Cubic, tangentMode=RCTM_Break, time=0.1, value=1}
          {arriveTangent=-3.7547, interpMode=RCIM_Cubic, leaveTangent=-3.7547, tangentMode=RCTM_Break, time=0.3823, value=0.0753}
          {arriveTangent=0.0378, interpMode=RCIM_Cubic, leaveTangent=0.0378, tangentMode=RCTM_User, time=0.5404, value=0}
         ]}}}
      Scale Curve=&FNTimeOfDayScalar(Night Scale=0.5, Morning Scale=0.7, Day Scale=0.7, Evening Scale=1)
     )
     1=&FloatFromCurve(
      FloatCurve=DataInterface() {propertyValues {Curve {keys [
          {arriveTangent=9.0725, interpMode=RCIM_Cubic, leaveTangent=9.0725, tangentMode=RCTM_User, time=0, value=0}
          {arriveTangent=-2.5249, interpMode=RCIM_Cubic, leaveTangent=-2.5249, tangentMode=RCTM_User, time=0.1408, value=0.9847}
          {arriveTangent=-0.0405, interpMode=RCIM_Cubic, leaveTangent=-0.0405, tangentMode=RCTM_User, time=1, value=0}
         ]}}}
      Scale Curve=&FNTimeOfDayScalar(Night Scale=1, Morning Scale=0.3, Day Scale=0.5, Evening Scale=0.3)
     )
     2=?
     3=?
     Index=&SpawnGroup()
     Num=::NewEnumerator0
    )
    Scale RGB=&MergeHelper_VectorSwitch(
     0=&VectorFromFloat(Value=&FNTimeOfDayScalar(Night Scale=1.5, Morning Scale=2, Day Scale=2, Evening Scale=2))
     1=&VectorFromFloat(Value=&FNTimeOfDayScalar(Night Scale=1, Morning Scale=2, Day Scale=2, Evening Scale=0.7))
     2=?
     3=?
     Index=&SpawnGroup()
     Num=::NewEnumerator0
    )
    ScaleA=true
    ScaleRGB=true
   )
   /SetVariables(
    Particles.1 - center ripple dynParam:Vector4f=&MakeVector4(
     X=&RandomRangeFloat(Minimum=0.1, Maximum=0.55)
     Y=&FloatFromCurve(
      FloatCurve=DataInterface() {propertyValues {Curve {keys [
          {interpMode=RCIM_Cubic, tangentMode=RCTM_Break, time=0, value=0}
          {interpMode=RCIM_Cubic, tangentMode=RCTM_Break, time=0.5, value=40}
          {interpMode=RCIM_Cubic, tangentMode=RCTM_Break, time=1, value=0}
         ]}}}
      Scale Curve=0.5
     )
     Z=0
     W=0
    )
    Particles.Fountain water splash dynparam:Vector4f=&MakeVector4(
     X=&FloatFromCurve(FloatCurve=DataInterface() {propertyValues {Curve {keys [
          {interpMode=RCIM_Cubic, tangentMode=RCTM_User, time=0, value=1}
          {arriveTangent=-3.7889, interpMode=RCIM_Cubic, leaveTangent=-3.7889, tangentMode=RCTM_User, time=0.8, value=0}
         ]}}})
     Y=0
     Z=0
     W=0
    )
   ) {isSetParameters=true}
   /SetVariables(
    Particles.Ripple age=0.5
    Particles.Update ripple scale:Vector3f=&VectorFromCurve(
     VectorCurve=DataInterface() {propertyValues {XCurve {keys [<at 0 1>,<at 1 3>]}, YCurve {keys [<at 0 1>,<at 1 3>]}, ZCurve {keys [<at 0 1>,<at 1 1>]}}}
     CurveIndex=$Particles.ripple lifetime
    )
    Particles.ripple color:LinearColor=&MakeLinearColorFromVectorAndFloat(
     "Vector (RGB)"=<1 1 1>
     "Float (Alpha)"=&FloatFromCurve(
      FloatCurve=DataInterface() {propertyValues {Curve {keys [
          {arriveTangent=9.0725, interpMode=RCIM_Cubic, leaveTangent=9.0725, tangentMode=RCTM_User, time=0, value=0}
          {arriveTangent=-2.5249, interpMode=RCIM_Cubic, leaveTangent=-2.5249, tangentMode=RCTM_User, time=0.1408, value=0.9847}
          {arriveTangent=-0.0405, interpMode=RCIM_Cubic, leaveTangent=-0.0405, tangentMode=RCTM_User, time=0.7313, value=0.0651}
         ]}}}
      Scale Curve=&FNTimeOfDayScalar(Night Scale=1, Morning Scale=0.3, Day Scale=0.5, Evening Scale=0.3)
     )
    )
   ) {isSetParameters=true}
  ]
  rendererClasses [&NiagaraSpriteRendererProperties, &NiagaraMeshRendererProperties]
  renderers [
   {rendererClass=&NiagaraSpriteRendererProperties}
   {rendererClass=&NiagaraMeshRendererProperties, rendererIndex=1}
  ]
 }
 tHICC_WATER_RIPPLE_lwe001 {rendererClasses [], renderers []}
 water_ripples001 {
  bEnabled=false
  emitterUpdateScript [
   &EmitterState(Life Cycle Mode=::NewEnumerator1, Loop Behavior=::NewEnumerator1, Loop Delay=0.5, Loop Duration=1.5, UseLoopDelay=false)
   &SpawnRate(SpawnRate=2, Use Spawn Probability=false)
  ]
  particleSpawnScript [&InitializeParticle(
    Color=$System.Ripples color
    Lifetime=&RandomRangeFloat(Minimum=1, Maximum=2)
    Mesh Scale Mode=::NewEnumerator1
    Mesh Uniform Scale Max=1
    Mesh Uniform Scale Min=0.5
    UsePositionOffset=true
    Write Material Random=true
   )]
  particleUpdateScript [
   &ParticleState()
   &ScaleColor(
    Scale Alpha=&FloatFromCurve(
     FloatCurve=DataInterface() {propertyValues {Curve {
        defaultValue=0
        keys [
         {interpMode=RCIM_Cubic, tangentMode=RCTM_User, time=0, value=0}
         {arriveTangent=-0.0455, interpMode=RCIM_Cubic, leaveTangent=-0.0455, tangentMode=RCTM_User, time=0.2151, value=1}
         {interpMode=RCIM_Cubic, tangentMode=RCTM_User, time=0.5, value=1}
         {tangentMode=RCTM_User, time=0.89, value=0}
        ]
       }}}
     Scale Curve=&FNTimeOfDayScalar(Night Scale=3, Morning Scale=2.5, Day Scale=5, Evening Scale=3)
    )
    Scale RGB=&VectorFromFloat(Value=&FNTimeOfDayScalar(Night Scale=1.5, Morning Scale=1.5, Day Scale=1.5, Evening Scale=1.5))
    ScaleA=true
    ScaleRGB=true
   )
   &ScaleMeshSize(Scale Factor=&VectorFromFloat(Value=&RandomRangeFloat(
      Minimum=&FloatFromCurve(FloatCurve=DataInterface() {propertyValues {Curve {defaultValue=1, keys [{tangentMode=RCTM_User, time=0, value=0.5}, {tangentMode=RCTM_User, time=1, value=1}]}}})
      Maximum=&FloatFromCurve(
       FloatCurve=DataInterface() {propertyValues {Curve {defaultValue=1, keys [{tangentMode=RCTM_User, time=0, value=0.5}, {tangentMode=RCTM_User, time=1, value=1}]}}}
       Scale Curve=2.3
      )
     ))) {module=ScaleMeshSize001}
  ]
  rendererClasses [&NiagaraMeshRendererProperties]
  renderers [{rendererClass=&NiagaraMeshRendererProperties, rendererIndex=0}]
 }
}
summary {
 systemName=NS_Emote_GnatGala_Pancake_bubbles
 emitters [
  {emitterName=Bubbles, rendererClasses [&NiagaraMeshRendererProperties]}
  {emitterName=Minimal, rendererClasses [&NiagaraSpriteRendererProperties]}
  {emitterName=Splash_Both001, rendererClasses [&NiagaraSpriteRendererProperties]}
  {bEnabled=false, emitterName=Center_Ripple001, rendererClasses [&NiagaraMeshRendererProperties]}
  {bEnabled=false, emitterName=water_ripples001, rendererClasses [&NiagaraMeshRendererProperties]}
  {
   bEnabled=false
   emitterName=ripple_compressed001
   rendererClasses [&NiagaraSpriteRendererProperties, &NiagaraMeshRendererProperties]
  }
  {emitterName=WATER_RIPPLE_lwe, rendererClasses []}
  {emitterName=tHICC_WATER_RIPPLE_lwe001, rendererClasses []}
 ]
 userVariables []
}
